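# Unlit translucent material for animated, flame-like hair strips on the Astral character's skeletal mesh.
# Vertex math combines SplineThicken width, sine/cosine waves, weighted camera offset and PlayerState-driven running offsets.
# Distorted, panning noise drives an eroded strand mask and brightened hotspots over a tip-to-root color gradient, with depth fade by quality level.
# Sliced pre-skinned X position staggers strand timing; virtual clip-plane math is also present.
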Material /BRCosmetics/Characters/Player/Female/Medium/Bodies/F_MED_Astral/FX/M_FX_AstralHairStrips_Mesh
graph {
 properties {
  bUsedWithMeshDeformer=true
  bUsedWithSkeletalMesh=true
  blendMode=BLEND_Translucent
  pixelDepthOffsetMode=PDOM_Legacy
  shadingModel=MSM_Unlit
  usedShadingModels=Unlit
 }
 nodes [
  n1 = VectorParameter() {group=PlayerState, parameterName=Running_Hair_Offsets, defaultValue=<rgba 0 0 0 1>}
  n2 = Multiply(A=Transform(None=@n1.RGB) {transformSourceType=TRANSFORMSOURCE_Local}, B=Multiply(A=@n1.A, B=NamedRerouteUsage()))
  n3 = ScalarParameter() {group=Particle_Replacement, parameterName=Wave_Strength}
  n4 = Transform(None=Normalize(VectorInput=VectorParameter() {group=Waves, parameterName=Wave_Direction, defaultValue=<rgba 0 0 0 1>}.RGB)) {transformSourceType=TRANSFORMSOURCE_Local}
  n5 = ComponentMask(None=TextureCoordinate()) {g=true}
  n6 = Add(A=@n5, B=Multiply(A=NamedRerouteUsage()) {constB=1.5})
  n7 = Transform(None=Normalize(VectorInput=VectorParameter() {group=Waves, parameterName=Wave_Direction_Alt, defaultValue=<rgba 0 1 0 1>}.RGB)) {transformSourceType=TRANSFORMSOURCE_Local}
  n8 = Add(A=Multiply(A=Multiply(A=@n3, B=@n4), B=Sine(None=@n6)), B=Multiply(A=Multiply(A=@n3, B=@n7), B=Cosine(None=@n6) {period=2}))
  n9 = OneMinus(None=@n5)
  n10 = ScalarParameter() {group=Particle_Replacement, parameterName=HairThickness}
  n11 = MaterialFunctionCall(WidthBase=@n10, WidthTip=@n10) {materialFunction=&SplineThicken}
  n12 = NamedRerouteUsage()
  n13 = Multiply(A=ScalarParameter() {group=Particle_Replacement, parameterName=CameraOffset}, B=OneMinus(None=ComponentMask(None=@n12) {g=true}))
  n14 = Add(A=WorldPosition().XYZ, B=@n11.WorldPositionOffset)
  n15 = Add(
   A=Add(A=Multiply(A=@n8, B=@n9), B=@n11.WorldPositionOffset)
   B=Multiply(
    A=MaterialFunctionCall(Offset Amount=@n13, World Position=@n14) {materialFunction=/Engine/Functions/Engine_MaterialFunctions02/WorldPositionOffset/CameraOffset}.Camera Offset
    B=NamedRerouteUsage()
   )
  )
  Add(A=Multiply(A=@n15, B=NamedRerouteUsage()), B=@n2)
  Add(A=ParticleRandom(), B=Time())
  n16 = VectorParameter() {group=Distortion, parameterName=Distortion_UV_Scale, defaultValue=<rgba 1 1 0 1>}
  n17 = Multiply(A=NamedRerouteUsage(), B=ComponentMask(None=@n16.RGB) {g=true, r=true})
  n18 = ScalarParameter() {defaultValue=1, group=Distortion, parameterName=DistortionSize}
  n19 = TextureSampleParameter2D(UVs=Panner(Coordinate=Multiply(A=@n17, B=@n18), Time=Time(), Speed=AppendVector(A=@n16.B, B=@n16.B)) {bFractionalPart=true, speedX=0.52, speedY=1.75}) {
   group=Distortion
   parameterName=T_DistortionTex
   samplerType=SAMPLERTYPE_LinearGrayscale
   texture=/Game/Effects/Fort_Effects/Emotes/UltraLight/Texture/T_FluidFlameTile_Greyscale
  }
  ComponentMask(None=@n19.RGB) {g=true, r=true}
  Constant() {r=1}
  Constant() {r=1}
  Constant() {r=1}
  n20 = Add(A=WorldPosition().XYZ, B=@n15)
  n21 = Subtract(A=@n20, B=CameraPositionWS())
  n22 = Normalize(VectorInput=VectorParameter() {parameterName=ClipPlaneNrm, defaultValue=<rgba 0 -1 0 1>}.RGB)
  n23 = Subtract(
   A=MaterialFunctionCall(Plane Normal Axis=@n22, Plane Center=VectorParameter() {parameterName=ClipPlanePos, defaultValue=<rgba -512 2304 1920 1>}.RGB) {materialFunction=&VirtualPlaneCoordinates}.World Position
   B=CameraPositionWS()
  )
  n24 = Distance(A=@n21, B=@n23)
  Divide(A=@n24) {constB=100}
  n25 = ComponentMask(None=TextureCoordinate()) {g=true}
  n26 = VectorParameter() {parameterName=ColorTip, defaultValue=<rgba 0 0 0 1>}
  n27 = VectorParameter() {desc=The color closest to the strand start, parameterName=SrcColor, defaultValue=<rgba 0 0 0 1>}
  n28 = Multiply(
   A=@n25
   B=LinearInterpolate(
    A=LinearInterpolate(A=@n26.RGB, B=@n27.RGB, Alpha=@n25)
    B=VectorParameter() {group=Particle_Replacement, parameterName=Overall_Col_Mul, defaultValue=<rgba 0 0 0 1>}.RGB
    Alpha=NamedRerouteUsage()
   )
  )
  n29 = Multiply(A=ScalarParameter() {group=Mask, parameterName=Mask_Edge}, B=NamedRerouteUsage())
  n30 = ScalarParameter() {defaultValue=0.35, group=Hotspots, parameterName=Hotspots_Edge}
  n31 = ScalarParameter() {group=Mask, parameterName=Mask_Edge_W}
  n32 = ScalarParameter() {defaultValue=0.75, group=Hotspots, parameterName=Hotspots_W}
  n33 = NamedRerouteUsage()
  n34 = TextureObjectParameter() {
   parameterName=T_FireNoise
   samplerType=SAMPLERTYPE_LinearGrayscale
   texture=&T_FluidFlameTile_Greyscale
  }
  n35 = TextureSample(UVs=Panner(Coordinate=TextureCoordinate() {uTiling=0.7, vTiling=0.5}, Time=NamedRerouteUsage()) {speedY=1}, Tex=@n34) {samplerType=SAMPLERTYPE_LinearGrayscale, texture=&T_FluidFlameTile_Greyscale}
  n36 = SmoothStep(
   Min=AppendVector(A=@n29, B=@n30)
   Max=AppendVector(A=Add(A=@n29, B=@n31), B=Add(A=@n30, B=@n32))
   Value=AppendVector(A=NamedRerouteUsage(), B=Add(A=@n33, B=@n35.R))
  )
  n37 = ComponentMask(None=@n36) {g=true}
  n38 = ScalarParameter() {parameterName=Hotspot_Brightness}
  n39 = Add(A=@n28, B=Multiply(A=@n28, B=Multiply(A=@n37, B=@n38)))
  MaterialFunctionCall(Emissive=@n39) {materialFunction=&MF_FX_EmissiveAdjustments}
  Multiply(A=@n33, B=@n35.R)
  n40 = ComponentMask(None=@n36) {r=true}
  n41 = DepthFade(Opacity=@n40) {fadeDistanceDefault=8}
  n42 = Multiply(B=@n40) {constA=1}
  Multiply(A=QualitySwitch(Default=@n41, Low=@n42, High=@n41, Medium=@n42, Epic=@n41), B=NamedRerouteUsage())
  n43 = TextureCoordinate()
  Multiply(A=@n2, B=OneMinus(None=ComponentMask(None=@n43) {g=true}))
  Multiply(A=NamedRerouteUsage()) {constA=0.7, constB=0.5}
  n44 = Subtract(A=CameraPositionWS(), B=@n20)
  n45 = Normalize(VectorInput=Subtract(A=@n21, B=@n23))
  Multiply(A=Multiply(A=Normalize(VectorInput=@n44), B=@n24), B=Round(None=DotProduct(A=@n45, B=@n22)))
  n46 = Transform(None=LinearInterpolate(A=Constant3Vector() {constant=<rgba 1 0 0 0>}, B=Constant3Vector() {constant=<rgba 0 0 1 0>}, Alpha=NamedRerouteUsage())) {transformSourceType=TRANSFORMSOURCE_Local}
  n47 = Subtract(
   A=CameraPositionWS()
   B=MaterialFunctionCall() {materialFunction=&ObjectPivotPoint}.Object Pivot Location
  )
  n48 = MaterialFunctionCall(
   Input=DotProduct(A=Normalize(VectorInput=@n46), B=Normalize(VectorInput=@n47))
   Input Low=Constant()
   Input High=Constant() {r=0.25}
   Target Low=Constant()
   Target High=Constant() {r=1}
  ) {materialFunction=&RemapValueRange}
  NamedRerouteDeclaration(None=LinearInterpolate(B=Saturate(None=@n48.Result), Alpha=NamedRerouteUsage()) {constA=1}) {name=CameraOffsetWeight, nodeColor=<rgba 0.4062 1 0 1>}
  n49 = ComponentMask(None=VectorParameter() {group=Distortion, parameterName=Distortion_Str, defaultValue=<rgba 0 0 0 1>}.RGB) {g=true, r=true}
  n50 = Subtract(A=Multiply(A=@n19.R) {constB=2})
  n51 = NamedRerouteUsage()
  NamedRerouteDeclaration(None=Add(A=Multiply(A=@n49, B=@n50), B=Multiply(A=@n51))) {name=StrandUV, nodeColor=<rgba 0 0.4141 1 1>}
  NamedRerouteDeclaration(None=ScalarParameter() {group=PlayerState, parameterName=AstralHair_IsRunning}) {name=PlayerState_IsRunning, nodeColor=<rgba 1 0 0.2812 1>}
  n52 = NamedRerouteUsage()
  n53 = ComponentMask(None=@n52) {g=true}
  n54 = ComponentMask(None=@n52) {r=true}
  n55 = Subtract(B=Abs(None=Subtract(A=@n54) {constB=0.5})) {constA=0.5}
  n56 = ComponentMask(None=NamedRerouteUsage()) {g=true}
  n57 = LinearInterpolate(
   A=ScalarParameter() {defaultValue=10, group=PlayerState, parameterName=FadeStrandSource}
   B=ScalarParameter() {defaultValue=10, group=PlayerState, parameterName=FadeStrandSource_WhileRunning}
   Alpha=NamedRerouteUsage()
  )
  n58 = Saturate(None=OneMinus(None=Power(Base=@n56, Exp=@n57) {constExponent=10}))
  NamedRerouteDeclaration(None=Multiply(A=Multiply(A=@n53, B=@n55), B=VertexInterpolator(VS=@n58).PS)) {name=ErodeMask, nodeColor=<rgba 1 0.1875 0 1>}
  NamedRerouteDeclaration(None=DynamicParameter() {parameterIndex=1, paramNames [HairThickness/DragX, WaveStrength/DragY, ExecIndex/DragZ, CameraOffset / IsSkydiving]}.CameraOffset / IsSkydiving) {name=IsSkydiving, nodeColor=<rgba 0 0.5078 1 1>}
  n59 = Constant3Vector() {constant=<rgba 0 0 1 0>}
  n60 = Subtract(A=ParticlePositionWS(), B=CameraPositionWS())
  n61 = MaterialFunctionCall(
   Input=DotProduct(A=Transform(None=@n59) {transformSourceType=TRANSFORMSOURCE_Instance}, B=Normalize(VectorInput=@n60))
   Input Low=Constant() {r=0.75}
   Input High=Constant() {r=0.8}
   Target Low=Constant() {r=1}
   Target High=Constant()
  ) {materialFunction=&RemapValueRange}
  NamedRerouteDeclaration(None=VertexInterpolator(VS=Saturate(None=@n61.Result)).PS) {name=FadeParticleLookDownStrip, nodeColor=<rgba 0.4766 0 1 1>}
  n62 = NamedRerouteUsage()
  n63 = VectorParameter() {parameterName=UV_Scale, defaultValue=<rgba 1 1 0 1>}
  n64 = Multiply(A=ScalarParameter() {group=Particle_Replacement, parameterName=ScaleUV}, B=ComponentMask(None=@n63.RGB) {g=true, r=true})
  n65 = TextureSampleParameter2D(UVs=Panner(Coordinate=Multiply(A=@n62, B=@n64), Time=NamedRerouteUsage(), Speed=AppendVector(A=@n63.B, B=@n63.B)) {bFractionalPart=true, speedX=0.52, speedY=1.75}) {
   parameterName=T_FireNoise
   samplerType=SAMPLERTYPE_LinearGrayscale
   texture=&T_SimplePerlin_Greyscale
  }
  NamedRerouteDeclaration(None=@n65.B) {name=FireNoise, nodeColor=<rgba 0.9453 1 0 1>}
  n66 = ComponentMask(None=PreSkinnedPosition()) {r=true}
  n67 = ScalarParameter() {group=Skel - StrandTime, parameterName=TimeOffset_PositionalScale}
  n68 = ScalarParameter() {defaultValue=5, group=Skel - StrandTime, parameterName=TimeOffset_Slicing}
  n69 = Ceil(None=Multiply(A=Divide(A=@n66, B=@n67), B=@n68))
  n70 = Add(A=Multiply(A=Divide(A=@n69, B=@n68), B=ScalarParameter() {group=Skel - StrandTime, parameterName=TimeOffset_Scale}) {constB=1.8}, B=Time())
  NamedRerouteDeclaration(None=VertexInterpolator(VS=@n70).PS) {name=StrandTime, nodeColor=<rgba 0 1 0.7188 1>}
  NamedRerouteDeclaration(None=TextureCoordinate()) {name=StdUV, nodeColor=<rgba 0 1 0.9766 1>}
  NamedRerouteDeclaration(None=@n70) {name=StrandTime_VS, nodeColor=<rgba 0 1 0.7188 1>}
  NamedRerouteDeclaration(None=OneMinus(None=ScalarParameter() {group=PlayerState, parameterName=HasLongCapeOrLongBP})) {name=PlayerState_FadeForLongBackpackOrCape, nodeColor=<rgba 0.9688 0 1 1>}
  NamedRerouteUsage()
  NamedRerouteUsage()
  ParticleColor()
  Reroute(None=@n39)
  TextureCoordinate()
  TextureCoordinate()
  TextureCoordinate()
  TextureCoordinate()
 ]
}
