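# Translucent unlit Niagara mesh-particle material for flame strips in the Astral female character's waving hair.
# Vertex offset sums sine/cosine waves, SplineThicken width and a weighted camera offset, with dynamic parameters supplying thickness, wave strength and skydiving state.
# Distorted, per-strand-offset panning noise and smoothstep thresholds shape eroded flames and brightened hotspots over gradient or particle color.
# Opacity uses quality-switched depth fade, with a view-angle term for fading strips seen end-on.
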Material /BRCosmetics/Characters/Player/Female/Medium/Bodies/F_MED_Astral/FX/M_FX_Astral_HairWave_FireStrips
graph {
 properties {
  bUsedWithNiagaraMeshParticles=true
  blendMode=BLEND_Translucent
  pixelDepthOffsetMode=PDOM_Legacy
  shadingModel=MSM_Unlit
  usedShadingModels=Unlit
 }
 nodes [
  n1 = DynamicParameter() {paramNames [HairThickness/DragX, WaveStrength/DragY, ExecIndex/DragZ, CameraOffset / IsSkydiving]}
  n2 = Transform(None=Normalize(VectorInput=VectorParameter() {group=Waves, parameterName=Wave_Direction, defaultValue=<rgba 0 0 0 1>}.RGB)) {transformSourceType=TRANSFORMSOURCE_Instance}
  n3 = ComponentMask(None=TextureCoordinate()) {g=true}
  n4 = ParticleRandom()
  n5 = Time()
  n6 = Add(A=@n3, B=Multiply(A=Add(A=@n4, B=@n5)) {constB=1.5})
  n7 = Transform(None=Normalize(VectorInput=VectorParameter() {group=Waves, parameterName=Wave_Direction_Alt, defaultValue=<rgba 0 1 0 1>}.RGB)) {transformSourceType=TRANSFORMSOURCE_Instance}
  n8 = Add(
   A=Multiply(A=Multiply(A=@n1.WaveStrength/DragY, B=@n2), B=Sine(None=@n6))
   B=Multiply(A=Multiply(A=@n1.WaveStrength/DragY, B=@n7), B=Cosine(None=@n6) {period=2})
  )
  n9 = OneMinus(None=@n3)
  n10 = DynamicParameter() {paramNames [HairThickness/DragX, WaveStrength/DragY, ExecIndex/DragZ, CameraOffset / IsSkydiving]}
  n11 = MaterialFunctionCall(WidthBase=@n10.HairThickness/DragX, WidthTip=@n10.HairThickness/DragX) {materialFunction=&SplineThicken}
  n12 = TextureCoordinate()
  n13 = Multiply(A=@n1.CameraOffset / IsSkydiving, B=OneMinus(None=ComponentMask(None=@n12) {g=true}))
  n14 = Add(A=WorldPosition().XYZ, B=@n11.WorldPositionOffset)
  Add(
   A=Add(A=Multiply(A=@n8, B=@n9), B=@n11.WorldPositionOffset)
   B=Multiply(
    A=MaterialFunctionCall(Offset Amount=@n13, World Position=@n14) {materialFunction=/Engine/Functions/Engine_MaterialFunctions02/WorldPositionOffset/CameraOffset}.Camera Offset
    B=NamedRerouteUsage()
   )
  )
  CameraPositionWS()
  n15 = VectorParameter() {group=Distortion, parameterName=Distortion_UV_Scale, defaultValue=<rgba 1 1 0 1>}
  n16 = Multiply(A=TextureCoordinate(), B=ComponentMask(None=@n15.RGB) {g=true, r=true})
  n17 = ScalarParameter() {defaultValue=1, group=Distortion, parameterName=DistortionSize}
  n18 = TextureSampleParameter2D(UVs=Panner(Coordinate=Multiply(A=@n16, B=@n17), Time=Time(), Speed=AppendVector(A=@n15.B, B=@n15.B)) {bFractionalPart=true, speedX=0.52, speedY=1.75}) {
   group=Distortion
   parameterName=T_DistortionTex
   samplerType=SAMPLERTYPE_LinearGrayscale
   texture=/Game/Effects/Fort_Effects/Emotes/UltraLight/Texture/T_FluidFlameTile_Greyscale
  }
  ComponentMask(None=@n18.RGB) {g=true, r=true}
  Constant()
  Constant() {r=1}
  Constant() {r=1}
  Constant() {r=1}
  Constant() {r=1}
  DynamicParameter() {paramNames [HairThickness/DragX, WaveStrength/DragY, ExecIndex/DragZ, CameraOffset / IsSkydiving]}
  LinearInterpolate(B=Constant3Vector() {constant=<rgba 0 44 0 0>}, Alpha=NamedRerouteUsage())
  n19 = ComponentMask(None=TextureCoordinate()) {g=true}
  n20 = VectorParameter() {parameterName=ColorTip, defaultValue=<rgba 0 0 0 1>}
  n21 = VectorParameter() {desc=The color closest to the strand start, parameterName=SrcColor, defaultValue=<rgba 0 0 0 1>}
  n22 = Multiply(A=@n19, B=LinearInterpolate(A=LinearInterpolate(A=@n20.RGB, B=@n21.RGB, Alpha=@n19), B=ParticleColor().RGB, Alpha=NamedRerouteUsage()))
  n23 = Multiply(A=ScalarParameter() {group=Mask, parameterName=Mask_Edge}, B=NamedRerouteUsage())
  n24 = ScalarParameter() {defaultValue=0.35, group=Hotspots, parameterName=Hotspots_Edge}
  n25 = ScalarParameter() {group=Mask, parameterName=Mask_Edge_W}
  n26 = ScalarParameter() {defaultValue=0.75, group=Hotspots, parameterName=Hotspots_W}
  n27 = NamedRerouteUsage()
  n28 = TextureObjectParameter() {
   parameterName=T_FireNoise
   samplerType=SAMPLERTYPE_LinearGrayscale
   texture=&T_FluidFlameTile_Greyscale
  }
  n29 = TextureSample(UVs=Panner(Coordinate=TextureCoordinate() {uTiling=0.7, vTiling=0.5}, Time=ParticleRelativeTime()) {speedY=2}, Tex=@n28) {samplerType=SAMPLERTYPE_LinearGrayscale, texture=&T_FluidFlameTile_Greyscale}
  n30 = SmoothStep(
   Min=AppendVector(A=@n23, B=@n24)
   Max=AppendVector(A=Add(A=@n23, B=@n25), B=Add(A=@n24, B=@n26))
   Value=AppendVector(A=NamedRerouteUsage(), B=Add(A=@n27, B=@n29.R))
  )
  n31 = ComponentMask(None=@n30) {g=true}
  n32 = ScalarParameter() {parameterName=Hotspot_Brightness}
  n33 = Add(A=@n22, B=Multiply(A=@n22, B=Multiply(A=@n31, B=@n32)))
  MaterialFunctionCall(Emissive=@n33) {materialFunction=&MF_FX_EmissiveAdjustments}
  Multiply()
  Multiply(A=@n27, B=@n29.R)
  n34 = ComponentMask(None=@n30) {r=true}
  n35 = DepthFade(Opacity=@n34) {fadeDistanceDefault=8}
  n36 = Saturate(None=ComponentMask(None=TextureCoordinate() {vTiling=1.125}) {g=true})
  n37 = Multiply(A=Multiply(A=OneMinus(None=@n36)) {constB=2}, B=@n34)
  Multiply(
   A=Multiply(A=ParticleColor().A, B=QualitySwitch(Default=@n35, Low=@n37, High=@n35, Medium=@n37, Epic=@n35))
   B=LinearInterpolate(A=NamedRerouteUsage(), Alpha=NamedRerouteUsage())
  )
  n38 = Transform(None=LinearInterpolate(A=Constant3Vector() {constant=<rgba 1 0 0 0>}, B=Constant3Vector() {constant=<rgba 0 0 1 0>}, Alpha=NamedRerouteUsage())) {transformSourceType=TRANSFORMSOURCE_Local}
  n39 = Subtract(
   A=CameraPositionWS()
   B=MaterialFunctionCall() {materialFunction=&ObjectPivotPoint}.Object Pivot Location
  )
  n40 = MaterialFunctionCall(
   Input=DotProduct(A=Normalize(VectorInput=@n38), B=Normalize(VectorInput=@n39))
   Input Low=Constant()
   Input High=Constant() {r=0.25}
   Target Low=Constant()
   Target High=Constant() {r=1}
  ) {materialFunction=&RemapValueRange}
  NamedRerouteDeclaration(None=LinearInterpolate(B=Saturate(None=@n40.Result), Alpha=NamedRerouteUsage()) {constA=1}) {name=CameraOffsetWeight, nodeColor=<rgba 0.4062 1 0 1>}
  n41 = Multiply(
   A=ComponentMask(None=VectorParameter() {group=Distortion, parameterName=Distortion_Str, defaultValue=<rgba 0 0 0 1>}.RGB) {g=true, r=true}
   B=Subtract(A=Multiply(A=@n18.R) {constB=2})
  )
  n42 = TextureCoordinate()
  n43 = DynamicParameter() {paramNames [HairThickness/DragX, WaveStrength/DragY, ExecIndex/DragZ, CameraOffset / IsSkydiving]}
  n44 = Constant2Vector() {r=0.125}
  NamedRerouteDeclaration(None=Add(A=Add(A=@n41, B=@n42), B=Multiply(A=@n43.ExecIndex/DragZ, B=@n44))) {name=StrandUV, nodeColor=<rgba 0 0.4141 1 1>}
  n45 = TextureCoordinate()
  n46 = ComponentMask(None=@n45) {g=true}
  n47 = ComponentMask(None=@n45) {r=true}
  n48 = Subtract(B=Abs(None=Subtract(A=@n47) {constB=0.5})) {constA=0.5}
  n49 = ComponentMask(None=TextureCoordinate()) {g=true}
  n50 = Saturate(None=OneMinus(None=Power(Base=@n49) {constExponent=4}))
  NamedRerouteDeclaration(None=Multiply(A=Multiply(A=@n46, B=@n48), B=VertexInterpolator(VS=@n50).PS)) {name=ErodeMask, nodeColor=<rgba 1 0.1875 0 1>}
  NamedRerouteDeclaration(None=DynamicParameter() {parameterIndex=1, paramNames [HairThickness/DragX, WaveStrength/DragY, ExecIndex/DragZ, CameraOffset / IsSkydiving]}.CameraOffset / IsSkydiving) {name=IsSkydiving, nodeColor=<rgba 0 0.5078 1 1>}
  n51 = Constant3Vector() {constant=<rgba 0 0 1 0>}
  n52 = Subtract(A=ParticlePositionWS(), B=CameraPositionWS())
  n53 = MaterialFunctionCall(
   Input=DotProduct(A=Transform(None=@n51) {transformSourceType=TRANSFORMSOURCE_Instance}, B=Normalize(VectorInput=@n52))
   Input Low=Constant() {r=0.75}
   Input High=Constant() {r=0.8}
   Target Low=Constant() {r=1}
   Target High=Constant()
  ) {materialFunction=&RemapValueRange}
  NamedRerouteDeclaration(None=VertexInterpolator(VS=Saturate(None=@n53.Result)).PS) {name=FadeParticleLookDownStrip, nodeColor=<rgba 0.4766 0 1 1>}
  n54 = NamedRerouteUsage()
  n55 = VectorParameter() {parameterName=UV_Scale, defaultValue=<rgba 1 1 0 1>}
  n56 = ComponentMask(None=@n55.RGB) {g=true, r=true}
  n57 = TextureSampleParameter2D(UVs=Panner(Coordinate=Multiply(A=@n54, B=@n56), Time=ParticleRelativeTime(), Speed=AppendVector(A=@n55.B, B=@n55.B)) {bFractionalPart=true, speedX=0.52, speedY=1.75}) {
   parameterName=T_FireNoise
   samplerType=SAMPLERTYPE_LinearGrayscale
   texture=&T_SimplePerlin_Greyscale
  }
  NamedRerouteDeclaration(None=@n57.B) {name=FireNoise, nodeColor=<rgba 0.9453 1 0 1>}
  ParticleColor()
  ParticleRelativeTime()
  Reroute(None=@n33)
 ]
}
