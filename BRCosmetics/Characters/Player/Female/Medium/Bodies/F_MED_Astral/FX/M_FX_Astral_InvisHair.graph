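# Masked, unlit material for the Astral female medium player body cosmetic, named as an invisible hair material.
# The graph holds a single Constant left at its default, with no textures, parameters or visible output wiring.
# It supports skeletal meshes and mesh deformers, and turns off ray-traced shadow casting.
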Material /BRCosmetics/Characters/Player/Female/Medium/Bodies/F_MED_Astral/FX/M_FX_Astral_InvisHair
graph {
 properties {
  bCastRayTracedShadows=false
  bTangentSpaceNormal=false
  bUsedWithMeshDeformer=true
  bUsedWithSkeletalMesh=true
  blendMode=BLEND_Masked
  pixelDepthOffsetMode=PDOM_Legacy
  shadingModel=MSM_Unlit
  usedShadingModels=Unlit
 }
 nodes [Constant()]
}
