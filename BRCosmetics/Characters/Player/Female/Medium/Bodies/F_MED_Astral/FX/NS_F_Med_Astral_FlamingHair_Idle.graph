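# Niagara system for the idle flaming hair effect on the Astral female player body.
# System scripts share bone mappings, flow vector and low-version flags with the hair emitters; the opening extension emitters are disabled.
# Those emitters burst-spawn looping mesh and sprite strands on a ring or skeletal sockets, varying shape and material parameters by index.
# Strands trail a smoothed owner-relative target, blending orientation by speed or skydiving, with clip fading on low versions.
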NiagaraSystem /BRCosmetics/Characters/Player/Female/Medium/Bodies/F_MED_Astral/FX/NS_F_Med_Astral_FlamingHair_Idle
dependencies {
 usedDataInterfaces [
  &NiagaraDataInterfaceCamera
  &NiagaraDataInterfaceCurve
  &NiagaraDataInterfacePlatformSet
  &NiagaraDataInterfaceArrayFloat
  &NiagaraDataInterfaceArrayInt32
 ]
 usedDynamicInputs [
  &IsPlatformSetActive
  &Boolean_OrOperation
  &Multiply_Vector
  &NormalizeVector
 ]
 usedModules [
  &self:SystemSpawnScript.SystemScriptSource.SystemScriptGraph.NiagaraNodeAssignment_2.SetVariables_#1
  &self:SystemSpawnScript.SystemScriptSource.SystemScriptGraph.NiagaraNodeAssignment_6.SetVariables_#2
  &self:SystemSpawnScript.SystemScriptSource.SystemScriptGraph.NiagaraNodeAssignment_5.SetVariables_#3
  &self:SystemSpawnScript.SystemScriptSource.SystemScriptGraph.NiagaraNodeAssignment_9.SetVariables_#4
  &self:SystemSpawnScript.SystemScriptSource.SystemScriptGraph.NiagaraNodeAssignment_10.SetVariables_#5
  &self:SystemSpawnScript.SystemScriptSource.SystemScriptGraph.NiagaraNodeAssignment_11.SetVariables_#6
  &self:SystemSpawnScript.SystemScriptSource.SystemScriptGraph.NiagaraNodeAssignment_14.SetVariables_#7
  &self:SystemSpawnScript.SystemScriptSource.SystemScriptGraph.NiagaraNodeAssignment_15.SetVariables_#8
  &self:SystemSpawnScript.SystemScriptSource.SystemScriptGraph.NiagaraNodeAssignment_16.SetVariables_#9
  &SystemState
  &self:SystemSpawnScript.SystemScriptSource.SystemScriptGraph.NiagaraNodeAssignment_4.SetVariables_#10
  &SampleSkeletalMesh
  &self:SystemSpawnScript.SystemScriptSource.SystemScriptGraph.NiagaraNodeAssignment_0.SetVariables_#11
  &self:SystemSpawnScript.SystemScriptSource.SystemScriptGraph.NiagaraNodeAssignment_3.SetVariables_#12
  &self:SystemSpawnScript.SystemScriptSource.SystemScriptGraph.NiagaraNodeAssignment_7.SetVariables_#13
  &self:SystemSpawnScript.SystemScriptSource.SystemScriptGraph.NiagaraNodeAssignment_1.SetVariables_#14
  &self:SystemSpawnScript.SystemScriptSource.SystemScriptGraph.NiagaraNodeAssignment_8.SetVariables_#15
  &self:SystemSpawnScript.SystemScriptSource.SystemScriptGraph.NiagaraNodeAssignment_12.SetVariables_#16
  &self:SystemSpawnScript.SystemScriptSource.SystemScriptGraph.NiagaraNodeAssignment_13.SetVariables_#17
  &self:SystemSpawnScript.SystemScriptSource.SystemScriptGraph.NiagaraNodeAssignment_17.SetVariables_#18
 ]
}
emitters {
 Extensions-Hair_-ASW {
  bEnabled=false
  emitterSpawnScript [/SetVariables(Emitter.ButtClipFade=1) {isSetParameters=true, module=SetVariables_#19}]
  emitterUpdateScript [
   &EmitterState(Life Cycle Mode=::NewEnumerator1, Loop Behavior=::NewEnumerator1, Loop Duration Mode=::NewEnumerator1, UseLoopDelay=false)
   &SpawnBurst_Instantaneous(
    Spawn Count=&MakeCustomIntFromBool(
     Bool=$System.bLowVersion
     True INT=&Add_Integer(A=/Niagara/DynamicInputs/Math/GetFloatArrayCount(Float Array:NiagaraDataInterfaceArrayFloat=$System.BoneMapping_Low), B=1)
     False INT=9
    )
    Use Spawn Probability=false
    UseLoopCountLimit=false
   )
   /SetVariables(Emitter.HairTargetLocation:NiagaraPosition=/Niagara/DynamicInputs/Lerp/SmoothLerpOverTime_Vector(
     Convergence Rate:NiagaraFloat=&MakeCustomFloatFromBool(Bool=$User.IsSkydiving, True Float=25, False Float=5)
     Smooth Value:Vector3f=Linked() {linkedVariable {name=Emitter.HairTargetLocation, type=NiagaraPosition}}
     Target Value:Vector3f=&Add_Vector(
      A=Linked() {linkedVariable {name=Engine.Owner.Position, type=NiagaraPosition}}
      B=&TransformVector(
       Vector=&MakeCustomVectorFromBool(Bool=$User.IsSkydiving, True Vector=<100 -30 0>, False Vector=<100 -13 0>)
       Destination Space=::World
      )
     )
    )) {isSetParameters=true, module=SetVariables_#20}
   /SetVariables(Emitter.PreviewTargetLoc:NiagaraPosition=/Niagara/DynamicInputs/Transforms/TransformPosition(
     Position=Linked() {linkedVariable {name=Emitter.HairTargetLocation, type=NiagaraPosition}}
     Source Space=::World
     Destination Space=::Simulation
    )) {isSetParameters=true, module=SetVariables_#21}
   /SetVariables(Emitter.RenderMaskedInWater:NiagaraBool=&Boolean_AndOperation(Bool1=$User.isinwater, Bool2=&InvertBool(Boolean:NiagaraBool=$User.HasCape))) {isSetParameters=true, module=SetVariables_#22}
   /SetVariables(Emitter.OrientStrandsToBones:NiagaraBool=&SetBoolByFloatComparison(A=&VectorLength(Vector=$Engine.Owner.Velocity), B=20, Comparison Type=::NewEnumerator6)) {isSetParameters=true, module=SetVariables_#23}
   /SetVariables(Emitter.IdleBlendWeight:NiagaraFloat=/Niagara/DynamicInputs/Lerp/SmoothLerpOverTime_Float(
     Convergence Rate:NiagaraFloat=2
     Smooth Value:NiagaraFloat=$Emitter.IdleBlendWeight
     Target Value:NiagaraFloat=&MakeCustomFloatFromBool(Bool=$Emitter.OrientStrandsToBones, True Float=1, False Float=0)
    )) {isSetParameters=true, module=SetVariables_#24}
   /SetVariables(Emitter.ButtClipFade:NiagaraFloat=&MakeCustomFloatFromBool(
     Bool=$User.IsSkydiving
     True Float=&SmoothLerpOverTime_Float(
      Convergence Rate:NiagaraFloat=&MakeCustomFloatFromBool(Bool=$System.ButtClip, True Float=25, False Float=15)
      Smooth Value:NiagaraFloat=$Emitter.ButtClipFade
      Target Value:NiagaraFloat=&MakeCustomFloatFromBool(Bool=$System.ButtClip, True Float=0, False Float=1)
     )
     False Float=1
    )) {isSetParameters=true, module=SetVariables_#25}
   /SetVariables(Emitter.CameraOffsetFadeWeightFromAngle:NiagaraFloat=&DotProduct(
     Vector A=&ReturnCameraVector(Camera Property=::NewEnumerator1, Coordinate Space=::World)
     Vector B=$Engine.Owner.SystemYAxis
    )) {enabled=false, isSetParameters=true, module=SetVariables_#26}
  ]
  particleSpawnScript [
   /SetVariables(Particles.MaterialRandom:NiagaraFloat=&ReturnNormalizedExecIndex(Normalized Index Scale=6)) {isSetParameters=true, module=SetVariables_#27}
   &InitializeParticle(
    Lifetime=&Lerp_Float(
     A=0.75
     B=2
     Alpha=&FloatFromCurve(
      FloatCurve=DataInterface() {propertyValues {Curve {keys [
          {interpMode=RCIM_Cubic, tangentMode=RCTM_User, time=0, value=0}
          {interpMode=RCIM_Cubic, tangentMode=RCTM_User, time=0.5, value=1}
          {tangentMode=RCTM_User, time=1, value=0}
         ]}}}
      CurveIndex=&ReturnNormalizedExecIndex()
     )
    )
    Mesh Scale Mode=::NewEnumerator3
    Mesh Uniform Scale=&FloatFromCurve(
     FloatCurve=DataInterface() {propertyValues {Curve {keys [
         {interpMode=RCIM_Cubic, tangentMode=RCTM_User, time=0, value=0.55}
         {arriveTangent=-1.6537, interpMode=RCIM_Cubic, leaveTangent=-1.6537, tangentMode=RCTM_User, time=0.5, value=1.425}
         {tangentMode=RCTM_User, time=1, value=0.55}
        ]}}}
     CurveIndex=&ReturnNormalizedExecIndex()
     Scale Curve=0.22
    )
    Position Offset Coordinate Space=::Local
    Sprite Size Mode=::NewEnumerator3
    Uniform Sprite Size=5
    UsePositionOffset=true
    Write Material Random=true
   ) {module=InitializeParticle001}
   &ApplyInitialForces()
   /SetVariables(Particles.isRHS:NiagaraBool=&SetBoolByFloatComparison(
     A=&MakeFloatFromVector(Vector=Linked() {linkedVariable {name=Particles.Position, type=NiagaraPosition}}, Channel=::NewEnumerator2)
     B=0
     Comparison Type=::NewEnumerator6
    )) {isSetParameters=true, module=SetVariables_#28}
   &Color(Color=<rgba 3 1.0361 0.1242 1>)
   &Color(Color=&MakeCustomLinearColorFromBool(Bool=$User.bIsFrontEndPreview, True Color=<rgba 3 1.0361 0.1242 1>, False Color=<rgba 0 0 0 1>)) {enabled=false, module=Color002}
   &ShapeLocation(
    MaskSpawnGroup=false
    Non Uniform Scale=<2.5 1.75 1>
    Offset=<-2.4583 2 0>
    Offset Mode=::NewEnumerator0
    Override Seed=false
    Ring / Disc Distribution Mode:ENiagaraSphereDistributionMode=::NewEnumerator1
    Ring Radius=7.8995
    Rotation Angle Type:ENiagara_AngleInput=::NewEnumerator0
    Rotation Axis=<-0.1 1 0>
    Rotation Mode=::NewEnumerator1
    Shape Primitive=::NewEnumerator4
    U Position=&Lerp_Float(A=1.125, B=0.9, Alpha=&ReturnNormalizedExecIndex())
    UseOwnerScale=true
   )
   /SetVariables(Particles.MeshOrientation=&MultiplyQuaternion(
     Quaternion A=&OrientMeshToVector(
      Look At Direction=&TransformVector(Vector=<0 0 -1>, Source Space=::World, Destination Space=::Simulation)
      Reference Axis Vector=<1 0 0>
     )
     Quaternion B=&MakeQuatFromAxisAngle(
      Axis=<0 1 0>
      AngleInDegrees=&FloatFromCurve(
       FloatCurve=DataInterface() {propertyValues {Curve {keys [<at 0 -5>,<at 1 5>]}}}
       CurveIndex=&ReturnNormalizedExecIndex()
       Scale Curve=2
      )
     )
    )) {isSetParameters=true, module=SetVariables_#29}
  ]
  particleUpdateScript [
   &ParticleState(Kill Particles When Lifetime Has Elapsed=false, Loop Particles Lifetime=true)
   /SetVariables(Particles.VisibilityTag=&MakeCustomIntFromBool(Bool=$User.HasCape, True INT=1, False INT=0)) {isSetParameters=true, module=SetVariables_#30}
   &DynamicMaterialParameters(
    Index 0 Param 1=&Lerp_Float(
     A=4
     B=9
     Alpha=&FloatFromCurve(
      FloatCurve=DataInterface() {propertyValues {Curve {keys [
          {interpMode=RCIM_Cubic, tangentMode=RCTM_User, time=0, value=0}
          {interpMode=RCIM_Cubic, tangentMode=RCTM_User, time=0.5, value=1}
          {tangentMode=RCTM_User, time=1, value=0}
         ]}}}
      CurveIndex=&ReturnNormalizedExecIndex()
     )
    )
    Index 0 Param 2=&Multiply_Float(
     A=&Lerp_Float(
      A=5
      B=10
      Alpha=&FloatFromCurve(
       FloatCurve=DataInterface() {propertyValues {Curve {keys [
           {interpMode=RCIM_Cubic, tangentMode=RCTM_User, time=0, value=0}
           {interpMode=RCIM_Cubic, tangentMode=RCTM_User, time=0.5, value=1}
           {tangentMode=RCTM_User, time=1, value=0}
          ]}}}
       CurveIndex=&ReturnNormalizedExecIndex()
      )
     )
     B=0.65
    )
    Index 0 Param 3=&ReturnNormalizedExecIndex()
    Index 0 Param 4=&Add_Float(
     A=&Multiply_Float(
      A=&FloatFromCurve(
       FloatCurve=DataInterface() {propertyValues {Curve {keys [<at 0 0 RCIM_Cubic>,<at 0.5 9 RCIM_Cubic>,<at 1 0 RCIM_Cubic>]}}}
       CurveIndex=&ReturnNormalizedExecIndex()
       Scale Curve=&MakeCustomFloatFromBool(Bool=$User.IsSkydiving, True Float=2, False Float=1)
      )
      B=1
     )
     B=&FloatFromCurve(
      FloatCurve=DataInterface() {propertyValues {Curve {keys [
          {interpMode=RCIM_Cubic, tangentMode=RCTM_User, time=0.1, value=0}
          {interpMode=RCIM_Cubic, tangentMode=RCTM_User, time=0.5, value=1}
          {tangentMode=RCTM_User, time=0.9, value=0}
         ]}}}
      CurveIndex=&ReturnNormalizedExecIndex()
      Scale Curve=5
     )
    )
    Module.Param0WriteEnabled=true
    Module.Param1WriteEnabled=true
    Module.Param2WriteEnabled=true
    Module.Param3WriteEnabled=true
   )
   &CameraOffset(Camera Offset Amount=&FloatFromCurve(
     FloatCurve=DataInterface() {propertyValues {Curve {keys [
         {interpMode=RCIM_Cubic, tangentMode=RCTM_User, time=0.1, value=0}
         {interpMode=RCIM_Cubic, tangentMode=RCTM_User, time=0.5, value=1}
         {tangentMode=RCTM_User, time=0.9, value=0}
        ]}}}
     CurveIndex=&ReturnNormalizedExecIndex()
     Scale Curve=5
    )) {enabled=false}
   &SolveForcesAndVelocity(Clamp Velocity=false, Limit Acceleration=false)
   &ShapeLocation(
    Additional Quaternion Rotation=true
    Additional Yaw / Pitch / Roll=<0 90 90>
    MaskSpawnGroup=false
    Non Uniform Scale=<2.5 1.25 1>
    Offset Coordinate Space=::World
    Offset Mode=::NewEnumerator0
    Override Seed=false
    Ring / Disc Distribution Mode:ENiagaraSphereDistributionMode=::NewEnumerator1
    Ring Radius=7.8995
    Rotation Angle Type:ENiagara_AngleInput=::NewEnumerator0
    Rotation Mode=::NewEnumerator3
    Rotation Quaternion=&MultiplyQuaternion(Quaternion A=$System.ASW_SrcRot, Quaternion B=&MakeQuatFromAxisAngle(Axis=<0 1 0>, AngleInDegrees=90))
    Shape Origin:Vector3f=$System.ASW_SrcPos
    Shape Primitive=::NewEnumerator4
    U Position=&Lerp_Float(A=1.125, B=0.85, Alpha=&ReturnNormalizedExecIndex())
    UseOwnerScale=true
   ) {enabled=false, module=ShapeLocation001}
   /SetVariables(Particles.HairDragV:Vector3f=&Subtract_Vector(A=Linked() {linkedVariable {name=Emitter.HairTargetLocation, type=NiagaraPosition}}, B=<0 0 0>)) {enabled=false, isSetParameters=true, module=SetVariables_#31}
   /SetVariables(Particles.HairDragV:Vector3f=&MakeVector(
     X=&MakeFloatFromVector(Vector=$Particles.HairDragV, Channel=::NewEnumerator0)
     Y=/Niagara/DynamicInputs/Math/Min_Float(
      A:NiagaraFloat=&MakeFloatFromVector(Vector=$Particles.HairDragV, Channel=::NewEnumerator1)
      B:NiagaraFloat=0
     )
     Z=&MakeFloatFromVector(Vector=$Particles.HairDragV, Channel=::NewEnumerator2)
    )) {enabled=false, isSetParameters=true, module=SetVariables_#32}
   /SetVariables(Particles.HairDragV:Vector3f=&Subtract_Vector(
     A=&TransformPosition(
      Position=Linked() {linkedVariable {name=Emitter.HairTargetLocation, type=NiagaraPosition}}
      Source Space=::World
      Destination Space=::Simulation
     )
     B=<0 0 0>
    )) {isSetParameters=true}
   /SetVariables(Particles.HairDragV:Vector3f=&MakeVector(
     X=&MakeFloatFromVector(Vector=$Particles.HairDragV, Channel=::NewEnumerator0)
     Y=&Min_Float(
      A:NiagaraFloat=&MakeFloatFromVector(Vector=$Particles.HairDragV, Channel=::NewEnumerator1)
      B:NiagaraFloat=0
     )
     Z=&MakeFloatFromVector(Vector=$Particles.HairDragV, Channel=::NewEnumerator2)
    )) {isSetParameters=true}
   /SetVariables(Particles.MeshOrientation=&MultiplyQuaternion(
     Quaternion A=&OrientMeshToVector(Look At Direction=$Particles.HairDragV, Reference Axis Vector=<1 0 0>)
     Quaternion B=<0 0 0 1>
    )) {isSetParameters=true, module=SetVariables_#33}
   /SetVariables(Particles.MeshOrientation=&LerpQuaternion(
     Quaternion A:Quat4f=$Particles.MeshOrientation
     Quaternion B:Quat4f=&OrientMeshToVector(Look At Direction=$System.HairFlowVector, Reference Axis Vector=<1 0 0>)
     Lerp Factor:NiagaraFloat=$Emitter.IdleBlendWeight
     Enable Spherical Linear Interpolation:NiagaraBool=true
    )) {isSetParameters=true, module=SetVariables_#34}
   &SkeletalMeshLocation(
    Bone / Socket Index=&MakeIntFromFloat(Float=&SelectFloatFromArray(Direct Array Index=&ReturnExecIndex(), Float Selection Array=$System.BoneMapping_Low))
    MaskSpawnGroup=false
    Mesh Sampling Type=::NewEnumerator5
    Orientation Sampling=::NewEnumerator1
    Skeletal Mesh=$User.SkeletalMesh
    Socket Sampling Mode=::NewEnumerator5
    Write To Particle Position:NiagaraBool=&Boolean_AndOperation(Bool1=$System.bLowVersion, Bool2=&Boolean_NotOperation(Bool=$User.IsSkydiving))
   )
   &CalculateAccurateVelocity()
   &Color(Color=&LerpLinearColors(StartColor=<rgba 0 0 20 1>, EndColor=<rgba 0 20 0.8792 1>, LerpFactor=$Emitter.CameraOffsetFadeWeightFromAngle)) {enabled=false, module=Color001}
   &ScaleColor(
    Scale Alpha=&MakeCustomFloatFromBool(Bool=$System.bLowVersion, True Float=$Emitter.ButtClipFade, False Float=1)
    ScaleA=true
    ScaleRGB=false
   )
  ]
  rendererClasses [&NiagaraMeshRendererProperties, &NiagaraSpriteRendererProperties]
  renderers [
   {rendererClass=&NiagaraMeshRendererProperties, rendererIndex=0}
   {rendererClass=&NiagaraSpriteRendererProperties, rendererIndex=1}
   {rendererClass=&NiagaraMeshRendererProperties, rendererIndex=2}
   {rendererClass=&NiagaraSpriteRendererProperties, rendererIndex=3}
  ]
 }
 Extensions-Hairie_-_ASW_Impl {
  bEnabled=false
  emitterSpawnScript [/SetVariables(Emitter.ButtClipFade=1) {isSetParameters=true, module=SetVariables_#19}]
  emitterUpdateScript [
   &EmitterState(Life Cycle Mode=::NewEnumerator1, Loop Behavior=::NewEnumerator1, Loop Duration Mode=::NewEnumerator1, UseLoopDelay=false)
   &SpawnBurst_Instantaneous(
    Spawn Count=&MakeCustomIntFromBool(Bool=$System.bUseAlternateSkinWeights, True INT=4, False INT=7)
    Use Spawn Probability=false
    UseLoopCountLimit=false
   )
   /SetVariables(Emitter.HairTargetLocation:NiagaraPosition=&SmoothLerpOverTime_Vector(
     Convergence Rate:NiagaraFloat=&MakeCustomFloatFromBool(
      Bool=$User.IsSkydiving
      True Float=25
      False Float=&MakeCustomFloatFromBool(Bool=$System.bUseAlternateSkinWeights, True Float=10, False Float=10)
     )
     Smooth Value:Vector3f=Linked() {linkedVariable {name=Emitter.HairTargetLocation, type=NiagaraPosition}}
     Target Value:Vector3f=&Add_Vector(
      A=Linked() {linkedVariable {name=Engine.Owner.Position, type=NiagaraPosition}}
      B=&TransformVector(
       Vector=&MakeCustomVectorFromBool(Bool=$User.IsSkydiving, True Vector=<0 0 250>, False Vector=<0 -80 0>)
       Destination Space=::World
      )
     )
    )) {isSetParameters=true, module=SetVariables_#20}
   /SetVariables(Emitter.PreviewTargetLoc:NiagaraPosition=$Emitter.HairTargetLocation) {isSetParameters=true, module=SetVariables_#21}
   /SetVariables(Emitter.RenderMaskedInWater:NiagaraBool=&Boolean_AndOperation(Bool1=$User.isinwater, Bool2=&InvertBool(Boolean:NiagaraBool=$User.HasCape))) {isSetParameters=true, module=SetVariables_#22}
   /SetVariables(Emitter.OrientStrandsToBones:NiagaraBool=&SetBoolByFloatComparison(A=&VectorLength(Vector=$Engine.Owner.Velocity), B=20, Comparison Type=::NewEnumerator6)) {isSetParameters=true, module=SetVariables_#23}
   /SetVariables(Emitter.IdleBlendWeight:NiagaraFloat=&SmoothLerpOverTime_Float(
     Convergence Rate:NiagaraFloat=2
     Smooth Value:NiagaraFloat=$Emitter.IdleBlendWeight
     Target Value:NiagaraFloat=&MakeCustomFloatFromBool(Bool=$Emitter.OrientStrandsToBones, True Float=1, False Float=0)
    )) {isSetParameters=true, module=SetVariables_#24}
   /SetVariables(Emitter.ButtClipFade:NiagaraFloat=&MakeCustomFloatFromBool(
     Bool=$User.IsSkydiving
     True Float=&SmoothLerpOverTime_Float(
      Convergence Rate:NiagaraFloat=&MakeCustomFloatFromBool(Bool=$System.ButtClip, True Float=25, False Float=15)
      Smooth Value:NiagaraFloat=$Emitter.ButtClipFade
      Target Value:NiagaraFloat=&MakeCustomFloatFromBool(Bool=$System.ButtClip, True Float=0, False Float=1)
     )
     False Float=1
    )) {isSetParameters=true, module=SetVariables_#25}
   /SetVariables(Emitter.CameraOffsetFadeWeightFromAngle:NiagaraFloat=&DotProduct(
     Vector A=&ReturnCameraVector(Camera Property=::NewEnumerator1, Coordinate Space=::World)
     Vector B=$Engine.Owner.SystemYAxis
    )) {enabled=false, isSetParameters=true, module=SetVariables_#26}
  ]
  particleSpawnScript [
   /SetVariables(Particles.MaterialRandom:NiagaraFloat=&ReturnNormalizedExecIndex(Normalized Index Scale=6)) {isSetParameters=true, module=SetVariables_#27}
   &InitializeParticle(
    Lifetime=&Lerp_Float(
     A=0.75
     B=2
     Alpha=&FloatFromCurve(
      FloatCurve=DataInterface() {propertyValues {Curve {keys [
          {interpMode=RCIM_Cubic, tangentMode=RCTM_User, time=0, value=0}
          {interpMode=RCIM_Cubic, tangentMode=RCTM_User, time=0.5, value=1}
          {tangentMode=RCTM_User, time=1, value=0}
         ]}}}
      CurveIndex=&ReturnNormalizedExecIndex()
     )
    )
    Mesh Scale Mode=::NewEnumerator3
    Mesh Uniform Scale=&FloatFromCurve(
     FloatCurve=DataInterface() {propertyValues {Curve {keys [
         {interpMode=RCIM_Cubic, tangentMode=RCTM_User, time=0, value=0.55}
         {arriveTangent=-1.6537, interpMode=RCIM_Cubic, leaveTangent=-1.6537, tangentMode=RCTM_User, time=0.5, value=1.425}
         {tangentMode=RCTM_User, time=1, value=0.55}
        ]}}}
     CurveIndex=&ReturnNormalizedExecIndex()
     Scale Curve=0.22
    )
    Position Offset Coordinate Space=::Local
    Sprite Size Mode=::NewEnumerator3
    UsePositionOffset=true
    Write Material Random=true
   ) {module=InitializeParticle001}
   &ApplyInitialForces()
   /SetVariables(Particles.isRHS:NiagaraBool=&SetBoolByFloatComparison(
     A=&MakeFloatFromVector(Vector=Linked() {linkedVariable {name=Particles.Position, type=NiagaraPosition}}, Channel=::NewEnumerator2)
     B=0
     Comparison Type=::NewEnumerator6
    )) {isSetParameters=true, module=SetVariables_#28}
   &Color(Color=<rgba 3 1.0361 0.1242 1>)
   &Color(Color=&MakeCustomLinearColorFromBool(Bool=$User.bIsFrontEndPreview, True Color=<rgba 3 1.0361 0.1242 1>, False Color=<rgba 0 0 0 1>)) {enabled=false, module=Color002}
   &ShapeLocation(
    MaskSpawnGroup=false
    Non Uniform Scale=<2.5 1.75 1>
    Offset=<-2.4583 2 0>
    Offset Mode=::NewEnumerator0
    Override Seed=false
    Ring / Disc Distribution Mode:ENiagaraSphereDistributionMode=::NewEnumerator1
    Ring Radius=7.8995
    Rotation Angle Type:ENiagara_AngleInput=::NewEnumerator0
    Rotation Axis=<-0.1 1 0>
    Rotation Mode=::NewEnumerator1
    Shape Primitive=::NewEnumerator4
    U Position=&Lerp_Float(A=1.125, B=0.9, Alpha=&ReturnNormalizedExecIndex())
    UseOwnerScale=true
   ) {enabled=false}
   /SetVariables(Particles.MeshOrientation=&MultiplyQuaternion(
     Quaternion A=&OrientMeshToVector(
      Look At Direction=&TransformVector(Vector=<0 0 -1>, Source Space=::World, Destination Space=::Simulation)
      Reference Axis Vector=<1 0 0>
     )
     Quaternion B=&MakeQuatFromAxisAngle(
      Axis=<0 1 0>
      AngleInDegrees=&FloatFromCurve(
       FloatCurve=DataInterface() {propertyValues {Curve {keys [<at 0 -5>,<at 1 5>]}}}
       CurveIndex=&ReturnNormalizedExecIndex()
       Scale Curve=2
      )
     )
    )) {isSetParameters=true, module=SetVariables_#29}
   /SetVariables() {isSetParameters=true}
  ]
  particleUpdateScript [
   &ParticleState(Kill Particles When Lifetime Has Elapsed=false, Loop Particles Lifetime=true)
   /SetVariables(Particles.VisibilityTag=&MakeCustomIntFromBool(Bool=$User.HasCape, True INT=1, False INT=0)) {isSetParameters=true, module=SetVariables_#30}
   &DynamicMaterialParameters(
    Index 0 Param 1=&Lerp_Float(
     A=4
     B=9
     Alpha=&FloatFromCurve(
      FloatCurve=DataInterface() {propertyValues {Curve {keys [
          {interpMode=RCIM_Cubic, tangentMode=RCTM_User, time=0, value=0}
          {interpMode=RCIM_Cubic, tangentMode=RCTM_User, time=0.5, value=1}
          {tangentMode=RCTM_User, time=1, value=0}
         ]}}}
      CurveIndex=&ReturnNormalizedExecIndex()
     )
    )
    Index 0 Param 2=&Multiply_Float(
     A=&Lerp_Float(
      A=5
      B=10
      Alpha=&FloatFromCurve(
       FloatCurve=DataInterface() {propertyValues {Curve {keys [
           {interpMode=RCIM_Cubic, tangentMode=RCTM_User, time=0, value=0}
           {interpMode=RCIM_Cubic, tangentMode=RCTM_User, time=0.5, value=1}
           {tangentMode=RCTM_User, time=1, value=0}
          ]}}}
       CurveIndex=&ReturnNormalizedExecIndex()
      )
     )
     B=0.65
    )
    Index 0 Param 3=&ReturnNormalizedExecIndex()
    Index 0 Param 4=&Add_Float(
     A=&Multiply_Float(
      A=&FloatFromCurve(
       FloatCurve=DataInterface() {propertyValues {Curve {keys [<at 0 1 RCIM_Cubic>,<at 0.5 9 RCIM_Cubic>,<at 1 1 RCIM_Cubic>]}}}
       CurveIndex=&ReturnNormalizedExecIndex()
       Scale Curve=&MakeCustomFloatFromBool(Bool=$User.IsSkydiving, True Float=2, False Float=1)
      )
      B=1
     )
     B=&MakeCustomFloatFromBool(
      Bool=&Boolean_AndOperation(Bool1=$System.bLowVersion, Bool2=&Boolean_NotOperation(Bool=$System.bUseAlternateSkinWeights))
      True Float=15
      False Float=10
     )
    )
    Index 1 Param 4=&MakeCustomFloatFromBool(Bool=$User.IsSkydiving, True Float=1, False Float=0)
    Module.Param0WriteEnabled=true
    Module.Param0WriteEnabled1=true
    Module.Param1WriteEnabled=true
    Module.Param1WriteEnabled1=true
    Module.Param2WriteEnabled=true
    Module.Param2WriteEnabled1=true
    Module.Param3WriteEnabled=true
    Module.Param3WriteEnabled1=true
    Write Parameter Index 1=true
   )
   &SkeletalMeshLocation(
    Bone / Socket Index=&MakeCustomIntFromBool(
     Bool=$System.bUseAlternateSkinWeights
     True INT=&SelectIntFromArray(Direct Array Index=&ReturnExecIndex(), Int Selection Array=$System.ASW_Indices)
     False INT=&SelectIntFromArray(Direct Array Index=&ReturnExecIndex(), Int Selection Array=$System.HQ_SocketIndices)
    )
    "Interpolate Mesh Sampling (Update)"=true
    MaskSpawnGroup=false
    Mesh Sampling Type=::NewEnumerator5
    Orientation Sampling=::NewEnumerator1
    Skeletal Mesh=$User.SkeletalMesh
    Socket Sampling Mode=::NewEnumerator5
   ) {module=SkeletalMeshLocation001}
   &CameraOffset(Camera Offset Amount=10) {enabled=false}
   /SetVariables(Particles.HairDragV:Vector3f=&Subtract_Vector(
     A=Linked() {linkedVariable {name=Emitter.HairTargetLocation, type=NiagaraPosition}}
     B=Linked() {linkedVariable {name=Particles.Position, type=NiagaraPosition}}
    )) {isSetParameters=true, module=SetVariables_#31}
   /SetVariables(Particles.HairDragV:Vector3f=$Particles.HairDragV) {isSetParameters=true, module=SetVariables_#32}
   /SetVariables(Particles.MeshOrientation=&MultiplyQuaternion(
     Quaternion A=&OrientMeshToVector(Look At Direction=$Particles.HairDragV, Reference Axis Vector=<1 0 0>)
     Quaternion B=<0 0 0 1>
    )) {isSetParameters=true, module=SetVariables_#33}
   /SetVariables(Particles.MeshOrientation=&LerpQuaternion(
     Quaternion A:Quat4f=$Particles.MeshOrientation
     Quaternion B:Quat4f=&OrientMeshToVector(Look At Direction=$System.HairFlowVector, Reference Axis Vector=<1 0 0>)
     Lerp Factor:NiagaraFloat=&MakeCustomFloatFromBool(Bool=$User.IsSkydiving, True Float=1, False Float=0)
     Enable Spherical Linear Interpolation:NiagaraBool=true
    )) {isSetParameters=true, module=SetVariables_#34}
   /SetVariables(Particles.MeshOrientation=/Niagara/DynamicInputs/TypeConversions/MakeCustomQuaternionFromBool(
     Bool:NiagaraBool=$System.bUseAlternateSkinWeights
     True Quat:Quat4f=$Particles.SkeletalMeshLocation001.SampledRotation
     False Quat:Quat4f=$Particles.MeshOrientation
    )) {isSetParameters=true}
   &SkeletalMeshLocation(
    Bone / Socket Index=&MakeIntFromFloat(Float=&SelectFloatFromArray(Direct Array Index=&ReturnExecIndex(), Float Selection Array=$System.BoneMapping_Low))
    MaskSpawnGroup=false
    Mesh Sampling Type=::NewEnumerator5
    Orientation Sampling=::NewEnumerator1
    Skeletal Mesh=$User.SkeletalMesh
    Socket Sampling Mode=::NewEnumerator5
    Write To Particle Position:NiagaraBool=&Boolean_AndOperation(Bool1=$System.bLowVersion, Bool2=&Boolean_NotOperation(Bool=$User.IsSkydiving))
   ) {enabled=false}
   &ShapeLocation(
    Additional Quaternion Rotation=true
    Additional Yaw / Pitch / Roll=<0 90 90>
    MaskSpawnGroup=false
    Non Uniform Scale=<8 2 1>
    Offset=&MakeCustomVectorFromBool(Bool=$System.bUseAlternateSkinWeights, True Vector=<-45 0 0>, False Vector=<-40 0 5>)
    Offset Mode=::NewEnumerator0
    Override Seed=false
    Ring / Disc Distribution Mode:ENiagaraSphereDistributionMode=::NewEnumerator1
    Ring Radius=7.8995
    Rotation Angle Type:ENiagara_AngleInput=::NewEnumerator0
    Rotation Mode=::NewEnumerator3
    Rotation Quaternion=&MultiplyQuaternion(Quaternion A=$System.ASW_SrcRot, Quaternion B=&MakeQuatFromAxisAngle(Axis=<0 1 0>, AngleInDegrees=90))
    Shape Origin:Vector3f=$System.ASW_SrcPos
    Shape Primitive=::NewEnumerator4
    Transform Order:ENiagara_TransformOrder=::NewEnumerator1
    U Position=&Lerp_Float(A=1.125, B=0.9, Alpha=&ReturnNormalizedExecIndex())
    UseOwnerScale=true
   ) {enabled=false, module=ShapeLocation001}
   &SolveForcesAndVelocity(Clamp Velocity=false, Limit Acceleration=false) {enabled=false}
   &CalculateAccurateVelocity() {enabled=false}
   /SetVariables(Particles.ParticleShouldFadeForButtClip:NiagaraFloat=/Game/Effects/Niagara/DynamicInputs/Math/FN_RemapDotProduct(
     SystemAxis:Vector3f=$Engine.Owner.SystemZAxis
     ComparisonAxis:Vector3f=&RotateVectorByQuaternion(Vector To Rotate=<0 0 1>, Quaternion=$Particles.MeshOrientation)
     SourceMin:NiagaraFloat=-1
     SourceMax:NiagaraFloat=0
     DestMin:NiagaraFloat=0
     DestMax:NiagaraFloat=1
    )) {enabled=false, isSetParameters=true}
   /SetVariables(Particles.MeshOrientation=&LerpQuaternion(
     Quaternion A:Quat4f=$Particles.MeshOrientation
     Quaternion B:Quat4f=&OrientMeshToVector(
      Look At Direction=&Add_Vector(A=<0 0 1>, B=&TransformVector(Vector=<0 -1 0>, Destination Space=::Simulation))
      Reference Axis Vector=<1 0 0>
     )
     Lerp Factor:NiagaraFloat=&MakeCustomFloatFromBool(
      Bool=&Boolean_AndOperation(Bool1=$User.IsSkydiving, Bool2=&Boolean_NotOperation(Bool=$System.bUseAlternateSkinWeights))
      True Float=&RemapRange(
       Input Value To Remap=&DotProduct(Vector A=$Engine.Owner.SystemYAxis, Vector B=$System.PelvisFwdVec)
       Input Min=0
       Input Max=1
       Output Min=1
       Output Max=0
      )
      False Float=0
     )
     Enable Spherical Linear Interpolation:NiagaraBool=true
    )) {isSetParameters=true}
   &Color(Color=&LerpLinearColors(
     StartColor=<rgba 1 0 10 1>
     EndColor=<rgba 0 20 0.8792 1>
     LerpFactor=&FN_RemapDotProduct(
      SystemAxis:Vector3f=$Engine.Owner.SystemYAxis
      ComparisonAxis:Vector3f=$System.PelvisFwdVec
      SourceMin:NiagaraFloat=0
      SourceMax:NiagaraFloat=1
      DestMin:NiagaraFloat=0
      DestMax:NiagaraFloat=1
     )
    )) {enabled=false, module=Color001}
   &ScaleColor(
    Scale Alpha=&MakeCustomFloatFromBool(Bool=$System.bLowVersion, True Float=$Emitter.ButtClipFade, False Float=1)
    ScaleA=true
    ScaleRGB=false
   )
  ]
  rendererClasses [&NiagaraMeshRendererProperties, &NiagaraSpriteRendererProperties]
  renderers [
   {rendererClass=&NiagaraMeshRendererProperties, rendererIndex=0}
   {rendererClass=&NiagaraSpriteRendererProperties, rendererIndex=1}
   {rendererClass=&NiagaraMeshRendererProperties, rendererIndex=2}
   {rendererClass=&NiagaraSpriteRendererProperties, rendererIndex=3}
  ]
 }
 Extensions-Hairier {
  bEnabled=false
  emitterSpawnScript [/SetVariables(Emitter.ButtClipFade=1) {isSetParameters=true, module=SetVariables_#19}]
  emitterUpdateScript [
   &EmitterState(Life Cycle Mode=::NewEnumerator1, Loop Behavior=::NewEnumerator1, Loop Duration Mode=::NewEnumerator1, UseLoopDelay=false)
   &SpawnBurst_Instantaneous(
    Spawn Count=&MakeCustomIntFromBool(
     Bool=$System.bLowVersion
     True INT=&Add_Integer(A=&GetFloatArrayCount(Float Array:NiagaraDataInterfaceArrayFloat=$System.BoneMapping_Low), B=1)
     False INT=9
    )
    Use Spawn Probability=false
    UseLoopCountLimit=false
   )
   /SetVariables(Emitter.HairTargetLocation:NiagaraPosition=&SmoothLerpOverTime_Vector(
     Convergence Rate:NiagaraFloat=&MakeCustomFloatFromBool(Bool=$User.IsSkydiving, True Float=25, False Float=5)
     Smooth Value:Vector3f=Linked() {linkedVariable {name=Emitter.HairTargetLocation, type=NiagaraPosition}}
     Target Value:Vector3f=&Add_Vector(
      A=Linked() {linkedVariable {name=Engine.Owner.Position, type=NiagaraPosition}}
      B=&TransformVector(
       Vector=&MakeCustomVectorFromBool(Bool=$User.IsSkydiving, True Vector=<100 -30 0>, False Vector=<100 -13 0>)
       Destination Space=::World
      )
     )
    )) {isSetParameters=true, module=SetVariables_#20}
   /SetVariables(Emitter.PreviewTargetLoc:NiagaraPosition=&TransformPosition(
     Position=Linked() {linkedVariable {name=Emitter.HairTargetLocation, type=NiagaraPosition}}
     Source Space=::World
     Destination Space=::Simulation
    )) {isSetParameters=true, module=SetVariables_#21}
   /SetVariables(Emitter.RenderMaskedInWater:NiagaraBool=&Boolean_AndOperation(Bool1=$User.isinwater, Bool2=&InvertBool(Boolean:NiagaraBool=$User.HasCape))) {isSetParameters=true, module=SetVariables_#22}
   /SetVariables(Emitter.OrientStrandsToBones:NiagaraBool=&SetBoolByFloatComparison(A=&VectorLength(Vector=$Engine.Owner.Velocity), B=20, Comparison Type=::NewEnumerator6)) {isSetParameters=true, module=SetVariables_#23}
   /SetVariables(Emitter.IdleBlendWeight:NiagaraFloat=&SmoothLerpOverTime_Float(
     Convergence Rate:NiagaraFloat=2
     Smooth Value:NiagaraFloat=$Emitter.IdleBlendWeight
     Target Value:NiagaraFloat=&MakeCustomFloatFromBool(Bool=$Emitter.OrientStrandsToBones, True Float=1, False Float=0)
    )) {isSetParameters=true, module=SetVariables_#24}
   /SetVariables(Emitter.ButtClipFade:NiagaraFloat=&MakeCustomFloatFromBool(
     Bool=$User.IsSkydiving
     True Float=&SmoothLerpOverTime_Float(
      Convergence Rate:NiagaraFloat=&MakeCustomFloatFromBool(Bool=$System.ButtClip, True Float=25, False Float=15)
      Smooth Value:NiagaraFloat=$Emitter.ButtClipFade
      Target Value:NiagaraFloat=&MakeCustomFloatFromBool(Bool=$System.ButtClip, True Float=0, False Float=1)
     )
     False Float=1
    )) {isSetParameters=true, module=SetVariables_#25}
   /SetVariables(Emitter.CameraOffsetFadeWeightFromAngle:NiagaraFloat=&DotProduct(
     Vector A=&ReturnCameraVector(Camera Property=::NewEnumerator1, Coordinate Space=::World)
     Vector B=$Engine.Owner.SystemYAxis
    )) {enabled=false, isSetParameters=true, module=SetVariables_#26}
  ]
  particleSpawnScript [
   /SetVariables(Particles.MaterialRandom:NiagaraFloat=&ReturnNormalizedExecIndex(Normalized Index Scale=6)) {isSetParameters=true, module=SetVariables_#27}
   &InitializeParticle(
    Lifetime=&Lerp_Float(
     A=0.75
     B=2
     Alpha=&FloatFromCurve(
      FloatCurve=DataInterface() {propertyValues {Curve {keys [
          {interpMode=RCIM_Cubic, tangentMode=RCTM_User, time=0, value=0}
          {interpMode=RCIM_Cubic, tangentMode=RCTM_User, time=0.5, value=1}
          {tangentMode=RCTM_User, time=1, value=0}
         ]}}}
      CurveIndex=&ReturnNormalizedExecIndex()
     )
    )
    Mesh Scale Mode=::NewEnumerator3
    Mesh Uniform Scale=&FloatFromCurve(
     FloatCurve=DataInterface() {propertyValues {Curve {keys [
         {interpMode=RCIM_Cubic, tangentMode=RCTM_User, time=0, value=0.55}
         {arriveTangent=-1.6537, interpMode=RCIM_Cubic, leaveTangent=-1.6537, tangentMode=RCTM_User, time=0.5, value=1.425}
         {tangentMode=RCTM_User, time=1, value=0.55}
        ]}}}
     CurveIndex=&ReturnNormalizedExecIndex()
     Scale Curve=0.22
    )
    Position Offset Coordinate Space=::Local
    Sprite Size Mode=::NewEnumerator3
    Uniform Sprite Size=5
    UsePositionOffset=true
    Write Material Random=true
   ) {module=InitializeParticle001}
   &ApplyInitialForces()
   /SetVariables(Particles.isRHS:NiagaraBool=&SetBoolByFloatComparison(
     A=&MakeFloatFromVector(Vector=Linked() {linkedVariable {name=Particles.Position, type=NiagaraPosition}}, Channel=::NewEnumerator2)
     B=0
     Comparison Type=::NewEnumerator6
    )) {isSetParameters=true, module=SetVariables_#28}
   &Color(Color=<rgba 3 1.0361 0.1242 1>)
   &Color(Color=&MakeCustomLinearColorFromBool(Bool=$User.bIsFrontEndPreview, True Color=<rgba 3 1.0361 0.1242 1>, False Color=<rgba 0 0 0 1>)) {enabled=false, module=Color002}
   &ShapeLocation(
    MaskSpawnGroup=false
    Non Uniform Scale=<2.5 1.75 1>
    Offset=<-2.4583 2 0>
    Offset Mode=::NewEnumerator0
    Override Seed=false
    Ring / Disc Distribution Mode:ENiagaraSphereDistributionMode=::NewEnumerator1
    Ring Radius=7.8995
    Rotation Angle Type:ENiagara_AngleInput=::NewEnumerator0
    Rotation Axis=<-0.1 1 0>
    Rotation Mode=::NewEnumerator1
    Shape Primitive=::NewEnumerator4
    U Position=&Lerp_Float(A=1.125, B=0.9, Alpha=&ReturnNormalizedExecIndex())
    UseOwnerScale=true
   )
   /SetVariables(Particles.MeshOrientation=&MultiplyQuaternion(
     Quaternion A=&OrientMeshToVector(
      Look At Direction=&TransformVector(Vector=<0 0 -1>, Source Space=::World, Destination Space=::Simulation)
      Reference Axis Vector=<1 0 0>
     )
     Quaternion B=&MakeQuatFromAxisAngle(
      Axis=<0 1 0>
      AngleInDegrees=&FloatFromCurve(
       FloatCurve=DataInterface() {propertyValues {Curve {keys [<at 0 -5>,<at 1 5>]}}}
       CurveIndex=&ReturnNormalizedExecIndex()
       Scale Curve=2
      )
     )
    )) {isSetParameters=true, module=SetVariables_#29}
  ]
  particleUpdateScript [
   &ParticleState(Kill Particles When Lifetime Has Elapsed=false, Loop Particles Lifetime=true)
   /SetVariables(Particles.VisibilityTag=&MakeCustomIntFromBool(Bool=$User.HasCape, True INT=1, False INT=0)) {isSetParameters=true, module=SetVariables_#30}
   &DynamicMaterialParameters(
    Index 0 Param 1=&Lerp_Float(
     A=4
     B=9
     Alpha=&FloatFromCurve(
      FloatCurve=DataInterface() {propertyValues {Curve {keys [
          {interpMode=RCIM_Cubic, tangentMode=RCTM_User, time=0, value=0}
          {interpMode=RCIM_Cubic, tangentMode=RCTM_User, time=0.5, value=1}
          {tangentMode=RCTM_User, time=1, value=0}
         ]}}}
      CurveIndex=&ReturnNormalizedExecIndex()
     )
    )
    Index 0 Param 2=&Multiply_Float(
     A=&Lerp_Float(
      A=5
      B=10
      Alpha=&FloatFromCurve(
       FloatCurve=DataInterface() {propertyValues {Curve {keys [
           {interpMode=RCIM_Cubic, tangentMode=RCTM_User, time=0, value=0}
           {interpMode=RCIM_Cubic, tangentMode=RCTM_User, time=0.5, value=1}
           {tangentMode=RCTM_User, time=1, value=0}
          ]}}}
       CurveIndex=&ReturnNormalizedExecIndex()
      )
     )
     B=0.65
    )
    Index 0 Param 3=&ReturnNormalizedExecIndex()
    Index 0 Param 4=&Add_Float(
     A=&Multiply_Float(
      A=&FloatFromCurve(
       FloatCurve=DataInterface() {propertyValues {Curve {keys [<at 0 0 RCIM_Cubic>,<at 0.5 9 RCIM_Cubic>,<at 1 0 RCIM_Cubic>]}}}
       CurveIndex=&ReturnNormalizedExecIndex()
       Scale Curve=&MakeCustomFloatFromBool(Bool=$User.IsSkydiving, True Float=2, False Float=1)
      )
      B=1
     )
     B=&FloatFromCurve(
      FloatCurve=DataInterface() {propertyValues {Curve {keys [
          {interpMode=RCIM_Cubic, tangentMode=RCTM_User, time=0.1, value=0}
          {interpMode=RCIM_Cubic, tangentMode=RCTM_User, time=0.5, value=1}
          {tangentMode=RCTM_User, time=0.9, value=0}
         ]}}}
      CurveIndex=&ReturnNormalizedExecIndex()
      Scale Curve=5
     )
    )
    Module.Param0WriteEnabled=true
    Module.Param1WriteEnabled=true
    Module.Param2WriteEnabled=true
    Module.Param3WriteEnabled=true
   )
   &CameraOffset(Camera Offset Amount=&FloatFromCurve(
     FloatCurve=DataInterface() {propertyValues {Curve {keys [
         {interpMode=RCIM_Cubic, tangentMode=RCTM_User, time=0.1, value=0}
         {interpMode=RCIM_Cubic, tangentMode=RCTM_User, time=0.5, value=1}
         {tangentMode=RCTM_User, time=0.9, value=0}
        ]}}}
     CurveIndex=&ReturnNormalizedExecIndex()
     Scale Curve=5
    )) {enabled=false}
   &SolveForcesAndVelocity(Clamp Velocity=false, Limit Acceleration=false)
   /SetVariables(Particles.HairDragV:Vector3f=&Subtract_Vector(
     A=&TransformPosition(
      Position=Linked() {linkedVariable {name=Emitter.HairTargetLocation, type=NiagaraPosition}}
      Source Space=::World
      Destination Space=::Simulation
     )
     B=<0 0 0>
    )) {isSetParameters=true, module=SetVariables_#31}
   /SetVariables(Particles.HairDragV:Vector3f=&MakeVector(
     X=&MakeFloatFromVector(Vector=$Particles.HairDragV, Channel=::NewEnumerator0)
     Y=&Min_Float(
      A:NiagaraFloat=&MakeFloatFromVector(Vector=$Particles.HairDragV, Channel=::NewEnumerator1)
      B:NiagaraFloat=0
     )
     Z=&MakeFloatFromVector(Vector=$Particles.HairDragV, Channel=::NewEnumerator2)
    )) {isSetParameters=true, module=SetVariables_#32}
   /SetVariables(Particles.MeshOrientation=&MultiplyQuaternion(
     Quaternion A=&OrientMeshToVector(Look At Direction=$Particles.HairDragV, Reference Axis Vector=<1 0 0>)
     Quaternion B=<0 0 0 1>
    )) {isSetParameters=true, module=SetVariables_#33}
   /SetVariables(Particles.MeshOrientation=&LerpQuaternion(
     Quaternion A:Quat4f=$Particles.MeshOrientation
     Quaternion B:Quat4f=&OrientMeshToVector(
      Look At Direction=&TransformVector(Vector=$System.HairFlowVector, Source Space=::World, Destination Space=::Simulation)
      Reference Axis Vector=<1 0 0>
     )
     Lerp Factor:NiagaraFloat=$Emitter.IdleBlendWeight
     Enable Spherical Linear Interpolation:NiagaraBool=true
    )) {isSetParameters=true, module=SetVariables_#34}
   &SkeletalMeshLocation(
    Bone / Socket Index=&MakeIntFromFloat(Float=&SelectFloatFromArray(Direct Array Index=&ReturnExecIndex(), Float Selection Array=$System.BoneMapping_Low))
    MaskSpawnGroup=false
    Mesh Sampling Type=::NewEnumerator5
    Orientation Sampling=::NewEnumerator1
    Skeletal Mesh=$User.SkeletalMesh
    Socket Sampling Mode=::NewEnumerator5
    Write To Particle Position:NiagaraBool=&Boolean_AndOperation(Bool1=$System.bLowVersion, Bool2=&Boolean_NotOperation(Bool=$User.IsSkydiving))
   )
   &CalculateAccurateVelocity()
   &Color(Color=&LerpLinearColors(StartColor=<rgba 0 0 20 1>, EndColor=<rgba 0 20 0.8792 1>, LerpFactor=$Emitter.CameraOffsetFadeWeightFromAngle)) {enabled=false, module=Color001}
   &ScaleColor(
    Scale Alpha=&MakeCustomFloatFromBool(Bool=$System.bLowVersion, True Float=$Emitter.ButtClipFade, False Float=1)
    ScaleA=true
    ScaleRGB=false
   )
  ]
  rendererClasses [&NiagaraMeshRendererProperties, &NiagaraSpriteRendererProperties]
  renderers [
   {rendererClass=&NiagaraMeshRendererProperties, rendererIndex=0}
   {rendererClass=&NiagaraSpriteRendererProperties, rendererIndex=1}
   {rendererClass=&NiagaraMeshRendererProperties, rendererIndex=2}
  ]
 }
 GlowAndLight {
  emitterUpdateScript [
   &EmitterState()
   &SpawnBurst_Instantaneous(Spawn Count=1, Use Spawn Probability=false)
  ]
  particleSpawnScript [
   &InitializeParticle(
    Color=<rgba 0.8958 0.4297 0.0021 0.25>
    Lifetime=$Emitter.CurrentLoopDuration
    Position Offset=<7 0 0>
    Position Offset Coordinate Space=::Local
    Sprite Size Mode=::NewEnumerator3
    Uniform Sprite Size=75
    UsePositionOffset=true
    Write Material Random=true
   )
   /SetVariables() {isSetParameters=true}
  ]
  particleUpdateScript [
   &ParticleState()
   /SetVariables(Particles.Position=$System.ASW_SrcPos) {isSetParameters=true}
   &SkeletalMeshLocation(
    MaskSpawnGroup=false
    Mesh Sampling Type=::NewEnumerator5
    Sampled Position Offset=<-5 0 -10>
    Skeletal Mesh=$User.SkeletalMesh
    Socket Sampling Mode=::NewEnumerator5
   ) {enabled=false}
   &SolveForcesAndVelocity(Clamp Velocity=false, Limit Acceleration=false)
   &DynamicMaterialParameters(
    Index 0 Param 1=1
    Index 0 Param 2=0.5
    Module.Param0WriteEnabled=true
    Module.Param1WriteEnabled=true
    Module.Param2WriteEnabled=true
    Module.Param3WriteEnabled=true
   )
   &CameraOffset(Camera Offset Amount=10)
   &FN_Gen5_Light(Attenutation=800, Intensity=1)
   /SetVariables(Particles.LightColor=&MultiplyLinearColor(Linear Color=$Particles.Color, Scale Factor=&MakeVector4_FromFloat(Float=2))) {isSetParameters=true}
   &ScaleColor(
    Scale Alpha=0.5
    Scale RGB=&VectorFromFloat(Value=&Waveform(
      Global Phase=&Multiply_Float(A=$Emitter.Age, B=1)
      Global Clamp Min=0.5
      "Waveform [1]"=::NewEnumerator9
      "[1] Phase"=&Multiply_Float(A=$Emitter.Age, B=8)
      "[1] Frequency"=1
      "[1] Amplitude Min/Max"=<0 1>
      "[1] Scale"=2
      "[1] Bias"=0.5
     ))
    ScaleA=true
    ScaleRGB=true
   )
  ]
  rendererClasses [&NiagaraSpriteRendererProperties, &NiagaraLightRendererProperties]
  renderers [
   {rendererClass=&NiagaraSpriteRendererProperties}
   {rendererClass=&NiagaraLightRendererProperties, rendererIndex=1}
   {rendererClass=&NiagaraSpriteRendererProperties, rendererIndex=2}
  ]
 }
}
summary {
 systemName=NS_F_Med_Astral_FlamingHair_Idle
 emitters [
  {emitterName=GlowAndLight, rendererClasses [&NiagaraSpriteRendererProperties, &NiagaraLightRendererProperties]}
  {
   bEnabled=false
   emitterName=Extensions-Hairier
   rendererClasses [&NiagaraMeshRendererProperties, &NiagaraSpriteRendererProperties]
  }
  {
   bEnabled=false
   emitterName=Extensions-Hair_-ASW
   rendererClasses [&NiagaraMeshRendererProperties, &NiagaraSpriteRendererProperties]
  }
  {
   bEnabled=false
   emitterName=Extensions-Hairie_-_ASW_Impl
   rendererClasses [&NiagaraMeshRendererProperties, &NiagaraSpriteRendererProperties]
  }
 ]
 userVariables [
  User.bIsFrontEndPreview {type=NiagaraBool}
  User.HasCape {type=NiagaraBool}
  User.isinwater {type=NiagaraBool}
  User.IsSkydiving {type=NiagaraBool}
  User.IsUsingRBD {type=NiagaraBool}
  User.LOD0Check {type=NiagaraBool}
  User.SkeletalMesh {
   type=NiagaraDataInterfaceSkeletalMesh
   defaultValue {
    struct=NiagaraExt_VariableValue_DataInterface
    value {dataInterface=&self:NiagaraDataInterfaceSkeletalMesh_0}
   }
  }
 ]
}
