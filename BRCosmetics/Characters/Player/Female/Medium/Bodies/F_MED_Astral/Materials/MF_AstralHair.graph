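# Material function for the Astral character's hair that optionally overrides emissive and world position offset through the UseAstralHair switch.
# A pre-skinned top-down gradient and channel mask help index a color gradient texture, with brightness scaled per vertex by quality level and time of day.
# Panning fire noise blends flowmapped UV and pre-skinned position samples.
# World position offset adds an adjustable push along the camera direction.
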MaterialFunction /BRCosmetics/Characters/Player/Female/Medium/Bodies/F_MED_Astral/Materials/MF_AstralHair
graph {nodes [
  Constant()
  Constant() {r=1}
  n1 = ScalarParameter() {defaultValue=1, group=EmissiveTweak, parameterName=Emissive_Mul_Default}
  n2 = ScalarParameter() {defaultValue=1, group=EmissiveTweak, parameterName=Emissive_Mul_Low}
  n3 = ScalarParameter() {defaultValue=1, group=EmissiveTweak, parameterName=Emissive_Mul_Med}
  n4 = ScalarParameter() {defaultValue=1, group=TimeOfDay_Brightness_Multipliers, parameterName=TOD_Morning}
  n5 = ScalarParameter() {defaultValue=1, group=TimeOfDay_Brightness_Multipliers, parameterName=TOD_Day}
  n6 = ScalarParameter() {defaultValue=1, group=TimeOfDay_Brightness_Multipliers, parameterName=TOD_Evening}
  n7 = ScalarParameter() {defaultValue=1, group=TimeOfDay_Brightness_Multipliers, parameterName=TOD_Night}
  n8 = VertexInterpolator(VS=Multiply(
    A=QualitySwitch(Default=@n1, Low=@n2, High=@n1, Medium=@n3, Epic=@n1)
    B=MaterialFunctionCall(Value_Morning=@n4, Value_Day=@n5, Value_Evening=@n6, Value_Night=@n7) {materialFunction=&Scalar_Time_of_Day}.Result
   ))
  n9 = NamedRerouteUsage()
  n10 = Multiply(A=Multiply(A=Add(A=@n9) {constB=0.5}) {constB=0.75}, B=ScalarParameter() {parameterName=AO_Contrib})
  n11 = PreSkinnedLocalBounds()
  n12 = ComponentMask(None=@n11.Min) {b=true}
  n13 = ComponentMask(None=FunctionInput() {inputName=PreskinnedPos}) {b=true}
  n14 = MaterialFunctionCall(
   Input=Divide(A=Subtract(A=@n13, B=@n12), B=ComponentMask(None=@n11.Extents) {b=true})
   Input Low=ScalarParameter() {defaultValue=0.5, group=TopDownGrad, parameterName=TopDownGrad_MaxPos}
   Input High=ScalarParameter() {defaultValue=0.5, group=TopDownGrad, parameterName=TopDownGrad_MinPos}
   Target Low=Constant()
   Target High=Constant() {r=1}
  ) {materialFunction=&RemapValueRange}
  n15 = NamedRerouteDeclaration(None=VertexInterpolator(VS=Power(Base=@n14.Result) {constExponent=1}).PS) {name=TopDownGrad, nodeColor=<rgba 1 0 0.1641 1>}
  n16 = TextureSample(UVs=TextureCoordinate()) {samplerType=SAMPLERTYPE_Masks, texture=/BRCosmetics/Characters/Player/Female/Medium/Bodies/F_MED_Astral/Textures/T_F_MED_Astral_FaceAcc_M}
  n17 = ChannelMaskParameter(None=FunctionInput(Preview=@n16.RGB) {inputName=Gmap}) {maskChannel=Green, parameterName="GmapChanne;", defaultValue=<rgba 0 1 0 0>}
  n18 = Power(Base=Multiply(A=@n15, B=OneMinus(None=@n17.R)), Exp=ScalarParameter() {defaultValue=2, group=AO Overlay, parameterName=AO Overlay Exponent})
  n19 = NamedRerouteUsage()
  n20 = Add(A=Add(A=Multiply(A=@n10, B=@n18), B=NamedRerouteUsage()), B=@n19)
  n21 = TextureSample(UVs=@n20) {
   samplerSource=SSM_Clamp_WorldGroupSettings
   texture=/BRCosmetics/Characters/Player/Female/Medium/Bodies/F_MED_Astral/FX/T_FX_F_Med_Astral_HairGradient
  }
  n22 = Multiply(A=Add(A=NamedRerouteUsage(), B=Multiply(A=@n8.PS, B=@n21.RGB)), B=@n19)
  n23 = Subtract(A=CameraPositionWS(), B=WorldPosition().XYZ)
  n24 = Add(A=NamedRerouteUsage(), B=Multiply(A=Normalize(VectorInput=@n23), B=ScalarParameter() {defaultValue=0.5, parameterName=Camera Offset}))
  n25 = GetMaterialAttributes(None=FunctionInput() {inputType=FunctionInput_MaterialAttributes}) {attributeGetTypes [#1, #2, #3]}
  FunctionOutput(None=StaticSwitchParameter(
    True=SetMaterialAttributes!(MaterialAttributes=@n25.MaterialAttributes, EmissiveColor=@n22, WorldPositionOffset=@n24) {attributeSetTypes [#1, #2]}
    False=@n25.MaterialAttributes
   ) {parameterName=UseAstralHair})
  LinearInterpolate(A=@n20, Alpha=NamedRerouteUsage())
  n26 = VectorParameter() {parameterName=ScreenColor, defaultValue=<rgba 0 0 0 1>}
  n27 = TextureObjectParameter() {
   parameterName=T_FireNoise
   samplerType=SAMPLERTYPE_LinearGrayscale
   texture=&T_SimplePerlin_Greyscale
  }
  n28 = NamedRerouteUsage()
  n29 = Add(
   A=Multiply(A=ComponentMask(None=@n28) {g=true, r=true}, B=ScalarParameter() {defaultValue=-0.06, parameterName=Normal UV Offset})
   B=TextureCoordinate()
  )
  n30 = AppendVector(A=ScalarParameter() {defaultValue=2, parameterName=UV Scale U}, B=ScalarParameter() {defaultValue=1.5, parameterName=UV Scale V})
  n31 = MaterialFunctionCall!(
   Diffuse=@n27
   Normal Map
   "Flow Vector Map (see tooltip)"=NamedRerouteUsage()
   UV texture 2 offset
   "Time (see tooltip)"=Time()
   UVs=Panner(Coordinate=Multiply(A=@n29, B=@n30)) {bFractionalPart=true, speedY=-0.278}
  ) {materialFunction=/Engine/Functions/Engine_MaterialFunctions02/Texturing/FlowMaps_Simple}
  n32 = ComponentMask(None=PreSkinnedPosition()) {b=true, r=true}
  n33 = Panner(Coordinate=Divide(A=VertexInterpolator(VS=@n32).PS, B=ScalarParameter() {group=FireNoise, parameterName=TilingScale}) {constB=100}) {speedY=-0.128}
  n34 = NamedRerouteDeclaration(None=LinearInterpolate(A=ComponentMask(None=@n31.Diffuse) {r=true}, B=TextureSample(UVs=@n33, Tex=@n27).R, Alpha=NamedRerouteUsage())) {name=RawFireNoise, nodeColor=<rgba 0 0.5312 1 1>}
  n35 = ScalarParameter() {defaultValue=2, group=FireNoise, parameterName=Fire Noise Exponent}
  n36 = NamedRerouteDeclaration(None=Multiply(A=Power(Base=@n34, Exp=@n35), B=ScalarParameter() {defaultValue=1, group=FireNoise, parameterName=Fire_Noise_Brightness})) {name=TweakedFireNoise, nodeColor=<rgba 0 0.5312 1 1>}
  Multiply(
   A=@n15
   B=MaterialFunctionCall(
    Base=@n18
    Blend=MaterialFunctionCall!(Blend=@n26.RGB, Base=@n36) {materialFunction=/Engine/Functions/Engine_MaterialFunctions03/Blends/Blend_Screen}.Result
   ) {materialFunction=&Blend_Overlay}.Result
  )
  n37 = TextureSample(UVs=TextureCoordinate() {uTiling=1.5}) {
   samplerSource=SSM_Clamp_WorldGroupSettings
   samplerType=SAMPLERTYPE_LinearColor
   texture=/BRCosmetics/Characters/Player/Female/Medium/Bodies/F_MED_Astral/FX/T_Astral_FlowMap
  }
  n38 = Add(A=Multiply(A=@n37.RGB) {constB=2}) {constB=-1}
  n39 = ScalarParameter() {defaultValue=-0.333, parameterName=Flowmap Distance}
  n40 = Constant2Vector() {g=1, r=2}
  NamedRerouteDeclaration(None=Multiply(A=ComponentMask(None=@n38) {g=true, r=true}, B=Multiply(A=@n39, B=@n40))) {name=Flowmap, nodeColor=<rgba 0 1 0.2031 1>}
  n41 = TextureSample() {
   samplerSource=SSM_Clamp_WorldGroupSettings
   samplerType=SAMPLERTYPE_LinearColor
   texture=&T_Astral_FlowMap
  }
  NamedRerouteDeclaration(None=@n41.B) {name=PaintedMask, nodeColor=<rgba 0 0.5547 1 1>}
  NamedRerouteDeclaration(None=@n25.EmissiveColor) {name=Base_Emissive, nodeColor=<rgba 0.0547 1 0 1>}
  NamedRerouteDeclaration(None=@n25.WorldPositionOffset) {name=Base_WPO, nodeColor=<rgba 1 0 0.9844 1>}
  NamedRerouteDeclaration(None=@n25.Normal) {name=Base_Normal, nodeColor=<rgba 0 0.0859 1 1>}
  TextureSample() {samplerType=SAMPLERTYPE_Normal, texture=/BRCosmetics/Characters/Player/Female/Medium/Bodies/F_MED_Astral/Textures/T_F_MED_Astral_FaceAcc_N}
 ]}
