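# Master material for the hair of the F_MED_Astral female player body, built for skeletal meshes and mesh deformers.
# The shared character PreFX function feeds its result, Gmap and pre-skinned position into MF_AstralHair, whose result goes to the PostFX function.
# PostFX's AO, Distance Blend, Fresnel, Gmap and pre-skinned position inputs all read the PreFX AO output.
# It uses material attributes, with Default Lit and Subsurface Profile shading models.
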Material /BRCosmetics/Characters/Player/Female/Medium/Bodies/F_MED_Astral/Materials/F_MED_Astral_Hair_MASTER
graph {
 properties {
  bUseMaterialAttributes=true
  bUsedWithMeshDeformer=true
  bUsedWithSkeletalMesh=true
  pixelDepthOffsetMode=PDOM_Legacy
  shadingModel=MSM_FromMaterialExpression
  usedShadingModels=Default Lit | Subsurface Profile
 }
 nodes [
  n1 = MaterialFunctionCall() {materialFunction=&BaseFN_CharacterFunction_PreFX}
  MaterialFunctionCall(
   AO=@n1.AO
   Distance Blend=@n1.AO
   Fresnel=@n1.AO
   Gmap=@n1.AO
   Pre Skinned Local Position=@n1.AO
   Material=MaterialFunctionCall!(PreskinnedPos=@n1.Pre skinned Local Position, Gmap=@n1.Gmap, In=@n1.Result) {materialFunction=&/MF_AstralHair}.Result
  ) {materialFunction=&BaseFN_CharacterFunction_PostFX}
 ]
}
