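# Physics asset for the bangs part of the F_MED_Astral female player cosmetic, defining collision and joint limits for the hair bones.
# Capsule bodies cover the head and five dyn_bangs bones, one on the right and four on the left.
# The two "a" roots constrain to the head, while left b, c and d all branch directly from left a rather than chaining.
# Swing limits are narrow, tightest on those branches, and twist is nearly locked throughout.
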PhysicsAsset /BRCosmetics/Characters/Player/Female/Medium/Bodies/F_MED_Astral/Meshes/Parts/F_MED_Astral_FaceAcc_Bangs_Physics
bodies {
 dyn_bangs_a_l [{shapeType=Capsule, length=1.1369, radius=0.863, shapeName=dyn_bangs_a_l_capsule}]
 dyn_bangs_a_r [{shapeType=Capsule, length=2.7826, radius=1.1192, shapeName=dyn_bangs_b_l_capsule, center=<2.5002 0 0>, rotation=<rot 90 0 -0.0005>}]
 dyn_bangs_b_l [{shapeType=Capsule, length=1.9731, radius=1.1192, shapeName=dyn_bangs_b_l_capsule, center=<2.1347 0 0>, rotation=<rot 90 0 -0.0005>}]
 dyn_bangs_c_l [{shapeType=Capsule, length=2.9536, radius=1.2793, shapeName=dyn_bangs_c_l_capsule, center=<1.9928 0 0>, rotation=<rot 90 179.9999 -179.9998>}]
 dyn_bangs_d_l [{shapeType=Capsule, length=3.1137, radius=1.1192, shapeName=dyn_bangs_d_l_capsule, center=<1.8035 0 0>, rotation=<rot 90 -179.9209 -179.9209>}]
 head [{shapeType=Capsule, length=4.8671, radius=4.675, shapeName=head_capsule, center=<0.8097 3.6747 0>, rotation=<rot 90 6.7215 6.7221>}]
}
constraints [
 {bone1Name=dyn_bangs_a_l, bone2Name=head, swing1LimitDegrees=20, swing2LimitDegrees=20, twistLimitDegrees=2}
 {bone1Name=dyn_bangs_a_r, bone2Name=head, swing1LimitDegrees=15, swing2LimitDegrees=15, twistLimitDegrees=2}
 {bone1Name=dyn_bangs_b_l, bone2Name=dyn_bangs_a_l, swing1LimitDegrees=10, swing2LimitDegrees=10, twistLimitDegrees=2}
 {bone1Name=dyn_bangs_c_l, bone2Name=dyn_bangs_a_l, swing1LimitDegrees=10, swing2LimitDegrees=10, twistLimitDegrees=2}
 {bone1Name=dyn_bangs_d_l, bone2Name=dyn_bangs_a_l, swing1LimitDegrees=10, swing2LimitDegrees=10, twistLimitDegrees=2}
]
