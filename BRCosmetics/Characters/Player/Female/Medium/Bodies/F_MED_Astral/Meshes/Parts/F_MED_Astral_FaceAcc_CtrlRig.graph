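# Control Rig for the F_MED_Astral face accessory that applies corrective rotations to hair bones based on character pose.
# Spherical pose readers on head and upperarm_r write alphas, some multiplied into combined look-and-arm poses.
# These alphas weight layered rotations on hair master and chain bones; look-left, look-right and skydiving alphas drive nothing here.
# When is_front_end is true, a separate pass adds custom-idle hair offsets weighted by is_playing_custom_idle_alpha.
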ControlRigBlueprint /BRCosmetics/Characters/Player/Female/Medium/Bodies/F_MED_Astral/Meshes/Parts/F_MED_Astral_FaceAcc_CtrlRig
graphs [
 ContainedGraph {nodes [
   n1 = RigVMFunctionEntryNode()
   SequenceAggregate_1 = &RigVMFunction_Sequence(ExecuteContext=@n1.ExecuteContext)
   SequenceAggregate_1_1 = &RigVMFunction_Sequence(ExecuteContext=@SequenceAggregate_1.B)
   SequenceAggregate_1_2 = &RigVMFunction_Sequence(ExecuteContext=@SequenceAggregate_1_1.B)
   RigVMFunctionReturnNode(A=@SequenceAggregate_1.A, B=@SequenceAggregate_1_1.A, C=@SequenceAggregate_1_2.A, D=@SequenceAggregate_1_2.B)
  ]}
 RigVMFunctionLibrary {links [], nodes []}
 RigVMModel {nodes [
   RigUnit_BeginExecution = &RigUnit_BeginExecution()
   SequenceAggregate = &RigVMFunction_Sequence(ExecuteContext=@RigUnit_BeginExecution.ExecutePin)
   Branch_RigVMUnitNode = &RigVMFunction_ControlFlowBranch(ExecuteContext=@SequenceAggregate.A, Condition=RigVMVariableNode() {variable=is_front_end}.Value)
   SequenceAggregate_1 = RigVMAggregateNode(ExecuteContext=@Branch_RigVMUnitNode.True)
   n1 = RigVMVariableNode() {variable=is_playing_custom_idle_alpha}
   &RigUnit_ModifyTransforms(
    ItemToModify=((Item=(Type=Bone,Name="dyn_hair_back_base_l"),Transform=(Rotation=(X=0,Y=0,Z=0,W=1),Translation=(X=0,Y=0,Z=0),Scale3D=(X=1,Y=1,Z=1))))
    ItemToModify.0.Transform.Rotation=&RigVMFunction_MathQuaternionFromEuler(Euler=(X=25,Y=15,Z=-20)).Result
    ExecutePin=@SequenceAggregate_1.A
    Weight=@n1.Value
   )
   &RigUnit_ModifyTransforms(
    ItemToModify=((Item=(Type=Bone,Name="dfrm_hair_mstr"),Transform=(Rotation=(X=0,Y=0,Z=0,W=1),Translation=(X=0,Y=0,Z=0),Scale3D=(X=1,Y=1,Z=1))))
    ItemToModify.0.Transform.Rotation=&RigVMFunction_MathQuaternionFromEuler(Euler=(X=0,Y=7.5,Z=0)).Result
    ExecutePin=@SequenceAggregate_1.D
    Weight=@n1.Value
   )
   &RigUnit_ModifyTransforms(
    ItemToModify=((Item=(Type=Bone,Name="dyn_hair_strand_a_mstr_l"),Transform=(Rotation=(X=0,Y=0,Z=0,W=1),Translation=(X=0,Y=0,Z=0),Scale3D=(X=1,Y=1,Z=1))))
    ItemToModify.0.Transform.Rotation=&RigVMFunction_MathQuaternionFromEuler(Euler=(X=0,Y=15,Z=-5)).Result
    ExecutePin=@SequenceAggregate_1.C
    Weight=@n1.Value
   )
   &RigUnit_ModifyTransforms(
    ItemToModify=((Item=(Type=Bone,Name="dyn_hair_mid_mstr"),Transform=(Rotation=(X=0,Y=0,Z=0,W=1),Translation=(X=0,Y=0,Z=0),Scale3D=(X=1,Y=1,Z=1))))
    ItemToModify.0.Transform.Rotation=&RigVMFunction_MathQuaternionFromEuler(Euler=(X=0,Y=10,Z=-15)).Result
    ExecutePin=@SequenceAggregate_1.B
    Weight=@n1.Value
   )
   n2 = /Script/ControlRig.RigUnit_SequenceExecution(ExecuteContext=@SequenceAggregate.B) {is=struct RigVMUnitNode}
   n3 = &RigUnit_SequenceExecution(ExecuteContext=@n2.B)
   n4 = &RigUnit_SequenceExecution(ExecuteContext=@n3.A)
   n5 = RigVMVariableNode() {variable=head_SPR_look_up_alpha}
   &RigUnit_ModifyTransforms(
    ItemToModify=((Item=(Type=Bone,Name="dyn_hair_back_mstr_r"),Transform=(Rotation=(X=0,Y=0,Z=0,W=1),Translation=(X=0,Y=0,Z=0),Scale3D=(X=1,Y=1,Z=1))))
    ItemToModify.0.Transform.Rotation=&RigVMFunction_MathQuaternionFromEuler(Euler=(X=0,Y=-10,Z=-10)).Result
    ExecutePin=@n4.B
    Weight=@n5.Value
   )
   > &RigUnit_ModifyTransforms(
    ItemToModify=((Item=(Type=Bone,Name="dyn_hair_back_03_mstr_r"),Transform=(Rotation=(X=0,Y=0,Z=0,W=1),Translation=(X=0,Y=0,Z=0),Scale3D=(X=1,Y=1,Z=1))))
    ItemToModify.0.Transform.Rotation=&RigVMFunction_MathQuaternionFromEuler(Euler=(X=0,Y=-15,Z=-10)).Result
    Weight=@n5.Value
   )
   n6 = RigVMVariableNode() {variable=head_SPR_look_up_alpha}
   &RigUnit_ModifyTransforms(
    ItemToModify=((Item=(Type=Bone,Name="dyn_hair_back_01_l"),Transform=(Rotation=(X=0,Y=0,Z=0,W=1),Translation=(X=0,Y=0,Z=0),Scale3D=(X=1,Y=1,Z=1))))
    ItemToModify.0.Transform.Rotation=&RigVMFunction_MathQuaternionFromEuler(Euler=(X=-15,Y=0,Z=5)).Result
    ExecutePin=@n4.D
    Weight=@n6.Value
   )
   > &RigUnit_ModifyTransforms(
    ItemToModify=((Item=(Type=Bone,Name="dyn_hair_back_03_mstr_l"),Transform=(Rotation=(X=0,Y=0,Z=0,W=1),Translation=(X=0,Y=0,Z=0),Scale3D=(X=1,Y=1,Z=1))))
    ItemToModify.0.Transform.Rotation=&RigVMFunction_MathQuaternionFromEuler(Euler=(X=0,Y=0,Z=-25)).Result
    Weight=@n6.Value
   )
   n7 = RigVMVariableNode() {variable=head_SPR_look_up_alpha}
   &RigUnit_ModifyTransforms(
    ItemToModify=((Item=(Type=Bone,Name="dyn_hair_mid_mstr"),Transform=(Rotation=(X=0,Y=0,Z=0,W=1),Translation=(X=0,Y=0,Z=0),Scale3D=(X=1,Y=1,Z=1))))
    ItemToModify.0.Transform.Rotation=&RigVMFunction_MathQuaternionFromEuler(Euler=(X=0,Y=0,Z=-10)).Result
    ExecutePin=@n4.C
    Weight=@n7.Value
   )
   > &RigUnit_ModifyTransforms(
    ItemToModify=((Item=(Type=Bone,Name="dyn_hair_03_mstr_mid"),Transform=(Rotation=(X=0,Y=0,Z=0,W=1),Translation=(X=0,Y=0,Z=0),Scale3D=(X=1,Y=1,Z=1))))
    ItemToModify.0.Transform.Rotation=&RigVMFunction_MathQuaternionFromEuler(Euler=(X=0,Y=0,Z=-15)).Result
    Weight=@n7.Value
   )
   &RigUnit_ModifyTransforms(
    ItemToModify=((Item=(Type=Bone,Name="dfrm_hair_mstr"),Transform=(Rotation=(X=0,Y=0,Z=0,W=1),Translation=(X=0,Y=0,Z=0),Scale3D=(X=1,Y=1,Z=1))))
    ItemToModify.0.Transform.Rotation=&RigVMFunction_MathQuaternionFromEuler(Euler=(X=0,Y=0,Z=20)).Result
    Weight=RigVMVariableNode() {variable=head_SPR_look_up_alpha}.Value
    ExecutePin=@n4.A
   )
   n8 = /Script/RigVM.RigVMFunction_MathFloatMax(
    B=RigVMVariableNode() {variable=look_up_two_hands_rifle_multiplied_alpha}.Value
    A=RigVMVariableNode() {variable=look_up_and_left_two_hands_multiplied_alpha}.Value
   ) {is=struct RigVMUnitNode, template="Maximum::Execute(in A,in B,out Result)"}
   > &RigUnit_ModifyTransforms(
    ItemToModify=((Item=(Type=Bone,Name="dfrm_hair_mstr"),Transform=(Rotation=(X=0,Y=0,Z=0,W=1),Translation=(X=0,Y=0,Z=0),Scale3D=(X=1,Y=1,Z=1))))
    ItemToModify.0.Transform.Rotation=&RigVMFunction_MathQuaternionFromEuler(Euler=(X=0,Y=10,Z=0)).Result
    Weight=@n8.Result
   )
   > &RigUnit_ModifyTransforms(
    ItemToModify=((Item=(Type=Bone,Name="dyn_hair_back_mstr_r"),Transform=(Rotation=(X=0,Y=0,Z=0,W=1),Translation=(X=0,Y=0,Z=0),Scale3D=(X=1,Y=1,Z=1))))
    ItemToModify.0.Transform.Rotation=&RigVMFunction_MathQuaternionFromEuler(Euler=(X=0,Y=20,Z=-5)).Result
    Weight=@n8.Result
   )
   > &RigUnit_ModifyTransforms(
    ItemToModify=((Item=(Type=Bone,Name="dyn_hair_strand_mstr_r"),Transform=(Rotation=(X=0,Y=0,Z=0,W=1),Translation=(X=0,Y=0,Z=0),Scale3D=(X=1,Y=1,Z=1))))
    ItemToModify.0.Transform.Rotation=&RigVMFunction_MathQuaternionFromEuler(Euler=(X=0,Y=-10,Z=-20)).Result
    Weight=@n8.Result
   )
   n9 = &RigUnit_SequenceExecution(ExecuteContext=@n3.B)
   n10 = RigVMVariableNode() {variable=look_up_and_left_two_hands_multiplied_alpha}
   &RigUnit_ModifyTransforms(
    ItemToModify=((Item=(Type=Bone,Name="dyn_hair_back_mstr_r"),Transform=(Rotation=(X=0,Y=0,Z=0,W=1),Translation=(X=0,Y=0,Z=0),Scale3D=(X=1,Y=1,Z=1))))
    ItemToModify.0.Transform.Rotation=&RigVMFunction_MathQuaternionFromEuler(Euler=(X=0,Y=5,Z=-15)).Result
    ExecutePin=@n9.A
    Weight=@n10.Value
   )
   > &RigUnit_ModifyTransforms(
    ItemToModify=((Item=(Type=Bone,Name="dyn_hair_back_03_mstr_r"),Transform=(Rotation=(X=0,Y=0,Z=0,W=1),Translation=(X=0,Y=0,Z=0),Scale3D=(X=1,Y=1,Z=1))))
    ItemToModify.0.Transform.Rotation=&RigVMFunction_MathQuaternionFromEuler(Euler=(X=0,Y=20,Z=10)).Result
    Weight=@n10.Value
   )
   n11 = RigVMVariableNode() {variable=look_up_and_left_two_hands_multiplied_alpha}
   &RigUnit_ModifyTransforms(
    ItemToModify=((Item=(Type=Bone,Name="dyn_hair_mid_mstr"),Transform=(Rotation=(X=0,Y=0,Z=0,W=1),Translation=(X=0,Y=0,Z=0),Scale3D=(X=1,Y=1,Z=1))))
    ItemToModify.0.Transform.Rotation=&RigVMFunction_MathQuaternionFromEuler(Euler=(X=0,Y=20,Z=0)).Result
    Weight=@n11.Value
    ExecutePin=@n9.B
   )
   > &RigUnit_ModifyTransforms(
    ItemToModify=((Item=(Type=Bone,Name="dyn_hair_03_mstr_mid"),Transform=(Rotation=(X=0,Y=0,Z=0,W=1),Translation=(X=0,Y=0,Z=0),Scale3D=(X=1,Y=1,Z=1))))
    ItemToModify.0.Transform.Rotation=&RigVMFunction_MathQuaternionFromEuler(Euler=(X=0,Y=0,Z=10)).Result
    Weight=@n11.Value
   )
   n12 = RigVMVariableNode() {variable=look_up_and_left_two_hands_multiplied_alpha}
   &RigUnit_ModifyTransforms(
    ItemToModify=((Item=(Type=Bone,Name="dyn_hair_back_01_l"),Transform=(Rotation=(X=0,Y=0,Z=0,W=1),Translation=(X=0,Y=0,Z=0),Scale3D=(X=1,Y=1,Z=1))))
    ItemToModify.0.Transform.Rotation=&RigVMFunction_MathQuaternionFromEuler(Euler=(X=10,Y=10,Z=15)).Result
    ExecutePin=@n9.C
    Weight=@n12.Value
   )
   > &RigUnit_ModifyTransforms(
    ItemToModify=((Item=(Type=Bone,Name="dyn_hair_back_03_mstr_l"),Transform=(Rotation=(X=0,Y=0,Z=0,W=1),Translation=(X=0,Y=0,Z=0),Scale3D=(X=1,Y=1,Z=1))))
    ItemToModify.0.Transform.Rotation=&RigVMFunction_MathQuaternionFromEuler(Euler=(X=0,Y=0,Z=-5)).Result
    Weight=@n12.Value
   )
   n13 = &RigUnit_SequenceExecution(ExecuteContext=@n3.C)
   n14 = &RigUnit_SequenceExecution(ExecuteContext=@n13.A)
   n15 = RigVMVariableNode() {variable=upperarm_r_SPR_back_alpha}
   &RigUnit_ModifyTransforms(
    ItemToModify=((Item=(Type=Bone,Name="dyn_hair_back_mstr_r"),Transform=(Rotation=(X=0,Y=0,Z=0,W=1),Translation=(X=0,Y=0,Z=0),Scale3D=(X=1,Y=1,Z=1))))
    ItemToModify.0.Transform.Rotation=&RigVMFunction_MathQuaternionFromEuler(Euler=(X=0,Y=-20,Z=0)).Result
    ExecutePin=@n14.B
    Weight=@n15.Value
   )
   > &RigUnit_ModifyTransforms(
    ItemToModify=((Item=(Type=Bone,Name="dyn_hair_back_03_mstr_r"),Transform=(Rotation=(X=0,Y=0,Z=0,W=1),Translation=(X=0,Y=0,Z=0),Scale3D=(X=1,Y=1,Z=1))))
    ItemToModify.0.Transform.Rotation=&RigVMFunction_MathQuaternionFromEuler(Euler=(X=0,Y=0,Z=-5)).Result
    Weight=@n15.Value
   )
   n16 = RigVMVariableNode() {variable=upperarm_r_SPR_back_alpha}
   &RigUnit_ModifyTransforms(
    ItemToModify=((Item=(Type=Bone,Name="dyn_hair_mid_mstr"),Transform=(Rotation=(X=0,Y=0,Z=0,W=1),Translation=(X=0,Y=0,Z=0),Scale3D=(X=1,Y=1,Z=1))))
    ItemToModify.0.Transform.Rotation=&RigVMFunction_MathQuaternionFromEuler(Euler=(X=0,Y=-10,Z=-10)).Result
    Weight=@n16.Value
    ExecutePin=@n14.C
   )
   > &RigUnit_ModifyTransforms(
    ItemToModify=((Item=(Type=Bone,Name="dyn_hair_03_mstr_mid"),Transform=(Rotation=(X=0,Y=0,Z=0,W=1),Translation=(X=0,Y=0,Z=0),Scale3D=(X=1,Y=1,Z=1))))
    ItemToModify.0.Transform.Rotation=&RigVMFunction_MathQuaternionFromEuler(Euler=(X=0,Y=5,Z=0)).Result
    Weight=@n16.Value
   )
   n17 = RigVMVariableNode() {variable=upperarm_r_SPR_back_alpha}
   &RigUnit_ModifyTransforms(
    ItemToModify=((Item=(Type=Bone,Name="dyn_hair_back_01_l"),Transform=(Rotation=(X=0,Y=0,Z=0,W=1),Translation=(X=0,Y=0,Z=0),Scale3D=(X=1,Y=1,Z=1))))
    ItemToModify.0.Transform.Rotation=&RigVMFunction_MathQuaternionFromEuler(Euler=(X=0,Y=10,Z=-10)).Result
    Weight=@n17.Value
    ExecutePin=@n14.D
   )
   > &RigUnit_ModifyTransforms(
    ItemToModify=((Item=(Type=Bone,Name="dyn_hair_back_03_mstr_l"),Transform=(Rotation=(X=0,Y=0,Z=0,W=1),Translation=(X=0,Y=0,Z=0),Scale3D=(X=1,Y=1,Z=1))))
    ItemToModify.0.Transform.Rotation=&RigVMFunction_MathQuaternionFromEuler(Euler=(X=0,Y=0,Z=0)).Result
    Weight=@n17.Value
   )
   n18 = &RigUnit_SequenceExecution(ExecuteContext=@n13.B)
   n19 = RigVMVariableNode() {variable=look_down_upperarm_r_back_multiplied_alpha}
   &RigUnit_ModifyTransforms(
    ItemToModify=((Item=(Type=Bone,Name="dyn_hair_back_mstr_r"),Transform=(Rotation=(X=0,Y=0,Z=0,W=1),Translation=(X=0,Y=0,Z=0),Scale3D=(X=1,Y=1,Z=1))))
    ItemToModify.0.Transform.Rotation=&RigVMFunction_MathQuaternionFromEuler(Euler=(X=0,Y=10,Z=10)).Result
    ExecutePin=@n18.B
    Weight=@n19.Value
   )
   > &RigUnit_ModifyTransforms(
    ItemToModify=((Item=(Type=Bone,Name="dyn_hair_back_03_mstr_r"),Transform=(Rotation=(X=0,Y=0,Z=0,W=1),Translation=(X=0,Y=0,Z=0),Scale3D=(X=1,Y=1,Z=1))))
    ItemToModify.0.Transform.Rotation=&RigVMFunction_MathQuaternionFromEuler(Euler=(X=0,Y=0,Z=0)).Result
    Weight=@n19.Value
   )
   n20 = RigVMVariableNode() {variable=look_down_upperarm_r_back_multiplied_alpha}
   &RigUnit_ModifyTransforms(
    ItemToModify=((Item=(Type=Bone,Name="dyn_hair_mid_mstr"),Transform=(Rotation=(X=0,Y=0,Z=0,W=1),Translation=(X=0,Y=0,Z=0),Scale3D=(X=1,Y=1,Z=1))))
    ItemToModify.0.Transform.Rotation=&RigVMFunction_MathQuaternionFromEuler(Euler=(X=0,Y=-10,Z=10)).Result
    ExecutePin=@n18.C
    Weight=@n20.Value
   )
   > &RigUnit_ModifyTransforms(
    ItemToModify=((Item=(Type=Bone,Name="dyn_hair_03_mstr_mid"),Transform=(Rotation=(X=0,Y=0,Z=0,W=1),Translation=(X=0,Y=0,Z=0),Scale3D=(X=1,Y=1,Z=1))))
    ItemToModify.0.Transform.Rotation=&RigVMFunction_MathQuaternionFromEuler(Euler=(X=0,Y=0,Z=10)).Result
    Weight=@n20.Value
   )
   &RigUnit_ModifyTransforms(
    ItemToModify=((Item=(Type=Bone,Name="dfrm_hair_mstr"),Transform=(Rotation=(X=0,Y=0,Z=0,W=1),Translation=(X=0,Y=0,Z=0),Scale3D=(X=1,Y=1,Z=1))))
    ItemToModify.0.Transform.Rotation=&RigVMFunction_MathQuaternionFromEuler(Euler=(X=5,Y=0,Z=0)).Result
    Weight=RigVMVariableNode() {variable=look_down_upperarm_r_back_multiplied_alpha}.Value
    ExecutePin=@n18.A
   )
   n21 = RigVMVariableNode() {variable=look_down_upperarm_r_back_multiplied_alpha}
   &RigUnit_ModifyTransforms(
    ItemToModify=((Item=(Type=Bone,Name="dyn_hair_back_base_l"),Transform=(Rotation=(X=0,Y=0,Z=0,W=1),Translation=(X=0,Y=0,Z=0),Scale3D=(X=1,Y=1,Z=1))))
    ItemToModify.0.Transform.Rotation=&RigVMFunction_MathQuaternionFromEuler(Euler=(X=0,Y=-30,Z=15)).Result
    Weight=@n21.Value
    ExecutePin=@n18.D
   )
   > &RigUnit_ModifyTransforms(
    ItemToModify=((Item=(Type=Bone,Name="dyn_hair_back_01_l"),Transform=(Rotation=(X=0,Y=0,Z=0,W=1),Translation=(X=0,Y=0,Z=0),Scale3D=(X=1,Y=1,Z=1))))
    ItemToModify.0.Transform.Rotation=&RigVMFunction_MathQuaternionFromEuler(Euler=(X=0,Y=-5,Z=5)).Result
    Weight=@n21.Value
   )
   > &RigUnit_ModifyTransforms(
    ItemToModify=((Item=(Type=Bone,Name="dyn_hair_back_03_mstr_l"),Transform=(Rotation=(X=0,Y=0,Z=0,W=1),Translation=(X=0,Y=0,Z=0),Scale3D=(X=1,Y=1,Z=1))))
    ItemToModify.0.Transform.Rotation=&RigVMFunction_MathQuaternionFromEuler(Euler=(X=0,Y=0,Z=-10)).Result
    Weight=@n21.Value
   )
   n22 = &RigUnit_SequenceExecution(ExecuteContext=@n2.A)
   n23 = &RigUnit_SequenceExecution(ExecuteContext=@n22.A)
   n24 = &RigUnit_SequenceExecution(ExecuteContext=@n23.A)
   n25 = &RigUnit_SphericalPoseReader(
    ActiveRegionScaleFactors=(PositiveWidth=1,NegativeWidth=1,PositiveHeight=1,NegativeHeight=1)
    ActiveRegionSize=0.065
    Debug=(bDrawDebug=false,DebugScale=25,DebugSegments=20,DebugThickness=0.25)
    DriverAxis=(X=-1,Y=0,Z=0)
    DriverItem=(Type=Bone,Name="upperarm_r")
    FalloffRegionScaleFactors=(PositiveWidth=1,NegativeWidth=1,PositiveHeight=1,NegativeHeight=1)
    FalloffSize=0.125
    OptionalParentItem=(Type=Bone)
    RotationOffset=(X=60,Y=0,Z=65)
    ExecutePin=@n24.A
   )
   > RigVMVariableNode(Value=@n25.OutputParam) {variable=upperarm_r_SPR_back_alpha}
   n26 = &RigUnit_SphericalPoseReader(
    ActiveRegionScaleFactors=(PositiveWidth=1,NegativeWidth=1,PositiveHeight=1,NegativeHeight=1)
    ActiveRegionSize=0.075
    Debug=(bDrawDebug=false,DebugScale=25,DebugSegments=20,DebugThickness=0.25)
    DriverAxis=(X=-1,Y=0,Z=0)
    DriverItem=(Type=Bone,Name="upperarm_r")
    FalloffRegionScaleFactors=(PositiveWidth=1,NegativeWidth=1,PositiveHeight=1,NegativeHeight=1)
    FalloffSize=0.2
    OptionalParentItem=(Type=Bone)
    RotationOffset=(X=-55,Y=0,Z=-20)
    ExecutePin=@n24.D
   )
   > RigVMVariableNode(Value=@n26.OutputParam) {variable=upperarm_r_rifle_up_alpha}
   n27 = &RigUnit_SphericalPoseReader(
    ActiveRegionScaleFactors=(PositiveWidth=1,NegativeWidth=1,PositiveHeight=1,NegativeHeight=1)
    ActiveRegionSize=0.03
    Debug=(bDrawDebug=false,DebugScale=25,DebugSegments=20,DebugThickness=0.25)
    DriverAxis=(X=-1,Y=0,Z=0)
    DriverItem=(Type=Bone,Name="upperarm_r")
    FalloffRegionScaleFactors=(PositiveWidth=1,NegativeWidth=1,PositiveHeight=1,NegativeHeight=1)
    FalloffSize=0.125
    OptionalParentItem=(Type=Bone)
    RotationOffset=(X=-115,Y=0,Z=25)
    ExecutePin=@n24.B
   )
   > RigVMVariableNode(Value=@n27.OutputParam) {variable=uppearm_r_SPR_aim_up_two_hands_alpha}
   n28 = &RigUnit_SphericalPoseReader(
    ActiveRegionScaleFactors=(PositiveWidth=2,NegativeWidth=2,PositiveHeight=1,NegativeHeight=1)
    ActiveRegionSize=0.05
    Debug=(bDrawDebug=false,DebugScale=25,DebugSegments=20,DebugThickness=0.25)
    DriverAxis=(X=-1,Y=0,Z=0)
    DriverItem=(Type=Bone,Name="upperarm_r")
    FalloffRegionScaleFactors=(PositiveWidth=1,NegativeWidth=1,PositiveHeight=0.75,NegativeHeight=0.75)
    FalloffSize=0.2
    OptionalParentItem=(Type=Bone)
    RotationOffset=(X=15,Y=150,Z=0)
    ExecutePin=@n24.C
   )
   > RigVMVariableNode(Value=@n28.OutputParam) {variable=upperarm_r_skydiving}
   n29 = &RigUnit_SequenceExecution(ExecuteContext=@n23.B)
   n30 = &RigUnit_SphericalPoseReader(
    ActiveRegionScaleFactors=(PositiveWidth=1,NegativeWidth=0.5,PositiveHeight=1.5,NegativeHeight=1.5)
    ActiveRegionSize=0.1
    Debug=(bDrawDebug=false,DebugScale=25,DebugSegments=20,DebugThickness=0.25)
    DriverAxis=(X=0,Y=1,Z=0)
    DriverItem=(Type=Bone,Name="head")
    FalloffRegionScaleFactors=(PositiveWidth=1,NegativeWidth=0.5,PositiveHeight=1.5,NegativeHeight=1.5)
    FalloffSize=0.2
    FlipWidthScaling=false
    OptionalParentItem=(Type=Bone)
    RotationOffset=(X=90,Y=0,Z=-25)
    ExecutePin=@n29.A
   )
   > RigVMVariableNode(Value=@n30.OutputParam) {variable=head_SPR_look_up_alpha}
   n31 = &RigUnit_SphericalPoseReader(
    ActiveRegionScaleFactors=(PositiveWidth=2,NegativeWidth=0.5,PositiveHeight=2,NegativeHeight=2)
    ActiveRegionSize=0.075
    Debug=(bDrawDebug=false,DebugScale=25,DebugSegments=20,DebugThickness=0.25)
    DriverAxis=(X=0,Y=1,Z=0)
    DriverItem=(Type=Bone,Name="head")
    FalloffRegionScaleFactors=(PositiveWidth=1,NegativeWidth=0.4,PositiveHeight=1,NegativeHeight=1)
    FalloffSize=0.2
    FlipHeightScaling=false
    FlipWidthScaling=true
    OptionalParentItem=(Type=Bone)
    RotationOffset=(X=90,Y=0,Z=12.5)
    ExecutePin=@n29.B
   )
   > RigVMVariableNode(Value=@n31.OutputParam) {variable=head_SPR_look_down_alpha}
   n32 = &RigUnit_SphericalPoseReader(
    ActiveRegionScaleFactors=(PositiveWidth=0.75,NegativeWidth=0.75,PositiveHeight=1,NegativeHeight=0.5)
    ActiveRegionSize=0.1
    Debug=(bDrawDebug=false,DebugScale=25,DebugSegments=20,DebugThickness=0.25)
    DriverAxis=(X=0,Y=1,Z=0)
    DriverItem=(Type=Bone,Name="head")
    FalloffRegionScaleFactors=(PositiveWidth=1,NegativeWidth=1,PositiveHeight=1,NegativeHeight=0.5)
    FalloffSize=0.2
    OptionalParentItem=(Type=Bone)
    RotationOffset=(X=120,Y=0,Z=0)
    ExecutePin=@n29.C
   )
   > RigVMVariableNode(Value=@n32.OutputParam) {variable=head_SPR_look_left_alpha}
   n33 = &RigUnit_SphericalPoseReader(
    ActiveRegionScaleFactors=(PositiveWidth=0.75,NegativeWidth=0.75,PositiveHeight=1,NegativeHeight=0.5)
    ActiveRegionSize=0.1
    Debug=(bDrawDebug=false,DebugScale=25,DebugSegments=20,DebugThickness=0.25)
    DriverAxis=(X=0,Y=1,Z=0)
    DriverItem=(Type=Bone,Name="head")
    FalloffRegionScaleFactors=(PositiveWidth=1,NegativeWidth=1,PositiveHeight=1,NegativeHeight=0.5)
    FalloffSize=0.2
    FlipHeightScaling=true
    FlipWidthScaling=false
    OptionalParentItem=(Type=Bone)
    RotationOffset=(X=60,Y=0,Z=0)
    ExecutePin=@n29.D
   )
   > RigVMVariableNode(Value=@n33.OutputParam) {variable=head_SPR_look_right_alpha}
   RigVMVariableNode(
    Value=&RigVMFunction_MathFloatMul(B=RigVMVariableNode() {variable=uppearm_r_SPR_aim_up_two_hands_alpha}.Value, A=RigVMVariableNode() {variable=head_SPR_look_up_alpha}.Value) {template="Multiply::Execute(in A,in B,out Result)"}.Result
    ExecuteContext=@n22.B
   ) {variable=look_up_and_left_two_hands_multiplied_alpha}
   RigVMVariableNode(
    Value=&RigVMFunction_MathFloatMul(A=RigVMVariableNode() {variable=head_SPR_look_up_alpha}.Value, B=RigVMVariableNode() {variable=upperarm_r_rifle_up_alpha}.Value) {template="Multiply::Execute(in A,in B,out Result)"}.Result
    ExecuteContext=@n22.C
   ) {variable=look_up_two_hands_rifle_multiplied_alpha}
   RigVMVariableNode(
    Value=&RigVMFunction_MathFloatMul(B=RigVMVariableNode() {variable=head_SPR_look_down_alpha}.Value, A=RigVMVariableNode() {variable=upperarm_r_SPR_back_alpha}.Value) {template="Multiply::Execute(in A,in B,out Result)"}.Result
    ExecuteContext=@n22.D
   ) {variable=look_down_upperarm_r_back_multiplied_alpha}
  ]}
]
hierarchy [
 root {loc=<0 0 0>, rot=<0 0 0 1>}
 pelvis {parent=root, loc=<0 -0.5301 98.0571>, rot=<0.0127 0.707 -0.0127 -0.707>}
 spine_01 {parent=pelvis, loc=<6.1208 -0.9328 0>, rot=<0 0 0.034 -0.9994>}
 spine_02 {parent=spine_01, loc=<6.5996 -0.2095 0>, rot=<0 0 0.0238 -0.9997>}
 spine_03 {parent=spine_02, loc=<6.4914 -0.1237 0>, rot=<0 0 0.0149 -0.9999>}
 spine_04 {parent=spine_03, loc=<6.9598 0.0794 0>, rot=<0 0 -0.0132 -0.9999>}
 spine_05 {parent=spine_04, loc=<6.9987 0 0>, rot=<0 0 -0.0015 -1>}
 clavicle_l {parent=spine_05, loc=<5.4831 -0.1053 -3.2946>, rot=<0.0465 -0.7289 -0.009 -0.6829>}
 upperarm_l {parent=clavicle_l, loc=<9.8658 -1.2136 -0.4105>, rot=<0.0164 -0.3245 -0.0269 -0.9454>}
 neck_01 {parent=spine_05, loc=<11.3138 0.1529 0>, rot=<0 0 -0.2003 -0.9797>}
 neck_02 {parent=neck_01, loc=<4.5627 0 0>, rot=<0 0 0.0504 -0.9987>}
 head {parent=neck_02, loc=<5.3914 0 0>, rot=<0 0 0.115 -0.9934>}
 dyn_bangs_a_r {parent=head, loc=<2.9665 6.8864 6.2676>, rot=<0.393 -0.5118 0.7638 -0.0152>}
 dyn_bangs_a_l {parent=head, loc=<5.3231 4.4529 -6.9191>, rot=<-0.0023 0.5483 0.8363 -0.0035>}
 dyn_bangs_b_l {parent=dyn_bangs_a_l, loc=<2.0508 -0.8156 1.9759>, rot=<-0.1941 -0.3778 0.4507 -0.7852>}
 dyn_bangs_c_l {parent=dyn_bangs_a_l, loc=<3.3621 0.1456 0.0484>, rot=<-0.0905 0.0144 -0.1192 -0.9886>}
 dyn_bangs_d_l {parent=dyn_bangs_a_l, loc=<4.2738 1.3134 -1.7313>, rot=<-0.25 -0.001 -0.0257 -0.9679>}
 dfrm_hair_mstr {parent=head, loc=<9.1012 -4.0009 0>, rot=<0 0 0 1>}
 dyn_hair_back_mstr_r {parent=dfrm_hair_mstr, loc=<0.047 0 5.564>, rot=<0.1529 -0.8383 0.523 -0.0176>}
 dyn_hair_back_01_r {parent=dyn_hair_back_mstr_r, loc=<0 0 0>, rot=<0 0 0 1>}
 dyn_hair_back_02_r {parent=dyn_hair_back_01_r, loc=<12.1504 0 0>, rot=<0.1266 0.0514 0.0733 -0.9879>}
 dyn_hair_back_strand_01_r {parent=dyn_hair_back_02_r, loc=<7.8276 -3.627 -5.1362>, rot=<-0.1607 0.2418 0.0391 -0.9561>}
 dyn_hair_back_03_mstr_r {parent=dyn_hair_back_02_r, loc=<12.151 0 0>, rot=<0.1049 -0.0068 -0.1403 -0.9845>}
 dyn_hair_back_03_r {parent=dyn_hair_back_03_mstr_r, loc=<0 0 0>, rot=<0 0 0 1>}
 dyn_hair_back_04_r {parent=dyn_hair_back_03_r, loc=<12.1512 0 0>, rot=<0.0622 0.0181 -0.3088 -0.9489>}
 dyn_hair_back_05_r {parent=dyn_hair_back_04_r, loc=<12.1518 0 0>, rot=<0.0257 0.0111 0.0592 -0.9979>}
 dyn_hair_strand_mstr_r {parent=dyn_hair_back_mstr_r, loc=<7.1607 4.0764 -8.7623>, rot=<-0.193 0.0264 0.0831 -0.9773>}
 dyn_hair_strand_01_r {parent=dyn_hair_strand_mstr_r, loc=<0 0 0>, rot=<0 0 0 1>}
 dyn_hair_strand_02_r {parent=dyn_hair_strand_01_r, loc=<7.3389 0 0>, rot=<0.137 0.1746 0.2377 -0.9456>}
 dyn_hair_strand_03_r {parent=dyn_hair_strand_02_r, loc=<7.918 0 0>, rot=<0.1921 0.0281 -0.0509 -0.9797>}
 dyn_hair_strand_04_r {parent=dyn_hair_strand_03_r, loc=<5.9738 0 0>, rot=<0.0794 -0.0747 -0.2168 -0.9701>}
 dyn_hair_strand_05_r {parent=dyn_hair_strand_04_r, loc=<7.8697 0 0>, rot=<0.1339 0.139 -0.1984 -0.9609>}
 dyn_hair_strand_06_r {parent=dyn_hair_strand_05_r, loc=<7.8697 0 0>, rot=<0.0143 0.3339 -0.0324 -0.942>}
 dyn_hair_strand_07_r {parent=dyn_hair_strand_06_r, loc=<7.8697 0 0>, rot=<0.0633 0.307 -0.014 -0.9495>}
 dyn_hair_strand_08_r {parent=dyn_hair_strand_07_r, loc=<7.87 0 0>, rot=<0 0 0 1>}
 dyn_hair_mid_mstr {parent=dfrm_hair_mstr, loc=<0.4694 -1.4988 0>, rot=<0.1309 -0.9914 0 0>}
 dyn_hair_01_mid {parent=dyn_hair_mid_mstr, loc=<0 0 0>, rot=<0 0 0 1>}
 dyn_hair_02_mid {parent=dyn_hair_01_mid, loc=<12.1511 0 0>, rot=<0.0049 0.001 0.0979 -0.9952>}
 dyn_hair_03_mstr_mid {parent=dyn_hair_02_mid, loc=<12.1511 0 0>, rot=<0 0 -0.1243 -0.9922>}
 dyn_hair_03_mid {parent=dyn_hair_03_mstr_mid, loc=<0 0 0>, rot=<0 0 0 1>}
 dyn_hair_04_mid {parent=dyn_hair_03_mid, loc=<12.1511 0 0>, rot=<0 0 -0.2275 -0.9738>}
 dyn_hair_05_mid {parent=dyn_hair_04_mid, loc=<12.1511 0 0>, rot=<0.0007 -0.0127 0.0589 -0.9982>}
 dyn_hair_06_mid {parent=dyn_hair_05_mid, loc=<12.151 0 0>, rot=<0 0 0 1>}
 dyn_hair_back_base_l {parent=dfrm_hair_mstr, loc=<0.0475 0 -5.564>, rot=<0.1529 -0.8383 -0.523 0.0175>}
 dyn_hair_back_mstr_l {parent=dyn_hair_back_base_l, loc=<0 0 0>, rot=<0 0 0 1>}
 dyn_hair_back_01_l {parent=dyn_hair_back_mstr_l, loc=<0 0 0>, rot=<0 0 0 1>}
 dyn_hair_back_02_l {parent=dyn_hair_back_01_l, loc=<12.1511 0 0>, rot=<-0.1266 -0.0514 0.0733 -0.9879>}
 dyn_hair_back_03_mstr_l {parent=dyn_hair_back_02_l, loc=<12.1511 0 0>, rot=<-0.1049 0.0068 -0.1403 -0.9845>}
 dyn_hair_back_03_l {parent=dyn_hair_back_03_mstr_l, loc=<0 0 0>, rot=<0 0 0 1>}
 dyn_hair_back_04_l {parent=dyn_hair_back_03_l, loc=<12.1511 0 0>, rot=<-0.0622 -0.0181 -0.3088 -0.9489>}
 dyn_hair_back_05_l {parent=dyn_hair_back_04_l, loc=<12.1511 0 0>, rot=<-0.0257 -0.0111 0.0592 -0.9979>}
 dyn_hair_strand_a_mstr_l {parent=dyn_hair_back_base_l, loc=<0 0 0>, rot=<-0.0951 0.116 0.0702 -0.9862>}
 dyn_hair_strand_a_01_l {parent=dyn_hair_strand_a_mstr_l, loc=<0 0 0>, rot=<0 0 0 1>}
 dyn_hair_strand_a_02_l {parent=dyn_hair_strand_a_01_l, loc=<13.0169 0 0>, rot=<0.1629 -0.0085 0.167 -0.9724>}
 dyn_hair_strand_a_03_l {parent=dyn_hair_strand_a_02_l, loc=<10.1098 0.1602 -0.2276>, rot=<-0.203 -0.1578 -0.1254 -0.9582>}
 dyn_hair_strand_a_04_l {parent=dyn_hair_strand_a_03_l, loc=<9.9166 0.5444 -0.281>, rot=<0.0167 -0.1725 -0.326 -0.9294>}
 dyn_hair_strand_a_05_l {parent=dyn_hair_strand_a_04_l, loc=<10.1217 0.19 0.1433>, rot=<-0.0286 0.2847 -0.0371 -0.9575>}
 dyn_hair_strand_a_06_l {parent=dyn_hair_strand_a_05_l, loc=<10.122 0 0>, rot=<0 0 0 1>}
 clavicle_r {parent=spine_05, loc=<5.4831 -0.1053 3.2946>, rot=<0.7289 0.0465 -0.6829 0.009>}
 upperarm_r {parent=clavicle_r, loc=<-9.8658 1.2136 0.4105>, rot=<0.0164 -0.3245 -0.0269 -0.9454>}
 upperarm_twist_01_r {parent=upperarm_r, loc=<-7.2135 0.1134 0>, rot=<0 0 0 -1>}
 deform_upperarm_bk_base_r {parent=upperarm_twist_01_r, loc=<6.4053 4.3746 -0.314>, rot=<0 0 0 -1>}
 deform_upperarm_bk_r {parent=deform_upperarm_bk_base_r, loc=<0 0 0>, rot=<0 0 0 1>}
 deform_upperarm_fr_base_r {parent=upperarm_twist_01_r, loc=<5.6312 -5.092 -0.3354>, rot=<0 0 0 -1>}
 deform_upperarm_fr_r {parent=deform_upperarm_fr_base_r, loc=<0 0 0>, rot=<0 0 0 1>}
 deform_upperarm_up_base_r {parent=upperarm_twist_01_r, loc=<8.891 -0.4825 -3.9517>, rot=<0 0 0 -1>}
 deform_upperarm_up_r {parent=deform_upperarm_up_base_r, loc=<0 0 0>, rot=<0 0 0 1>}
 deform_pec_aim_base_r {parent=spine_04, loc=<3.2452 9.3309 1.2237>, rot=<-0.4033 -0.0603 0.8507 0.3318>}
 deform_pec_aim_r {parent=deform_pec_aim_base_r, loc=<0 0 0>, rot=<0 0 0 1>}
 deform_pec_base_r {parent=deform_pec_aim_r, loc=<-9.3667 -0.0003 -0.0004>, rot=<0 0 0 1>}
]
variables [
 clavicle_r_SPR_back_alpha {type=double, default=0}
 head_SPR_look_down_alpha {type=double, default=0}
 head_SPR_look_left_alpha {type=double, default=0}
 head_SPR_look_right_alpha {type=double, default=0}
 head_SPR_look_up_alpha {type=double, default=0}
 is_front_end {type=bool, default=False}
 is_playing_custom_idle_alpha {type=double, default=0}
 look_down_upperarm_r_back_multiplied_alpha {type=double, default=0}
 look_up_and_left_two_hands_multiplied_alpha {type=double, default=0}
 look_up_two_hands_rifle_multiplied_alpha {type=double, default=0}
 uppearm_r_SPR_aim_up_two_hands_alpha {type=double, default=0}
 upperarm_r_SPR_back_alpha {type=double, default=0}
 upperarm_r_rifle_up_alpha {type=double, default=0}
 upperarm_r_skydiving {type=double, default=0}
]
