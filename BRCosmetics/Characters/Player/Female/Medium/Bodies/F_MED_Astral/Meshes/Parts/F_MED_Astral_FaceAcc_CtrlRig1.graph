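# Control Rig for the Astral female character's FaceAcc part, procedurally posing the long back hair against the body.
# Setup caches the middle-chain controls, whose nulls follow head, neck and upper spine in translation to feed a B-spline IK and twist pass.
# Left and right chains blend between head, spine, clavicles and the middle chain.
# Spherical pose readers on upper arms and clavicles weight corrective constraints and transform offsets as those bones move.
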ControlRigBlueprint /BRCosmetics/Characters/Player/Female/Medium/Bodies/F_MED_Astral/Meshes/Parts/F_MED_Astral_FaceAcc_CtrlRig1
graphs [
 ContainedGraph {nodes [
   n1 = RigVMFunctionEntryNode()
   SequenceAggregate = &RigVMFunction_Sequence(ExecuteContext=@n1.ExecuteContext)
   SequenceAggregate_1 = &RigVMFunction_Sequence(ExecuteContext=@SequenceAggregate.B)
   SequenceAggregate_2 = &RigVMFunction_Sequence(ExecuteContext=@SequenceAggregate_1.B)
   RigVMFunctionReturnNode(A=@SequenceAggregate.A, B=@SequenceAggregate_1.A, C=@SequenceAggregate_2.A, D=@SequenceAggregate_2.B)
  ]}
 ContainedGraph {nodes [
   n1 = RigVMFunctionEntryNode()
   SequenceAggregate = &RigVMFunction_Sequence(ExecuteContext=@n1.ExecuteContext)
   SequenceAggregate_1 = &RigVMFunction_Sequence(ExecuteContext=@SequenceAggregate.B)
   SequenceAggregate_2 = &RigVMFunction_Sequence(ExecuteContext=@SequenceAggregate_1.B)
   RigVMFunctionReturnNode(A=@SequenceAggregate.A, B=@SequenceAggregate_1.A, C=@SequenceAggregate_2.A, D=@SequenceAggregate_2.B)
  ]}
 RigVMFunctionLibrary {links [], nodes []}
 RigVMModel {nodes [
   RigUnit_BeginExecution = &RigUnit_BeginExecution()
   SequenceAggregate = RigVMAggregateNode(ExecuteContext=@RigUnit_BeginExecution.ExecutePin)
   &RigUnit_ParentConstraint(
    AdvancedSettings=(RotationOrderForFilter=XZY)
    Child=(Type=Null,Name="dyn_hair_01_mid_Null")
    Filter=(TranslationFilter=(bX=true,bY=true,bZ=true),RotationFilter=(bX=false,bY=false,bZ=false),ScaleFilter=(bX=True,bY=True,bZ=True))
    Parents=((Item=(Type=Bone,Name="head"),Weight=0.9),(Item=(Type=Bone,Name="spine_05"),Weight=0.1))
    Weight=1
    ExecutePin=@SequenceAggregate.A
   )
   > &RigUnit_ParentConstraint(
    AdvancedSettings=(RotationOrderForFilter=XZY)
    Child=(Type=Null,Name="dyn_hair_02_mid_Null")
    Filter=(TranslationFilter=(bX=True,bY=True,bZ=True),RotationFilter=(bX=false,bY=false,bZ=false),ScaleFilter=(bX=True,bY=True,bZ=True))
    Parents=((Item=(Type=Bone,Name="neck_02"),Weight=0.6),(Item=(Type=Bone,Name="spine_05"),Weight=0.4))
    Weight=1
   )
   > &RigUnit_ParentConstraint(
    AdvancedSettings=(RotationOrderForFilter=XZY)
    Child=(Type=Null,Name="dyn_hair_03_mid_Null")
    Filter=(TranslationFilter=(bX=True,bY=True,bZ=True),RotationFilter=(bX=false,bY=false,bZ=false),ScaleFilter=(bX=True,bY=True,bZ=True))
    Parents=((Item=(Type=Bone,Name="neck_02"),Weight=0.3),(Item=(Type=Bone,Name="spine_05"),Weight=0.7))
    Weight=1
   )
   > &RigUnit_ParentConstraint(
    AdvancedSettings=(RotationOrderForFilter=XZY)
    Child=(Type=Null,Name="dyn_hair_04_mid_Null")
    Filter=(TranslationFilter=(bX=True,bY=True,bZ=True),RotationFilter=(bX=false,bY=false,bZ=false),ScaleFilter=(bX=True,bY=True,bZ=True))
    Parents=((Item=(Type=Bone,Name="head"),Weight=0),(Item=(Type=Bone,Name="spine_05"),Weight=1),(Item=(Type=Bone,Name="clavicle_l"),Weight=0))
    Weight=1
   )
   > &RigUnit_ParentConstraint(
    AdvancedSettings=(RotationOrderForFilter=XZY)
    Child=(Type=Null,Name="dyn_hair_05_mid_Null")
    Filter=(TranslationFilter=(bX=True,bY=True,bZ=True),RotationFilter=(bX=false,bY=false,bZ=false),ScaleFilter=(bX=True,bY=True,bZ=True))
    Parents=((Item=(Type=Bone,Name="clavicle_l"),Weight=0),(Item=(Type=Bone,Name="spine_05"),Weight=1),(Item=(Type=Bone,Name="clavicle_r"),Weight=0))
    Weight=1
   )
   > &RigUnit_ParentConstraint(
    AdvancedSettings=(RotationOrderForFilter=XZY)
    Child=(Type=Null,Name="None")
    Filter=(TranslationFilter=(bX=True,bY=True,bZ=True),RotationFilter=(bX=false,bY=false,bZ=false),ScaleFilter=(bX=True,bY=True,bZ=True))
    Parents=((Item=(Type=Bone,Name="dyn_Hair_Back_06"),Weight=0),(Item=(Type=Bone,Name="spine_03"),Weight=1))
    Weight=1
   )
   RerouteNode_2 = RigVMDispatchNode(Values=((Type=Bone,Name="dyn_hair_01_mid"),((),Type=Bone,Name="dyn_hair_02_mid"),((),Type=Bone,Name="dyn_hair_03_mid"),((),Type=Bone,Name="dyn_hair_04_mid"),((),Type=Bone,Name="dyn_hair_05_mid"),((),Type=Bone,Name="None"))) {template="DISPATCH_RigVMDispatch_ArrayMake(in Values,out Array)"}
   > SplineIK_2 = RigVMFunctionReferenceNode(
    Primary Axis=(X=1,Y=0,Z=0)
    Samples Per Segment=1
    Secondary Spline Direction=(X=-1,Y=-1,Z=-1)
    Spline Mode=BSpline
    Stretch=false
    Up Axis=(X=0,Y=0,Z=0)
    debug=false
    Bones=@RerouteNode_2.Array
    Controls=RigVMVariableNode() {variable=dyn_Hair_Back_ctrls}.Value
   )
   > TwistBones = RigVMFunctionReferenceNode(
    Axis to Twist=(bX=true,bY=true,bZ=true)
    Bones to Twist=((Type=Bone,Name="dyn_hair_03_mid"),(Type=Bone,Name="dyn_hair_05_mid"))
    End=(Type=Bone,Name="spine_05")
    Start=(Type=Bone,Name="pelvis")
   )
   &RigUnit_ParentConstraint(
    AdvancedSettings=(RotationOrderForFilter=XZY)
    Child=(Type=Bone,Name="dyn_hair_back_01_l")
    Filter=(TranslationFilter=(bX=true,bY=true,bZ=true),RotationFilter=(bX=true,bY=true,bZ=true),ScaleFilter=(bX=True,bY=True,bZ=True))
    Parents=((Item=(Type=Bone,Name="spine_05"),Weight=0.5),(Item=(Type=Bone,Name="head"),Weight=0.5))
    Weight=1
    ExecutePin=@SequenceAggregate.B
   )
   > &RigUnit_ParentConstraint(
    AdvancedSettings=(RotationOrderForFilter=XZY)
    Child=(Type=Bone,Name="dyn_hair_back_02_l")
    Filter=(TranslationFilter=(bX=True,bY=True,bZ=True),RotationFilter=(bX=true,bY=true,bZ=true),ScaleFilter=(bX=True,bY=True,bZ=True))
    Parents=((Item=(Type=Bone,Name="clavicle_l"),Weight=0.5),(Item=(Type=Bone,Name="spine_05"),Weight=0.5),(Item=(Type=Bone,Name="dyn_hair_02_mid"),Weight=0))
    Weight=1
   )
   > &RigUnit_ParentConstraint(
    AdvancedSettings=(RotationOrderForFilter=XZY)
    Child=(Type=Bone,Name="dyn_hair_back_03_l")
    Filter=(TranslationFilter=(bX=True,bY=True,bZ=True),RotationFilter=(bX=true,bY=true,bZ=true),ScaleFilter=(bX=True,bY=True,bZ=True))
    Parents=((Item=(Type=Bone,Name="dyn_hair_03_mid"),Weight=1),(Item=(Type=Bone,Name="spine_05"),Weight=0))
    Weight=1
   )
   > &RigUnit_ParentConstraint(
    AdvancedSettings=(RotationOrderForFilter=XZY)
    Child=(Type=Bone,Name="dyn_hair_back_04_l")
    Filter=(TranslationFilter=(bX=True,bY=True,bZ=True),RotationFilter=(bX=true,bY=true,bZ=true),ScaleFilter=(bX=True,bY=True,bZ=True))
    Parents=((Item=(Type=Bone,Name="dyn_hair_04_mid"),Weight=1),(Item=(Type=Bone,Name="spine_05"),Weight=0))
    Weight=1
   )
   > &RigUnit_ParentConstraint(
    AdvancedSettings=(RotationOrderForFilter=XZY)
    Child=(Type=Bone,Name="dyn_hair_back_05_l")
    Filter=(TranslationFilter=(bX=True,bY=True,bZ=True),RotationFilter=(bX=true,bY=true,bZ=true),ScaleFilter=(bX=True,bY=True,bZ=True))
    Parents=((Item=(Type=Bone,Name="dyn_hair_05_mid"),Weight=1),(Item=(Type=Bone,Name="spine_04"),Weight=0))
    Weight=1
   )
   > &RigUnit_ParentConstraint(
    AdvancedSettings=(RotationOrderForFilter=XZY)
    Child=(Type=Bone,Name="None")
    Filter=(TranslationFilter=(bX=true,bY=true,bZ=true),RotationFilter=(bX=true,bY=true,bZ=true),ScaleFilter=(bX=True,bY=True,bZ=True))
    Parents=((Item=(Type=Bone,Name="None"),Weight=1),(Item=(Type=Bone,Name="spine_02"),Weight=0))
    Weight=1
   )
   > n1 = &RigUnit_SphericalPoseReader(
    ActiveRegionScaleFactors=(PositiveWidth=0.4,NegativeWidth=0.6,PositiveHeight=1,NegativeHeight=0.6)
    ActiveRegionSize=0.1
    Debug=(bDrawDebug=false,DebugScale=25,DebugSegments=20,DebugThickness=0.25)
    DriverAxis=(X=1,Y=0,Z=0)
    DriverItem=(Type=Bone,Name="upperarm_l")
    FalloffRegionScaleFactors=(PositiveWidth=0.4,NegativeWidth=0.6,PositiveHeight=1,NegativeHeight=0.6)
    FalloffSize=0.7
    FlipHeightScaling=false
    FlipWidthScaling=false
    OptionalParentItem=(Type=Bone,Name="clavicle_l")
    RotationOffset=(X=-80,Y=-30,Z=0)
   )
   > &RigUnit_ParentConstraint(
    AdvancedSettings=(RotationOrderForFilter=XZY)
    Child=(Type=Bone,Name="dyn_hair_back_03_l")
    Filter=(TranslationFilter=(bX=false,bY=false,bZ=false),RotationFilter=(bX=true,bY=true,bZ=true),ScaleFilter=(bX=True,bY=True,bZ=True))
    Parents=((Item=(Type=Bone,Name="dyn_hair_02_mid"),Weight=0.3),(Item=(Type=Bone,Name="upperarm_l"),Weight=0.7))
    Weight=@n1.OutputParam
   )
   n2 = &RigVMFunction_MathQuaternionFromEuler(Euler=(X=0,Y=40,Z=0))
   > &RigUnit_ModifyTransforms(
    ItemToModify="((Item=(Type=Bone,Name=\"dyn_hair_back_03_l\"),Transform=(Rotation=(X=0,Y=0,Z=0,W=1),Translation=(X=5,Y=-5,Z=10),Scale3D=(X=1,Y=1,Z=1))),(Item=(Type=Bone,Name=\"dyn_hair_back_04_l\"),Transform=(Rotation=(X=0,Y=0,Z=0,W=1),Translation=(X=0,Y=0,Z=0),Scale3D=(X=1, ...(truncated)"
    ItemToModify.0.Transform.Rotation=&RigVMFunction_MathQuaternionFromEuler(Euler=(X=-10,Y=10,Z=0), RotationOrder=YXZ).Result
    ItemToModify.1.Transform.Rotation=@n2.Result
    ItemToModify.2.Transform.Rotation=@n2.Result
    Weight=@n1.OutputParam
   )
   n3 = &RigVMFunction_MathQuaternionFromEuler(Euler=(X=0,Y=0,Z=5))
   > &RigUnit_ModifyTransforms(
    ItemToModify="((Item=(Type=Bone,Name=\"dyn_hair_back_03_l\"),Transform=(Rotation=(X=0,Y=0,Z=0,W=1),Translation=(X=0,Y=0,Z=1),Scale3D=(X=1,Y=1,Z=1))),(Item=(Type=Bone,Name=\"dyn_hair_back_04_l\"),Transform=(Rotation=(X=0,Y=0,Z=0,W=1),Translation=(X=0,Y=0,Z=0),Scale3D=(X=1,Y= ...(truncated)"
    ItemToModify.2.Transform.Rotation=&RigVMFunction_MathQuaternionFromEuler(Euler=(X=0,Y=0,Z=2)).Result
    ItemToModify.0.Transform.Rotation=@n3.Result
    ItemToModify.1.Transform.Rotation=@n3.Result
    Weight=@n1.OutputParam
   )
   > n4 = &RigUnit_SphericalPoseReader(
    ActiveRegionScaleFactors=(PositiveWidth=0.4,NegativeWidth=0.6,PositiveHeight=1,NegativeHeight=0.6)
    ActiveRegionSize=0.1
    Debug=(bDrawDebug=false,DebugScale=25,DebugSegments=20,DebugThickness=0.25)
    DriverAxis=(X=1,Y=0,Z=0)
    DriverItem=(Type=Bone,Name="clavicle_l")
    FalloffRegionScaleFactors=(PositiveWidth=0.4,NegativeWidth=0.6,PositiveHeight=1,NegativeHeight=0.6)
    FalloffSize=0.7
    FlipHeightScaling=false
    FlipWidthScaling=false
    OptionalParentItem=(Type=Bone,Name="None")
    RotationOffset=(X=-80,Y=-30,Z=0)
   )
   > &RigUnit_ParentConstraint(
    AdvancedSettings=(RotationOrderForFilter=XZY)
    Child=(Type=Bone,Name="dyn_hair_back_01_l")
    Filter=(TranslationFilter=(bX=false,bY=false,bZ=false),RotationFilter=(bX=true,bY=true,bZ=true),ScaleFilter=(bX=True,bY=True,bZ=True))
    Parents=((Item=(Type=Bone,Name="head"),Weight=0.7),(Item=(Type=Bone,Name="clavicle_l"),Weight=0.3))
    Weight=@n4.OutputParam
   )
   n5 = &RigVMFunction_MathQuaternionFromEuler(Euler=(X=0,Y=40,Z=0))
   > &RigUnit_ModifyTransforms(
    ItemToModify="((Item=(Type=Bone,Name=\"dyn_hair_back_02_l\"),Transform=(Rotation=(X=0,Y=0,Z=0,W=1),Translation=(X=5,Y=-5,Z=10),Scale3D=(X=1,Y=1,Z=1))),(Item=(Type=Bone,Name=\"None\"),Transform=(Rotation=(X=0,Y=0,Z=0,W=1),Translation=(X=0,Y=0,Z=0),Scale3D=(X=1,Y=1,Z=1))),(It ...(truncated)"
    ItemToModify.1.Transform.Rotation=@n5.Result
    ItemToModify.2.Transform.Rotation=@n5.Result
    ItemToModify.0.Transform.Rotation=&RigVMFunction_MathQuaternionFromEuler(Euler=(X=-10,Y=10,Z=0), RotationOrder=YXZ).Result
    Weight=@n4.OutputParam
   )
   &RigUnit_ParentConstraint(
    AdvancedSettings=(RotationOrderForFilter=XZY)
    Child=(Type=Bone,Name="dyn_hair_back_01_r")
    Filter=(TranslationFilter=(bX=true,bY=true,bZ=true),RotationFilter=(bX=true,bY=true,bZ=true),ScaleFilter=(bX=True,bY=True,bZ=True))
    Parents=((Item=(Type=Bone,Name="spine_05"),Weight=0.2),(Item=(Type=Bone,Name="head"),Weight=0.8))
    Weight=1
    ExecutePin=@SequenceAggregate.C
   )
   > &RigUnit_ParentConstraint(
    AdvancedSettings=(RotationOrderForFilter=XZY)
    Child=(Type=Bone,Name="dyn_hair_back_02_r")
    Filter=(TranslationFilter=(bX=True,bY=True,bZ=True),RotationFilter=(bX=true,bY=true,bZ=true),ScaleFilter=(bX=True,bY=True,bZ=True))
    Parents=((Item=(Type=Bone,Name="clavicle_r"),Weight=0.5),(Item=(Type=Bone,Name="spine_05"),Weight=0.5),(Item=(Type=Bone,Name="dyn_hair_02_mid"),Weight=0))
    Weight=1
   )
   > &RigUnit_ParentConstraint(
    AdvancedSettings=(RotationOrderForFilter=XZY)
    Child=(Type=Bone,Name="dyn_hair_back_03_r")
    Filter=(TranslationFilter=(bX=True,bY=True,bZ=True),RotationFilter=(bX=true,bY=true,bZ=true),ScaleFilter=(bX=True,bY=True,bZ=True))
    Parents=((Item=(Type=Bone,Name="dyn_hair_03_mid"),Weight=0.8),(Item=(Type=Bone,Name="clavicle_r"),Weight=0.2))
    Weight=1
   )
   > &RigUnit_ParentConstraint(
    AdvancedSettings=(RotationOrderForFilter=XZY)
    Child=(Type=Bone,Name="dyn_hair_back_04_r")
    Filter=(TranslationFilter=(bX=True,bY=True,bZ=True),RotationFilter=(bX=true,bY=true,bZ=true),ScaleFilter=(bX=True,bY=True,bZ=True))
    Parents=((Item=(Type=Bone,Name="dyn_hair_04_mid"),Weight=1),(Item=(Type=Bone,Name="spine_05"),Weight=0))
    Weight=1
   )
   > &RigUnit_ParentConstraint(
    AdvancedSettings=(RotationOrderForFilter=XZY)
    Child=(Type=Bone,Name="dyn_hair_back_05_r")
    Filter=(TranslationFilter=(bX=True,bY=True,bZ=True),RotationFilter=(bX=true,bY=true,bZ=true),ScaleFilter=(bX=True,bY=True,bZ=True))
    Parents=((Item=(Type=Bone,Name="dyn_hair_05_mid"),Weight=1),(Item=(Type=Bone,Name="spine_04"),Weight=0))
    Weight=1
   )
   > &RigUnit_ParentConstraint(
    AdvancedSettings=(RotationOrderForFilter=XZY)
    Child=(Type=Bone,Name="None")
    Filter=(TranslationFilter=(bX=true,bY=true,bZ=true),RotationFilter=(bX=true,bY=true,bZ=true),ScaleFilter=(bX=True,bY=True,bZ=True))
    Parents=((Item=(Type=Bone,Name="None"),Weight=1),(Item=(Type=Bone,Name="spine_02"),Weight=0))
    Weight=1
   )
   > n6 = &RigUnit_SphericalPoseReader(
    ActiveRegionScaleFactors=(PositiveWidth=0.7,NegativeWidth=0.6,PositiveHeight=1,NegativeHeight=0.6)
    ActiveRegionSize=0.1
    Debug=(bDrawDebug=false,DebugScale=25,DebugSegments=20,DebugThickness=0.25)
    DriverAxis=(X=-1,Y=0,Z=0)
    DriverItem=(Type=Bone,Name="upperarm_r")
    FalloffRegionScaleFactors=(PositiveWidth=0.7,NegativeWidth=0.6,PositiveHeight=1,NegativeHeight=0.6)
    FalloffSize=0.7
    FlipHeightScaling=false
    FlipWidthScaling=true
    OptionalParentItem=(Type=Bone,Name="clavicle_r")
    RotationOffset=(X=80,Y=-30,Z=0)
   )
   > &RigUnit_ParentConstraint(
    AdvancedSettings=(RotationOrderForFilter=XZY)
    Child=(Type=Bone,Name="dyn_hair_back_03_r")
    Filter=(TranslationFilter=(bX=true,bY=true,bZ=true),RotationFilter=(bX=true,bY=true,bZ=true),ScaleFilter=(bX=True,bY=True,bZ=True))
    Parents=((Item=(Type=Bone,Name="dyn_hair_02_mid"),Weight=0.3),(Item=(Type=Bone,Name="upperarm_r"),Weight=0.7))
    Weight=@n6.OutputParam
   )
   n7 = &RigVMFunction_MathQuaternionFromEuler(Euler=(X=0,Y=40,Z=0))
   > &RigUnit_ModifyTransforms(
    ItemToModify="((Item=(Type=Bone,Name=\"dyn_hair_back_03_r\"),Transform=(Rotation=(X=0,Y=0,Z=0,W=1),Translation=(X=5,Y=-5,Z=10),Scale3D=(X=1,Y=1,Z=1))),(Item=(Type=Bone,Name=\"dyn_hair_back_04_r\"),Transform=(Rotation=(X=0,Y=0,Z=0,W=1),Translation=(X=0,Y=0,Z=10),Scale3D=(X=1 ...(truncated)"
    ItemToModify.0.Transform.Rotation=&RigVMFunction_MathQuaternionFromEuler(Euler=(X=-10,Y=10,Z=0), RotationOrder=YXZ).Result
    ItemToModify.1.Transform.Rotation=@n7.Result
    ItemToModify.2.Transform.Rotation=@n7.Result
    Weight=@n6.OutputParam
   )
   n8 = &RigVMFunction_MathQuaternionFromEuler(Euler=(X=0,Y=0,Z=5))
   > &RigUnit_ModifyTransforms(
    ItemToModify="((Item=(Type=Bone,Name=\"dyn_hair_03_mid\"),Transform=(Rotation=(X=0,Y=0,Z=0,W=1),Translation=(X=0,Y=0,Z=1),Scale3D=(X=1,Y=1,Z=1))),(Item=(Type=Bone,Name=\"dyn_hair_04_mid\"),Transform=(Rotation=(X=0,Y=0,Z=0,W=1),Translation=(X=0,Y=0,Z=0),Scale3D=(X=1,Y=1,Z=1) ...(truncated)"
    ItemToModify.2.Transform.Rotation=&RigVMFunction_MathQuaternionFromEuler(Euler=(X=0,Y=0,Z=2)).Result
    ItemToModify.0.Transform.Rotation=@n8.Result
    ItemToModify.1.Transform.Rotation=@n8.Result
    Weight=@n6.OutputParam
   )
   > n9 = &RigUnit_SphericalPoseReader(
    ActiveRegionScaleFactors=(PositiveWidth=0.7,NegativeWidth=0.6,PositiveHeight=1,NegativeHeight=0.6)
    ActiveRegionSize=0.1
    Debug=(bDrawDebug=false,DebugScale=25,DebugSegments=20,DebugThickness=0.25)
    DriverAxis=(X=-1,Y=0,Z=0)
    DriverItem=(Type=Bone,Name="clavicle_r")
    FalloffRegionScaleFactors=(PositiveWidth=0.7,NegativeWidth=0.6,PositiveHeight=1,NegativeHeight=0.6)
    FalloffSize=0.7
    FlipHeightScaling=false
    FlipWidthScaling=true
    OptionalParentItem=(Type=Bone,Name="None")
    RotationOffset=(X=80,Y=-30,Z=0)
   )
   > &RigUnit_ParentConstraint(
    AdvancedSettings=(RotationOrderForFilter=XZY)
    Child=(Type=Bone,Name="dyn_hair_back_01_r")
    Filter=(TranslationFilter=(bX=true,bY=true,bZ=true),RotationFilter=(bX=true,bY=true,bZ=true),ScaleFilter=(bX=True,bY=True,bZ=True))
    Parents=((Item=(Type=Bone,Name="head"),Weight=0.7),(Item=(Type=Bone,Name="clavicle_r"),Weight=0.3))
    Weight=@n9.OutputParam
   )
   n10 = &RigVMFunction_MathQuaternionFromEuler(Euler=(X=0,Y=40,Z=0))
   > &RigUnit_ModifyTransforms(
    ItemToModify="((Item=(Type=Bone,Name=\"dyn_hair_back_02_r\"),Transform=(Rotation=(X=0,Y=0,Z=0,W=1),Translation=(X=5,Y=-5,Z=10),Scale3D=(X=1,Y=1,Z=1))),(Item=(Type=Bone,Name=\"None\"),Transform=(Rotation=(X=0,Y=0,Z=0,W=1),Translation=(X=0,Y=0,Z=0),Scale3D=(X=1,Y=1,Z=1))),(It ...(truncated)"
    ItemToModify.1.Transform.Rotation=@n10.Result
    ItemToModify.2.Transform.Rotation=@n10.Result
    ItemToModify.0.Transform.Rotation=&RigVMFunction_MathQuaternionFromEuler(Euler=(X=-10,Y=10,Z=0), RotationOrder=YXZ).Result
    Weight=@n9.OutputParam
   )
  ]}
 RigVMModel Rig Graph {nodes [
   n1 = &RigUnit_PrepareForExecution()
   SequenceAggregate = RigVMAggregateNode(ExecuteContext=@n1.ExecutePin)
   RerouteNode_6_1 = RigVMDispatchNode(Values="((Type=Control,Name=\"dyn_hair_01_mid_ctrl\"),((),Type=Control,Name=\"dyn_hair_02_mid_ctrl\"),((),Type=Control,Name=\"dyn_hair_03_mid_ctrl\"),((),Type=Control,Name=\"dyn_hair_04_mid_ctrl\"),((),Type=Control,Name=\"dyn_hair_05_mid_ctrl\"),((),Type=Control,Name=\"None\" ...(truncated)") {template="DISPATCH_RigVMDispatch_ArrayMake(in Values,out Array)"}
   RigVMVariableNode(Value=@RerouteNode_6_1.Array, ExecuteContext=@SequenceAggregate.A) {variable=dyn_Hair_Back_ctrls}
   > RigVMVariableNode(Value=/Script/ControlRig.RigUnit_HierarchyGetPoseItemArray(ElementType=All, Initial=false, ItemsToGet=@RerouteNode_6_1.Array) {is=struct RigVMUnitNode}.Pose) {variable=dyn_Hair_Back_pose}
   n2 = RigVMRerouteNode(Value=())
   RerouteNode_6_2 = RigVMDispatchNode(Values="((Type=Control,Name=\"dyn_Hair_BackR_01_ctrl\"),((),Type=Control,Name=\"dyn_Hair_BackR_02_ctrl\"),((),Type=Control,Name=\"dyn_Hair_BackR_03_ctrl\"),((),Type=Control,Name=\"dyn_Hair_BackR_04_ctrl\"),((),Type=Control,Name=\"dyn_Hair_BackR_05_ctrl\"),((),Type=Control,N ...(truncated)") {template="DISPATCH_RigVMDispatch_ArrayMake(in Values,out Array)"}
   RigVMVariableNode(ExecuteContext=@n2.Value, Value=@RerouteNode_6_2.Array) {variable=dyn_Hair_BackR_ctrls}
   RerouteNode_6 = RigVMDispatchNode(Values="((Type=Control,Name=\"dyn_Hair_BackL_01_ctrl\"),((),Type=Control,Name=\"dyn_Hair_BackL_02_ctrl\"),((),Type=Control,Name=\"dyn_Hair_BackL_03_ctrl\"),((),Type=Control,Name=\"dyn_Hair_BackL_04_ctrl\"),((),Type=Control,Name=\"dyn_Hair_BackL_05_ctrl\"),((),Type=Control,N ...(truncated)") {template="DISPATCH_RigVMDispatch_ArrayMake(in Values,out Array)"}
   RigVMVariableNode(ExecuteContext=(), Value=@RerouteNode_6.Array) {variable=dyn_Hair_BackL_ctrls}
   > RigVMVariableNode(Value=&RigUnit_HierarchyGetPoseItemArray(ElementType=All, Initial=false, ItemsToGet=@RerouteNode_6.Array).Pose) {variable=dyn_Hair_BackL_pose}
   RigVMVariableNode(
    ExecuteContext=()
    Value=&RigUnit_HierarchyGetPoseItemArray(ElementType=All, Initial=false, ItemsToGet=@RerouteNode_6_2.Array).Pose
   ) {variable=dyn_Hair_BackR_pose}
  ]}
]
hierarchy [
 root {loc=<0 0 0>, rot=<0 0 0 1>}
 pelvis {parent=root, loc=<0 -0.5301 98.0571>, rot=<0.0127 0.707 -0.0127 -0.707>}
 spine_01 {parent=pelvis, loc=<6.1208 -0.9328 0>, rot=<0 0 0.034 -0.9994>}
 spine_02 {parent=spine_01, loc=<6.5996 -0.2095 0>, rot=<0 0 0.0238 -0.9997>}
 spine_03 {parent=spine_02, loc=<6.4914 -0.1237 0>, rot=<0 0 0.0149 -0.9999>}
 spine_04 {parent=spine_03, loc=<6.9598 0.0794 0>, rot=<0 0 -0.0132 -0.9999>}
 spine_05 {parent=spine_04, loc=<6.9987 0 0>, rot=<0 0 -0.0015 -1>}
 clavicle_l {parent=spine_05, loc=<5.4831 -0.1053 -3.2946>, rot=<0.0465 -0.7289 -0.009 -0.6829>}
 upperarm_l {parent=clavicle_l, loc=<9.8658 -1.2136 -0.4105>, rot=<0.0164 -0.3245 -0.0269 -0.9454>}
 neck_01 {parent=spine_05, loc=<11.3138 0.1529 0>, rot=<0 0 -0.2003 -0.9797>}
 neck_02 {parent=neck_01, loc=<4.5627 0 0>, rot=<0 0 0.0504 -0.9987>}
 head {parent=neck_02, loc=<5.3914 0 0>, rot=<0 0 0.115 -0.9934>}
 dyn_bangs_a_r {parent=head, loc=<2.9665 6.8864 6.2676>, rot=<0.393 -0.5118 0.7638 -0.0152>}
 dyn_bangs_a_l {parent=head, loc=<5.3231 4.4529 -6.9191>, rot=<-0.0023 0.5483 0.8363 -0.0035>}
 dyn_bangs_b_l {parent=dyn_bangs_a_l, loc=<2.0508 -0.8156 1.9759>, rot=<-0.1941 -0.3778 0.4507 -0.7852>}
 dyn_bangs_c_l {parent=dyn_bangs_a_l, loc=<3.3621 0.1456 0.0484>, rot=<-0.0905 0.0144 -0.1192 -0.9886>}
 dyn_bangs_d_l {parent=dyn_bangs_a_l, loc=<4.2738 1.3134 -1.7313>, rot=<-0.25 -0.001 -0.0257 -0.9679>}
 dfrm_hair_mstr {parent=head, loc=<9.1012 -4.0009 0>, rot=<0 0 0 1>}
 dyn_hair_back_base_l {parent=dfrm_hair_mstr, loc=<0.0475 0 -5.564>, rot=<0.1529 -0.8383 -0.523 0.0175>}
 dyn_hair_back_mstr_l {parent=dyn_hair_back_base_l, loc=<0 0 0>, rot=<0 0 0 1>}
 dyn_hair_back_01_l {parent=dyn_hair_back_mstr_l, loc=<0 0 0>, rot=<0 0 0 1>}
 dyn_hair_back_02_l {parent=dyn_hair_back_01_l, loc=<12.1511 0 0>, rot=<-0.1266 -0.0514 0.0733 -0.9879>}
 dyn_hair_back_03_mstr_l {parent=dyn_hair_back_02_l, loc=<12.1511 0 0>, rot=<-0.1049 0.0068 -0.1403 -0.9845>}
 dyn_hair_back_03_l {parent=dyn_hair_back_03_mstr_l, loc=<0 0 0>, rot=<0 0 0 1>}
 dyn_hair_back_04_l {parent=dyn_hair_back_03_l, loc=<12.1511 0 0>, rot=<-0.0622 -0.0181 -0.3088 -0.9489>}
 dyn_hair_back_05_l {parent=dyn_hair_back_04_l, loc=<12.1511 0 0>, rot=<-0.0257 -0.0111 0.0592 -0.9979>}
 dyn_hair_strand_a_mstr_l {parent=dyn_hair_back_base_l, loc=<0 0 0>, rot=<-0.0951 0.116 0.0702 -0.9862>}
 dyn_hair_strand_a_01_l {parent=dyn_hair_strand_a_mstr_l, loc=<0 0 0>, rot=<0 0 0 1>}
 dyn_hair_strand_a_02_l {parent=dyn_hair_strand_a_01_l, loc=<13.0169 0 0>, rot=<0.1629 -0.0085 0.167 -0.9724>}
 dyn_hair_strand_a_03_l {parent=dyn_hair_strand_a_02_l, loc=<10.1098 0.1602 -0.2276>, rot=<-0.203 -0.1578 -0.1254 -0.9582>}
 dyn_hair_strand_a_04_l {parent=dyn_hair_strand_a_03_l, loc=<9.9166 0.5444 -0.281>, rot=<0.0167 -0.1725 -0.326 -0.9294>}
 dyn_hair_strand_a_05_l {parent=dyn_hair_strand_a_04_l, loc=<10.1217 0.19 0.1433>, rot=<-0.0286 0.2847 -0.0371 -0.9575>}
 dyn_hair_mid_mstr {parent=dfrm_hair_mstr, loc=<0.4694 -1.4988 0>, rot=<0.1309 -0.9914 0 0>}
 dyn_hair_01_mid {parent=dyn_hair_mid_mstr, loc=<0 0 0>, rot=<0 0 0 1>}
 dyn_hair_02_mid {parent=dyn_hair_01_mid, loc=<12.1511 0 0>, rot=<0.0049 0.001 0.0979 -0.9952>}
 dyn_hair_03_mstr_mid {parent=dyn_hair_02_mid, loc=<12.1511 0 0>, rot=<0 0 -0.1243 -0.9922>}
 dyn_hair_03_mid {parent=dyn_hair_03_mstr_mid, loc=<0 0 0>, rot=<0 0 0 1>}
 dyn_hair_04_mid {parent=dyn_hair_03_mid, loc=<12.1511 0 0>, rot=<0 0 -0.2275 -0.9738>}
 dyn_hair_05_mid {parent=dyn_hair_04_mid, loc=<12.1511 0 0>, rot=<0.0007 -0.0127 0.0589 -0.9982>}
 dyn_hair_back_mstr_r {parent=dfrm_hair_mstr, loc=<0.047 0 5.564>, rot=<0.1529 -0.8383 0.523 -0.0176>}
 dyn_hair_back_01_r {parent=dyn_hair_back_mstr_r, loc=<0 0 0>, rot=<0 0 0 1>}
 dyn_hair_back_02_r {parent=dyn_hair_back_01_r, loc=<12.1504 0 0>, rot=<0.1266 0.0514 0.0733 -0.9879>}
 dyn_hair_back_strand_01_r {parent=dyn_hair_back_02_r, loc=<7.8276 -3.627 -5.1362>, rot=<-0.1607 0.2418 0.0391 -0.9561>}
 dyn_hair_back_03_mstr_r {parent=dyn_hair_back_02_r, loc=<12.151 0 0>, rot=<0.1049 -0.0068 -0.1403 -0.9845>}
 dyn_hair_back_03_r {parent=dyn_hair_back_03_mstr_r, loc=<0 0 0>, rot=<0 0 0 1>}
 dyn_hair_back_04_r {parent=dyn_hair_back_03_r, loc=<12.1512 0 0>, rot=<0.0622 0.0181 -0.3088 -0.9489>}
 dyn_hair_back_05_r {parent=dyn_hair_back_04_r, loc=<12.1518 0 0>, rot=<0.0257 0.0111 0.0592 -0.9979>}
 dyn_hair_strand_mstr_r {parent=dyn_hair_back_mstr_r, loc=<7.1607 4.0764 -8.7623>, rot=<-0.193 0.0264 0.0831 -0.9773>}
 dyn_hair_strand_01_r {parent=dyn_hair_strand_mstr_r, loc=<0 0 0>, rot=<0 0 0 1>}
 dyn_hair_strand_02_r {parent=dyn_hair_strand_01_r, loc=<7.3389 0 0>, rot=<0.137 0.1746 0.2377 -0.9456>}
 dyn_hair_strand_03_r {parent=dyn_hair_strand_02_r, loc=<7.918 0 0>, rot=<0.1921 0.0281 -0.0509 -0.9797>}
 dyn_hair_strand_04_r {parent=dyn_hair_strand_03_r, loc=<5.9738 0 0>, rot=<0.0794 -0.0747 -0.2168 -0.9701>}
 dyn_hair_strand_05_r {parent=dyn_hair_strand_04_r, loc=<7.8697 0 0>, rot=<0.1339 0.139 -0.1984 -0.9609>}
 dyn_hair_strand_06_r {parent=dyn_hair_strand_05_r, loc=<7.8697 0 0>, rot=<0.0143 0.3339 -0.0324 -0.942>}
 dyn_hair_strand_07_r {parent=dyn_hair_strand_06_r, loc=<7.8697 0 0>, rot=<0.0633 0.307 -0.014 -0.9495>}
 clavicle_r {parent=spine_05, loc=<5.4831 -0.1053 3.2946>, rot=<0.7289 0.0465 -0.6829 0.009>}
 upperarm_r {parent=clavicle_r, loc=<-9.8658 1.2136 0.4105>, rot=<0.0164 -0.3245 -0.0269 -0.9454>}
 upperarm_twist_01_r {parent=upperarm_r, loc=<-7.2135 0.1134 0>, rot=<0 0 0 -1>}
 deform_upperarm_bk_base_r {parent=upperarm_twist_01_r, loc=<6.4053 4.3746 -0.314>, rot=<0 0 0 1>}
 deform_upperarm_bk_r {parent=deform_upperarm_bk_base_r, loc=<0 0 0>, rot=<0 0 0 1>}
 deform_upperarm_fr_base_r {parent=upperarm_twist_01_r, loc=<5.6312 -5.092 -0.3354>, rot=<0 0 0 1>}
 deform_upperarm_fr_r {parent=deform_upperarm_fr_base_r, loc=<0 0 0>, rot=<0 0 0 1>}
 deform_upperarm_up_base_r {parent=upperarm_twist_01_r, loc=<8.891 -0.4825 -3.9517>, rot=<0 0 0 1>}
 deform_upperarm_up_r {parent=deform_upperarm_up_base_r, loc=<0 0 0>, rot=<0 0 0 1>}
 deform_pec_aim_base_r {parent=spine_04, loc=<3.2452 9.3309 1.2237>, rot=<-0.4033 -0.0603 0.8507 0.3318>}
 deform_pec_aim_r {parent=deform_pec_aim_base_r, loc=<0 0 0>, rot=<0 0 0 1>}
 deform_pec_base_r {parent=deform_pec_aim_r, loc=<-9.3667 -0.0003 -0.0004>, rot=<0 0 0 1>}
 dyn_hair_01_mid_Null {parent=pelvis, type=Null, loc=<63.2198 -9.0063 0>, rot=<0.109 -0.994 0 0>}
 dyn_hair_01_mid_ctrl {controlType=EULER_TRANSFORM, parent=dyn_hair_01_mid_Null, type=Control, loc=<0 0 0>, rot=<0 0 0 1>}
 dyn_hair_02_mid_Null {parent=pelvis, type=Null, loc=<51.3575 -11.6398 0>, rot=<-0.2058 0.9786 0.005 0.0004>}
 dyn_hair_02_mid_ctrl {controlType=EULER_TRANSFORM, parent=dyn_hair_02_mid_Null, type=Control, loc=<0 0 0>, rot=<0 0 0 1>}
 dyn_hair_03_mid_Null {parent=pelvis, type=Null, loc=<40.2359 -16.5345 -0.0347>, rot=<-0.0826 0.9966 0.005 -0.0002>}
 dyn_hair_03_mid_ctrl {controlType=EULER_TRANSFORM, parent=dyn_hair_03_mid_Null, type=Control, loc=<0 0 0>, rot=<0 0 0 1>}
 dyn_hair_04_mid_Null {parent=pelvis, type=Null, loc=<28.2505 -18.535 -0.0395>, rot=<-0.1463 -0.9892 -0.0048 0.0014>}
 dyn_hair_04_mid_ctrl {controlType=EULER_TRANSFORM, parent=dyn_hair_04_mid_Null, type=Control, loc=<0 0 0>, rot=<0 0 0 1>}
 dyn_hair_05_mid_Null {parent=pelvis, type=Null, loc=<16.6194 -15.0188 0.0102>, rot=<0.0877 0.996 0.0075 -0.0135>}
 dyn_hair_05_mid_ctrl {controlType=EULER_TRANSFORM, parent=dyn_hair_05_mid_Null, type=Control, loc=<0 0 0>, rot=<0 0 0 1>}
]
variables [
 dyn_Hair_BackL_ctrls {
  type="TArray<FRigElementKey>"
  default=((Type=Control,Name="dyn_Hair_BackL_02"),(Type=Control,Name="dyn_Hair_BackL_03"),(Type=Control,Name="dyn_Hair_BackL_04"),(Type=Control,Name="dyn_Hair_BackL_05"),(Type=Control,Name="dyn_Hair_BackL_06"))
 }
 dyn_Hair_BackL_pose {type=FRigPose, default=(Elements=,HierarchyTopologyVersion=-1,PoseHash=-1)}
 dyn_Hair_BackR_ctrls {
  type="TArray<FRigElementKey>"
  default=((Type=Control,Name="dfrm_arm_spline_01_r_ctrl"),(Type=Control,Name="dfrm_arm_spline_03_r_ctrl"),(Type=Control,Name="dfrm_arm_spline_05_r_ctrl"),(Type=Control,Name="dfrm_arm_spline_07_r_ctrl"),(Type=Control,Name="dfrm_arm_spline_09_r_ctrl"))
 }
 dyn_Hair_BackR_pose {type=FRigPose, default=(Elements=,HierarchyTopologyVersion=-1,PoseHash=-1)}
 dyn_Hair_Back_ctrls {
  type="TArray<FRigElementKey>"
  default=((Type=Control,Name="dyn_Hair_BackL_02"),(Type=Control,Name="dyn_Hair_BackL_03"),(Type=Control,Name="dyn_Hair_BackL_04"),(Type=Control,Name="dyn_Hair_BackL_05"),(Type=Control,Name="dyn_Hair_BackL_06"))
 }
 dyn_Hair_Back_pose {type=FRigPose, default=(Elements=,HierarchyTopologyVersion=-1,PoseHash=-1)}
]
