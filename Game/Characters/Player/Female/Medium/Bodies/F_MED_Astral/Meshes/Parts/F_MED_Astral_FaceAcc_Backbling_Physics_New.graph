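# Physics asset for the F_MED_Astral face accessory (Backbling variant), simulating long hair against the upper body.
# Capsules on pelvis, spine, neck, head, clavicles and upper arms form the body volumes.
# Left, middle and right back-hair chains, plus side strands and a short branch, use swing limits with tight twist.
# Angularly free constraints cross-link each left and right back segment to its middle counterpart, tying the three chains together.
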PhysicsAsset /Game/Characters/Player/Female/Medium/Bodies/F_MED_Astral/Meshes/Parts/F_MED_Astral_FaceAcc_Backbling_Physics_New
bodies {
 clavicle_l [{
   shapeType=Capsule
   length=8.9294
   radius=4.9149
   shapeName=clavicle_r_capsule
   center=<4.9099 -0.9692 -0.0729>
   rotation=<rot -80 179.9985 0.0012>
  }]
 clavicle_r [{
   shapeType=Capsule
   length=8.6124
   radius=5.2319
   shapeName=clavicle_r_capsule
   center=<-4.9099 0.9692 0.0729>
   rotation=<rot -80 179.9985 -179.9988>
  }]
 dyn_hair_01_mid [{
   shapeType=Capsule
   length=10.9409
   radius=3.3435
   shapeName=dyn_hair_back_01_r_capsule
   center=<5.4454 0 0>
   rotation=<rot 90 179.9992 179.999>
  }]
 dyn_hair_02_mid [{shapeType=Capsule, length=10.9409, radius=3.3435, shapeName=dyn_hair_back_01_r_capsule, center=<5.4454 0 0>, rotation=<rot 90 90 89.9998>}]
 dyn_hair_03_mid [{shapeType=Capsule, length=10.9409, radius=3.3435, shapeName=dyn_hair_back_01_r_capsule, center=<5.4454 0 0>, rotation=<rot 90 90 89.9998>}]
 dyn_hair_04_mid [{shapeType=Capsule, length=10.9409, radius=3.3435, shapeName=dyn_hair_back_01_r_capsule, center=<5.4454 0 0>, rotation=<rot 90 90 89.9998>}]
 dyn_hair_05_mid [{shapeType=Capsule, length=10.9409, radius=3.3435, shapeName=dyn_hair_back_01_r_capsule, center=<5.4454 0 0>, rotation=<rot 90 90 89.9998>}]
 dyn_hair_back_01_l [{
   shapeType=Capsule
   length=10.9409
   radius=3.3435
   shapeName=dyn_hair_back_01_r_capsule
   center=<5.4454 0 0>
   rotation=<rot 90 179.9992 179.999>
  }]
 dyn_hair_back_01_r [{
   shapeType=Capsule
   length=10.9409
   radius=3.3435
   shapeName=dyn_hair_back_01_r_capsule
   center=<5.4454 0 0>
   rotation=<rot 90 179.9992 179.999>
  }]
 dyn_hair_back_02_l [{shapeType=Capsule, length=10.9409, radius=3.3435, shapeName=dyn_hair_back_01_r_capsule, center=<5.4454 0 0>, rotation=<rot 90 90 89.9998>}]
 dyn_hair_back_02_r [
  {shapeType=Capsule, length=10.9409, radius=3.3435, shapeName=dyn_hair_back_02_r_capsule, center=<5.4454 0 0>, rotation=<rot 90 0 -0.0002>}
  {shapeType=Capsule, length=10.7, radius=2.2, shapeName=dyn_hair_back_02_r_capsule, center=<5.2616 -0.8898 -3.5141>, rotation=<rot 90 0 15>}
 ]
 dyn_hair_back_03_l [
  {shapeType=Capsule, length=10.9409, radius=3.3435, shapeName=dyn_hair_back_01_r_capsule, center=<5.4454 0 0>, rotation=<rot 90 90 89.9998>}
  {
   shapeType=Capsule
   length=8.2536
   radius=2.7761
   shapeName=dyn_hair_back_01_r_capsule
   center=<5.4454 -1.038 3.3787>
   rotation=<rot 90 179.9998 179.9997>
  }
 ]
 dyn_hair_back_03_r [{shapeType=Capsule, length=10.651, radius=3.6334, shapeName=dyn_hair_back_01_r_capsule, center=<5.4454 0 0>, rotation=<rot 90 90 89.9998>}]
 dyn_hair_back_04_l [{shapeType=Capsule, length=10.9409, radius=3.3435, shapeName=dyn_hair_back_01_r_capsule, center=<5.4454 0 0>, rotation=<rot 90 90 89.9998>}]
 dyn_hair_back_04_r [{
   shapeType=Capsule
   length=10.9409
   radius=3.3435
   shapeName=dyn_hair_back_04_r_capsule
   center=<5.4454 0 0>
   rotation=<rot 90 -6.7215 -6.7221>
  }]
 dyn_hair_back_05_l [{shapeType=Capsule, length=10.9409, radius=3.3435, shapeName=dyn_hair_back_01_r_capsule, center=<5.4454 0 0>, rotation=<rot 90 90 89.9998>}]
 dyn_hair_back_05_r [{
   shapeType=Capsule
   length=10.9409
   radius=3.3435
   shapeName=dyn_hair_back_05_r_capsule
   center=<5.4454 0 0>
   rotation=<rot 90 -179.9602 -179.9608>
  }]
 dyn_hair_back_mstr_l [{shapeType=Capsule, length=0.4139, radius=2.7324, shapeName=dyn_hair_back_mstr_r_capsule}]
 dyn_hair_back_mstr_r [{shapeType=Capsule, length=0.4139, radius=2.7324, shapeName=dyn_hair_back_mstr_r_capsule}]
 dyn_hair_back_strand_01_r [{shapeType=Capsule, length=7.5497, radius=0.8042, shapeName=dyn_hair_back_01_r_capsule, center=<3.8788 0 0>, rotation=<rot 90 90 89.9998>}]
 dyn_hair_mid_mstr [{shapeType=Capsule, length=0.4139, radius=2.7324, shapeName=dyn_hair_back_mstr_r_capsule}]
 dyn_hair_strand_01_r [{shapeType=Capsule, length=7.1496, radius=1.6464, shapeName=dyn_hair_strand_06_r_capsule, center=<4.0125 0 0>, rotation=<rot 90 0 -0.0005>}]
 dyn_hair_strand_02_r [{shapeType=Capsule, length=7.1496, radius=1.6464, shapeName=dyn_hair_strand_06_r_capsule, center=<4.0125 0 0>, rotation=<rot 90 0 -0.0005>}]
 dyn_hair_strand_03_r [{shapeType=Capsule, length=7.1496, radius=1.6464, shapeName=dyn_hair_strand_06_r_capsule, center=<4.0125 0 0>, rotation=<rot 90 0 -0.0005>}]
 dyn_hair_strand_04_r [{shapeType=Capsule, length=7.1496, radius=1.6464, shapeName=dyn_hair_strand_06_r_capsule, center=<4.0125 0 0>, rotation=<rot 90 0 -0.0005>}]
 dyn_hair_strand_05_r [{shapeType=Capsule, length=7.1496, radius=1.6464, shapeName=dyn_hair_strand_06_r_capsule, center=<4.0125 0 0>, rotation=<rot 90 0 -0.0005>}]
 dyn_hair_strand_06_r [{shapeType=Capsule, length=7.1496, radius=1.6464, shapeName=dyn_hair_strand_06_r_capsule, center=<4.0125 0 0>, rotation=<rot 90 0 -0.0005>}]
 dyn_hair_strand_07_r [{shapeType=Capsule, length=7.1496, radius=1.6464, shapeName=dyn_hair_strand_06_r_capsule, center=<4.0125 0 0>, rotation=<rot 90 0 -0.0005>}]
 dyn_hair_strand_a_01_l [{
   shapeType=Capsule
   length=8.2659
   radius=1.5735
   shapeName=dyn_hair_strand_c_04_r_capsule
   center=<7.3558 0.0001 0>
   rotation=<rot 90 0 -0.0005>
  }]
 dyn_hair_strand_a_02_l [{
   shapeType=Capsule
   length=8.2659
   radius=1.5735
   shapeName=dyn_hair_strand_c_04_r_capsule
   center=<3.3079 0 0>
   rotation=<rot 90 -4.0448 -4.0453>
  }]
 dyn_hair_strand_a_03_l [{
   shapeType=Capsule
   length=8.2659
   radius=1.5735
   shapeName=dyn_hair_strand_c_04_r_capsule
   center=<3.3079 0 0>
   rotation=<rot 90 -4.0448 -4.0453>
  }]
 dyn_hair_strand_a_04_l [{
   shapeType=Capsule
   length=8.2659
   radius=1.5735
   shapeName=dyn_hair_strand_c_04_r_capsule
   center=<3.3079 0 0>
   rotation=<rot 90 -4.0448 -4.0453>
  }]
 dyn_hair_strand_a_05_l [{
   shapeType=Capsule
   length=8.2659
   radius=1.5735
   shapeName=dyn_hair_strand_c_04_r_capsule
   center=<3.3079 0 0>
   rotation=<rot 90 -4.0448 -4.0453>
  }]
 dyn_hair_strand_a_mstr_l [{shapeType=Capsule, length=1.3965, radius=1.1471, shapeName=dyn_hair_strand_a_mstr_l_capsule}]
 dyn_hair_strand_mstr_r [{shapeType=Capsule, length=0.4388, radius=0.8934, shapeName=dyn_hair_strand_mstr_r_capsule}]
 head [{shapeType=Capsule, length=2.6607, radius=5.5505, shapeName=head_capsule, center=<5.3408 3.1049 0>, rotation=<rot 90 180 -89.9991>}]
 neck_01 [{shapeType=Capsule, length=3.5821, radius=5.572, shapeName=neck_01_capsule}]
 pelvis [{shapeType=Capsule, length=9.5653, radius=11.8, shapeName=pelvis_capsule, center=<-2.8703 -1.4325 0>}]
 spine_01 [{shapeType=Capsule, length=7.9176, radius=8.1952, shapeName=spine_03_capsule, center=<0 0.2079 0>}]
 spine_02 [{shapeType=Capsule, length=7.1581, radius=7.3186, shapeName=spine_03_capsule, center=<0 1.4866 0>}]
 spine_03 [{shapeType=Capsule, length=5.9789, radius=8.4977, shapeName=spine_03_capsule, center=<0 2.7317 0>}]
 spine_04 [{shapeType=Capsule, length=7.1147, radius=8.393, shapeName=spine_04_capsule, center=<0 1.3459 0>}]
 spine_05 [{shapeType=Capsule, length=10.2993, radius=8.0518, shapeName=spine_05_capsule, center=<1.9792 -0.0282 0>}]
 upperarm_l [{shapeType=Capsule, length=25.4462, radius=4.5538, shapeName=upperarm_r_capsule, center=<13.5766 0.0002 0>, rotation=<rot 90 0 179.999>}]
 upperarm_r [
  {shapeType=Capsule, length=25.1695, radius=4.8305, shapeName=upperarm_r_capsule, center=<-13.5766 -0.0002 0>, rotation=<rot 90 0 -0.001>}
  {shapeType=Capsule, length=4.1929, radius=5.2646, shapeName=upperarm_r_capsule, center=<-1.5492 0 -1.095>, rotation=<rot -90 0.0395 -0.0396>}
 ]
}
constraints [
 {bone1Name=dyn_hair_back_01_r, bone2Name=dyn_hair_back_mstr_r, swing1LimitDegrees=5, swing2LimitDegrees=5, twistLimitDegrees=2}
 {bone1Name=dyn_hair_back_02_r, bone2Name=dyn_hair_back_01_r, swing1LimitDegrees=15, swing2LimitDegrees=15, twistLimitDegrees=2}
 {bone1Name=dyn_hair_back_03_r, bone2Name=dyn_hair_back_02_r, swing1LimitDegrees=35, swing2LimitDegrees=20, twistLimitDegrees=2}
 {bone1Name=dyn_hair_back_04_r, bone2Name=dyn_hair_back_03_r, swing1LimitDegrees=25, swing2LimitDegrees=15, twistLimitDegrees=2}
 {bone1Name=dyn_hair_back_05_r, bone2Name=dyn_hair_back_04_r, swing1LimitDegrees=20, swing2LimitDegrees=15, twistLimitDegrees=2}
 {bone1Name=dyn_hair_05_mid, bone2Name=dyn_hair_04_mid, swing1LimitDegrees=20, swing2LimitDegrees=15, twistLimitDegrees=2}
 {bone1Name=dyn_hair_04_mid, bone2Name=dyn_hair_03_mid, swing1LimitDegrees=25, swing2LimitDegrees=15, twistLimitDegrees=2}
 {bone1Name=dyn_hair_01_mid, bone2Name=dyn_hair_mid_mstr, swing1LimitDegrees=20, swing2LimitDegrees=5, twistLimitDegrees=2}
 {bone1Name=dyn_hair_02_mid, bone2Name=dyn_hair_01_mid, swing1LimitDegrees=20, swing2LimitDegrees=20, twistLimitDegrees=2}
 {bone1Name=dyn_hair_03_mid, bone2Name=dyn_hair_02_mid, swing1LimitDegrees=35, swing2LimitDegrees=20, twistLimitDegrees=2}
 {bone1Name=dyn_hair_back_01_l, bone2Name=dyn_hair_back_mstr_l, swing1LimitDegrees=5, swing2LimitDegrees=5, twistLimitDegrees=2}
 {bone1Name=dyn_hair_back_02_l, bone2Name=dyn_hair_back_01_l, swing1LimitDegrees=15, swing2LimitDegrees=15, twistLimitDegrees=2}
 {bone1Name=dyn_hair_back_03_l, bone2Name=dyn_hair_back_02_l, swing1LimitDegrees=35, swing2LimitDegrees=20, twistLimitDegrees=2}
 {bone1Name=dyn_hair_back_04_l, bone2Name=dyn_hair_back_03_l, swing1LimitDegrees=25, swing2LimitDegrees=15, twistLimitDegrees=2}
 {bone1Name=dyn_hair_back_05_l, bone2Name=dyn_hair_back_04_l, swing1LimitDegrees=20, swing2LimitDegrees=15, twistLimitDegrees=2}
 {bone1Name=dyn_hair_back_strand_01_r, bone2Name=dyn_hair_back_02_r, swing1LimitDegrees=20, swing2LimitDegrees=20, twistLimitDegrees=2}
 {
  bone1Name=upperarm_r
  bone2Name=clavicle_r
  swing1LimitDegrees=45
  swing1Motion=Free
  swing2LimitDegrees=45
  swing2Motion=Free
  twistLimitDegrees=45
  twistMotion=Free
 }
 {
  bone1Name=clavicle_r
  bone2Name=spine_05
  swing1LimitDegrees=45
  swing1Motion=Free
  swing2LimitDegrees=45
  swing2Motion=Free
  twistLimitDegrees=45
  twistMotion=Free
 }
 {bone1Name=dyn_hair_strand_a_05_l, bone2Name=dyn_hair_strand_a_04_l, swing1LimitDegrees=15, swing2LimitDegrees=15, twistLimitDegrees=2}
 {bone1Name=dyn_hair_strand_a_04_l, bone2Name=dyn_hair_strand_a_03_l, swing1LimitDegrees=15, swing2LimitDegrees=15, twistLimitDegrees=2}
 {bone1Name=dyn_hair_strand_a_03_l, bone2Name=dyn_hair_strand_a_02_l, swing1LimitDegrees=15, swing2LimitDegrees=15, twistLimitDegrees=2}
 {bone1Name=dyn_hair_strand_a_02_l, bone2Name=dyn_hair_strand_a_01_l, swing1LimitDegrees=15, swing2LimitDegrees=15, twistLimitDegrees=2}
 {bone1Name=dyn_hair_strand_a_01_l, bone2Name=dyn_hair_strand_a_mstr_l, swing1LimitDegrees=15, swing2LimitDegrees=15, twistLimitDegrees=2}
 {bone1Name=dyn_hair_strand_01_r, bone2Name=dyn_hair_strand_mstr_r, swing1LimitDegrees=20, swing2LimitDegrees=60, twistLimitDegrees=2}
 {bone1Name=dyn_hair_strand_02_r, bone2Name=dyn_hair_strand_01_r, swing1LimitDegrees=40, swing2LimitDegrees=25, twistLimitDegrees=2}
 {bone1Name=dyn_hair_strand_03_r, bone2Name=dyn_hair_strand_02_r, swing1LimitDegrees=35, swing2LimitDegrees=20, twistLimitDegrees=2}
 {bone1Name=dyn_hair_strand_07_r, bone2Name=dyn_hair_strand_06_r, swing1LimitDegrees=20, swing2LimitDegrees=20, twistLimitDegrees=2}
 {bone1Name=dyn_hair_strand_06_r, bone2Name=dyn_hair_strand_05_r, swing1LimitDegrees=20, swing2LimitDegrees=20, twistLimitDegrees=2}
 {bone1Name=dyn_hair_strand_05_r, bone2Name=dyn_hair_strand_04_r, swing1LimitDegrees=15, swing2LimitDegrees=30, twistLimitDegrees=2}
 {bone1Name=dyn_hair_strand_04_r, bone2Name=dyn_hair_strand_03_r, swing1LimitDegrees=15, swing2LimitDegrees=15, twistLimitDegrees=2}
 {
  bone1Name=dyn_hair_back_01_r
  bone2Name=dyn_hair_01_mid
  swing1LimitDegrees=45
  swing1Motion=Free
  swing2LimitDegrees=45
  swing2Motion=Free
  twistLimitDegrees=45
  twistMotion=Free
 }
 {
  bone1Name=dyn_hair_back_01_l
  bone2Name=dyn_hair_01_mid
  swing1LimitDegrees=45
  swing1Motion=Free
  swing2LimitDegrees=45
  swing2Motion=Free
  twistLimitDegrees=45
  twistMotion=Free
 }
 {
  bone1Name=dyn_hair_back_02_r
  bone2Name=dyn_hair_02_mid
  swing1LimitDegrees=45
  swing1Motion=Free
  swing2LimitDegrees=45
  swing2Motion=Free
  twistLimitDegrees=45
  twistMotion=Free
 }
 {
  bone1Name=dyn_hair_back_02_l
  bone2Name=dyn_hair_02_mid
  swing1LimitDegrees=45
  swing1Motion=Free
  swing2LimitDegrees=45
  swing2Motion=Free
  twistLimitDegrees=45
  twistMotion=Free
 }
 {
  bone1Name=dyn_hair_back_03_r
  bone2Name=dyn_hair_03_mid
  swing1LimitDegrees=45
  swing1Motion=Free
  swing2LimitDegrees=45
  swing2Motion=Free
  twistLimitDegrees=45
  twistMotion=Free
 }
 {
  bone1Name=dyn_hair_back_03_l
  bone2Name=dyn_hair_03_mid
  swing1LimitDegrees=45
  swing1Motion=Free
  swing2LimitDegrees=45
  swing2Motion=Free
  twistLimitDegrees=45
  twistMotion=Free
 }
 {
  bone1Name=dyn_hair_back_04_r
  bone2Name=dyn_hair_04_mid
  swing1LimitDegrees=45
  swing1Motion=Free
  swing2LimitDegrees=45
  swing2Motion=Free
  twistLimitDegrees=45
  twistMotion=Free
 }
 {
  bone1Name=dyn_hair_back_04_l
  bone2Name=dyn_hair_04_mid
  swing1LimitDegrees=45
  swing1Motion=Free
  swing2LimitDegrees=45
  swing2Motion=Free
  twistLimitDegrees=45
  twistMotion=Free
 }
 {
  bone1Name=dyn_hair_back_05_r
  bone2Name=dyn_hair_05_mid
  swing1LimitDegrees=45
  swing1Motion=Free
  swing2LimitDegrees=45
  swing2Motion=Free
  twistLimitDegrees=45
  twistMotion=Free
 }
 {
  bone1Name=dyn_hair_back_05_l
  bone2Name=dyn_hair_05_mid
  swing1LimitDegrees=45
  swing1Motion=Free
  swing2LimitDegrees=45
  swing2Motion=Free
  twistLimitDegrees=45
  twistMotion=Free
 }
]
