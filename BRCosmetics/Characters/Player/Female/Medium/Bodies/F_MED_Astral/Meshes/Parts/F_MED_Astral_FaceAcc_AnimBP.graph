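# Animation Blueprint for the Astral female character's FaceAcc mesh, driving secondary motion on her hair.
# Control Rigs, one fed by the attached parent's copied pose, produce cached poses for the simulation chains.
# These chains layer bone constraints, additive offsets, AnimDynamics and RigidBody physics, including a separate head-based bangs simulation.
# Bool blends and alphas adapt the result for backpacks, front end, custom idles, carried DBNO state and physics toggles.
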AnimBlueprint /BRCosmetics/Characters/Player/Female/Medium/Bodies/F_MED_Astral/Meshes/Parts/F_MED_Astral_FaceAcc_AnimBP
target_skeleton=&/F_MED_Astral_FaceAcc_Skeleton
graphs [
 AnimGraph {nodes [
   n1 = "Constraint - dyn_hair_strand_01_r (PP)"!(ComponentPose=Local To Component(LocalPose=Use cached pose 'presim'()), Alpha=0.7, ConstraintWeights_0=0.7, ConstraintWeights_1:float=0.3) -> (Pose) {
    alphaBoolBlend {blendOption=Linear}
    boneToModify {boneName=dyn_hair_strand_01_r}
    constraintSetup [{targetBone {boneName=head}}, {targetBone {boneName=clavicle_r}}]
    constraintWeights [1,1]
   }
   n2 = "Constraint - dyn_hair_strand_02_r (PP)"!(ComponentPose=@n1, Alpha=0.7, ConstraintWeights_0=1, ConstraintWeights_1:float=0) -> (Pose) {
    alphaBoolBlend {blendOption=Linear}
    boneToModify {boneName=dyn_hair_strand_02_r}
    constraintSetup [{targetBone {boneName=clavicle_r}}, {targetBone {boneName=head}}]
    constraintWeights [1,1]
   }
   n3 = "Constraint - dyn_hair_strand_03_r (P)"(ComponentPose=@n2, Alpha=0.7, ConstraintWeights_0=1) {
    alphaBoolBlend {blendOption=Linear}
    boneToModify {boneName=dyn_hair_strand_03_r}
    constraintSetup [{targetBone {boneName=clavicle_r}}]
    constraintWeights [1]
   }
   n4 = "Constraint - dyn_hair_strand_a_01_l (PP)"!(ComponentPose=@n3, Alpha=0.5, ConstraintWeights_0=0, ConstraintWeights_1:float=1) -> (Pose) {
    alphaBoolBlend {blendOption=Linear}
    boneToModify {boneName=dyn_hair_strand_a_01_l}
    constraintSetup [{targetBone {boneName=spine_05}}, {targetBone {boneName=head}}]
    constraintWeights [1,1]
   }
   n5 = "Constraint - dyn_hair_strand_a_02_l (PP)"!(ComponentPose=@n4, Alpha=0.5, ConstraintWeights_0=1, ConstraintWeights_1:float=0) -> (Pose) {
    alphaBoolBlend {blendOption=Linear}
    boneToModify {boneName=dyn_hair_strand_a_02_l}
    constraintSetup [{targetBone {boneName=spine_05}}, {targetBone {boneName=head}}]
    constraintWeights [1,1]
   }
   n6 = "Transform (Modify) Bone - Bone: dyn_hair_strand_a_02_l"!(ComponentPose=@n5, bAlphaBoolEnabled:bool=True, Translation="0,0,0", Rotation="-25,-35.000,-0.000", Scale="1,1,1") -> (Pose) {
    alphaInputType=Bool
    rotationMode=BMM_Additive
    rotationSpace=BCS_BoneSpace
    alphaBoolBlend {blendInTime=0.1, blendOption=Linear, blendOutTime=0.1}
    alphaScaleBiasClamp {bInterpResult=true}
    boneToModify {boneName=dyn_hair_strand_a_02_l}
    rotation {pitch=-25, yaw=-35}
   }
   n7 = "Constraint - dyn_hair_02_mid (P)"(ComponentPose=@n6, Alpha=0.3, ConstraintWeights_0=1) {
    alphaBoolBlend {blendOption=Linear}
    boneToModify {boneName=dyn_hair_02_mid}
    constraintSetup [{targetBone {boneName=spine_05}}]
    constraintWeights [1]
   }
   n8 = "Constraint - dyn_hair_back_02_l (P)"(ComponentPose=@n7, Alpha=0.3, ConstraintWeights_0=1) {
    alphaBoolBlend {blendOption=Linear}
    boneToModify {boneName=dyn_hair_back_02_l}
    constraintSetup [{targetBone {boneName=spine_05}}]
    constraintWeights [1]
   }
   n9 = "Constraint - dyn_hair_back_02_r (P)"(ComponentPose=@n8, Alpha=0.3, ConstraintWeights_0=1) {
    alphaBoolBlend {blendOption=Linear}
    boneToModify {boneName=dyn_hair_back_02_r}
    constraintSetup [{targetBone {boneName=spine_05}}]
    constraintWeights [1]
   }
   n10 = "Constraint - dyn_hair_03_mid (P)"(ComponentPose=@n9, Alpha=0.7, ConstraintWeights_0=1) {
    alphaBoolBlend {blendOption=Linear}
    boneToModify {boneName=dyn_hair_03_mid}
    constraintSetup [{targetBone {boneName=spine_05}}]
    constraintWeights [1]
   }
   n11 = "Constraint - dyn_hair_back_03_l (P)"(ComponentPose=@n10, Alpha=0.7, ConstraintWeights_0=1) {
    alphaBoolBlend {blendOption=Linear}
    boneToModify {boneName=dyn_hair_back_03_l}
    constraintSetup [{targetBone {boneName=spine_05}}]
    constraintWeights [1]
   }
   n12 = "Constraint - dyn_hair_back_03_r (P)"(ComponentPose=@n11, Alpha=0.7, ConstraintWeights_0=1) {
    alphaBoolBlend {blendOption=Linear}
    boneToModify {boneName=dyn_hair_back_03_r}
    constraintSetup [{targetBone {boneName=spine_05}}]
    constraintWeights [1]
   }
   n13 = "Constraint - dyn_hair_04_mid (P)"(ComponentPose=@n12, Alpha=0.7, ConstraintWeights_0=1) {
    alphaBoolBlend {blendOption=Linear}
    boneToModify {boneName=dyn_hair_04_mid}
    constraintSetup [{targetBone {boneName=spine_05}}]
    constraintWeights [1]
   }
   n14 = "Constraint - dyn_hair_back_04_l (P)"(ComponentPose=@n13, Alpha=0.7, ConstraintWeights_0=1) {
    alphaBoolBlend {blendOption=Linear}
    boneToModify {boneName=dyn_hair_back_04_l}
    constraintSetup [{targetBone {boneName=spine_05}}]
    constraintWeights [1]
   }
   n15 = "Constraint - dyn_hair_back_04_r (P)"(ComponentPose=@n14, Alpha=0.7, ConstraintWeights_0=1) {
    alphaBoolBlend {blendOption=Linear}
    boneToModify {boneName=dyn_hair_back_04_r}
    constraintSetup [{targetBone {boneName=spine_05}}]
    constraintWeights [1]
   }
   n16 = "Transform (Modify) Bone - Bone: dyn_hair_back_02_r"!(ComponentPose=@n15, bAlphaBoolEnabled:bool=True, Translation="0,0,0", Rotation="10,-15,-0.000", Scale="1,1,1") -> (Pose) {
    alphaInputType=Bool
    rotationMode=BMM_Additive
    rotationSpace=BCS_BoneSpace
    alphaBoolBlend {blendInTime=0.1, blendOption=Linear, blendOutTime=0.1}
    alphaScaleBiasClamp {bInterpResult=true}
    boneToModify {boneName=dyn_hair_back_02_r}
    rotation {pitch=10, yaw=-15}
   }
   n17 = "Transform (Modify) Bone - Bone: dyn_hair_02_mid"!(ComponentPose=@n16, bAlphaBoolEnabled:bool=True, Translation="0,0,0", Rotation="0,-15,-0.000", Scale="1,1,1") -> (Pose) {
    alphaInputType=Bool
    rotationMode=BMM_Additive
    rotationSpace=BCS_BoneSpace
    alphaBoolBlend {blendInTime=0.1, blendOption=Linear, blendOutTime=0.1}
    alphaScaleBiasClamp {bInterpResult=true}
    boneToModify {boneName=dyn_hair_02_mid}
    rotation {yaw=-15}
   }
   n18 = "Transform (Modify) Bone - Bone: dyn_hair_back_02_l"!(ComponentPose=@n17, bAlphaBoolEnabled:bool=True, Translation="0,0,0", Rotation="-10,-15,0", Scale="1,1,1") -> (Pose) {
    alphaInputType=Bool
    rotationMode=BMM_Additive
    rotationSpace=BCS_BoneSpace
    alphaBoolBlend {blendInTime=0.1, blendOption=Linear, blendOutTime=0.1}
    alphaScaleBiasClamp {bInterpResult=true}
    boneToModify {boneName=dyn_hair_back_02_l}
    rotation {pitch=-10, yaw=-15}
   }
   n19 = "Transform (Modify) Bone - Bone: dyn_hair_back_04_r"!(ComponentPose=@n18, bAlphaBoolEnabled:bool=True, Translation="0,0,0", Rotation="0,-15,-0.000", Scale="1,1,1") -> (Pose) {
    alphaInputType=Bool
    rotationMode=BMM_Additive
    rotationSpace=BCS_BoneSpace
    alphaBoolBlend {blendInTime=0.1, blendOption=Linear, blendOutTime=0.1}
    alphaScaleBiasClamp {bInterpResult=true}
    boneToModify {boneName=dyn_hair_back_04_r}
    rotation {yaw=-15}
   }
   n20 = "Transform (Modify) Bone - Bone: dyn_hair_04_mid"!(ComponentPose=@n19, bAlphaBoolEnabled:bool=True, Translation="0,0,0", Rotation="0,-15,-0.000", Scale="1,1,1") -> (Pose) {
    alphaInputType=Bool
    rotationMode=BMM_Additive
    rotationSpace=BCS_BoneSpace
    alphaBoolBlend {blendInTime=0.1, blendOption=Linear, blendOutTime=0.1}
    alphaScaleBiasClamp {bInterpResult=true}
    boneToModify {boneName=dyn_hair_04_mid}
    rotation {yaw=-15}
   }
   n21 = "Transform (Modify) Bone - Bone: dyn_hair_back_04_l"!(ComponentPose=@n20, bAlphaBoolEnabled:bool=True, Translation="0,0,0", Rotation="0,-15,-0.000", Scale="1,1,1") -> (Pose) {
    alphaInputType=Bool
    rotationMode=BMM_Additive
    rotationSpace=BCS_BoneSpace
    alphaBoolBlend {blendInTime=0.1, blendOption=Linear, blendOutTime=0.1}
    alphaScaleBiasClamp {bInterpResult=true}
    boneToModify {boneName=dyn_hair_back_04_l}
    rotation {yaw=-15}
   }
   n22 = Get BwdMovement() -> (BwdMovement:double)
   n23 = Anim Dynamics - Bone: dyn_hair_back_03_l(ComponentPose=@n21, Alpha=@n22, ExternalForce="0,0,0") {
    angularDampingOverride=0.9999
    angularSpringConstant=100
    bAngularSpring=true
    bGravityOverrideInSimSpace=true
    bOverrideAngularDamping=true
    simulationSpace=RootRelative
    alphaBoolBlend {blendOption=Linear}
    alphaScaleBiasClamp {bInterpResult=true}
    boundBone {boneName=dyn_hair_back_03_l}
    componentAppliedLinearAccClamp=<3500 3500 3500>
    componentLinearAccScale=<0.8 0.8 1.2>
    componentLinearVelScale=<0.8 0.8 1.2>
    physicsBodyDefinitions [{
      sphereCollisionRadius=0
      boundBone {boneName=dyn_hair_back_03_l}
      boxExtents=<20 20 10>
      constraintSetup {angularLimitsMax=<15 15 0>, angularLimitsMin=<-15 -3 -35>, angularTarget=<1 0 0>}
      localJointOffset=<2 0 0>
     }]
    relativeSpaceBone {boneName=root}
   }
   n24 = Anim Dynamics - Bone: dyn_hair_03_mid(ComponentPose=@n23, Alpha=@n22, ExternalForce="0,0,0") {
    angularDampingOverride=0.9999
    angularSpringConstant=100
    bAngularSpring=true
    bGravityOverrideInSimSpace=true
    bOverrideAngularDamping=true
    simulationSpace=RootRelative
    alphaBoolBlend {blendOption=Linear}
    alphaScaleBiasClamp {bInterpResult=true}
    boundBone {boneName=dyn_hair_03_mid}
    componentAppliedLinearAccClamp=<3500 3500 3500>
    componentLinearAccScale=<0.8 0.8 1.2>
    componentLinearVelScale=<0.8 0.8 1.2>
    physicsBodyDefinitions [{
      sphereCollisionRadius=0
      boundBone {boneName=dyn_hair_03_mid}
      boxExtents=<20 20 10>
      constraintSetup {angularLimitsMax=<15 5 0>, angularLimitsMin=<-15 -5 -35>, angularTarget=<1 0 0>}
      localJointOffset=<2 0 0>
     }]
    relativeSpaceBone {boneName=root}
   }
   n25 = Anim Dynamics - Bone: dyn_hair_back_03_r(ComponentPose=@n24, Alpha=@n22, ExternalForce="0,0,0") {
    angularDampingOverride=0.9999
    angularSpringConstant=100
    bAngularSpring=true
    bGravityOverrideInSimSpace=true
    bOverrideAngularDamping=true
    simulationSpace=RootRelative
    alphaBoolBlend {blendOption=Linear}
    alphaScaleBiasClamp {bInterpResult=true}
    boundBone {boneName=dyn_hair_back_03_r}
    componentAppliedLinearAccClamp=<3500 3500 3500>
    componentLinearAccScale=<0.8 0.8 1.2>
    componentLinearVelScale=<0.8 0.8 1.2>
    physicsBodyDefinitions [{
      sphereCollisionRadius=0
      boundBone {boneName=dyn_hair_back_03_r}
      boxExtents=<20 20 10>
      constraintSetup {angularLimitsMax=<15 3 0>, angularLimitsMin=<-15 -15 -35>, angularTarget=<1 0 0>}
      localJointOffset=<2 0 0>
     }]
    relativeSpaceBone {boneName=root}
   }
   n26 = Anim Dynamics - Bone: dyn_hair_back_04_l(ComponentPose=@n25, Alpha=@n22, ExternalForce="0,0,0") {
    angularDampingOverride=0.9999
    angularSpringConstant=100
    bAngularSpring=true
    bGravityOverrideInSimSpace=true
    bOverrideAngularDamping=true
    simulationSpace=RootRelative
    alphaBoolBlend {blendOption=Linear}
    alphaScaleBiasClamp {bInterpResult=true}
    boundBone {boneName=dyn_hair_back_04_l}
    componentAppliedLinearAccClamp=<3500 3500 3500>
    componentLinearAccScale=<0.8 0.8 1.2>
    componentLinearVelScale=<0.8 0.8 1.2>
    physicsBodyDefinitions [{
      sphereCollisionRadius=0
      boundBone {boneName=dyn_hair_back_04_l}
      boxExtents=<20 20 10>
      constraintSetup {angularLimitsMax=<15 15 0>, angularLimitsMin=<-15 -3 -35>, angularTarget=<1 0 0>}
      localJointOffset=<2 0 0>
     }]
    relativeSpaceBone {boneName=root}
   }
   n27 = Anim Dynamics - Bone: dyn_hair_04_mid(ComponentPose=@n26, Alpha=@n22, ExternalForce="0,0,0") {
    angularDampingOverride=0.9999
    angularSpringConstant=100
    bAngularSpring=true
    bGravityOverrideInSimSpace=true
    bOverrideAngularDamping=true
    simulationSpace=RootRelative
    alphaBoolBlend {blendOption=Linear}
    alphaScaleBiasClamp {bInterpResult=true}
    boundBone {boneName=dyn_hair_04_mid}
    componentAppliedLinearAccClamp=<3500 3500 3500>
    componentLinearAccScale=<0.8 0.8 1.2>
    componentLinearVelScale=<0.8 0.8 1.2>
    physicsBodyDefinitions [{
      sphereCollisionRadius=0
      boundBone {boneName=dyn_hair_04_mid}
      boxExtents=<20 20 10>
      constraintSetup {angularLimitsMax=<15 5 0>, angularLimitsMin=<-15 -5 -35>, angularTarget=<1 0 0>}
      localJointOffset=<2 0 0>
     }]
    relativeSpaceBone {boneName=root}
   }
   n28 = Anim Dynamics - Bone: dyn_hair_back_04_r(ComponentPose=@n27, Alpha=@n22, ExternalForce="0,0,0") {
    angularDampingOverride=0.9999
    angularSpringConstant=100
    bAngularSpring=true
    bGravityOverrideInSimSpace=true
    bOverrideAngularDamping=true
    simulationSpace=RootRelative
    alphaBoolBlend {blendOption=Linear}
    alphaScaleBiasClamp {bInterpResult=true}
    boundBone {boneName=dyn_hair_back_04_r}
    componentAppliedLinearAccClamp=<3500 3500 3500>
    componentLinearAccScale=<0.8 0.8 1.2>
    componentLinearVelScale=<0.8 0.8 1.2>
    physicsBodyDefinitions [{
      sphereCollisionRadius=0
      boundBone {boneName=dyn_hair_back_04_r}
      boxExtents=<20 20 10>
      constraintSetup {angularLimitsMax=<15 3 0>, angularLimitsMin=<-15 -15 -35>, angularTarget=<1 0 0>}
      localJointOffset=<2 0 0>
     }]
    relativeSpaceBone {boneName=root}
   }
   n29 = Anim Dynamics - Bone: dyn_hair_back_05_l(ComponentPose=@n28, Alpha=@n22, ExternalForce="0,0,0") {
    angularDampingOverride=0.9999
    angularSpringConstant=100
    bAngularSpring=true
    bGravityOverrideInSimSpace=true
    bOverrideAngularDamping=true
    simulationSpace=RootRelative
    alphaBoolBlend {blendOption=Linear}
    alphaScaleBiasClamp {bInterpResult=true}
    boundBone {boneName=dyn_hair_back_05_l}
    componentAppliedLinearAccClamp=<3500 3500 3500>
    componentLinearAccScale=<0.8 0.8 1.2>
    componentLinearVelScale=<0.8 0.8 1.2>
    physicsBodyDefinitions [{
      sphereCollisionRadius=0
      boundBone {boneName=dyn_hair_back_05_l}
      boxExtents=<20 20 10>
      constraintSetup {angularLimitsMax=<15 15 0>, angularLimitsMin=<-15 -3 -35>, angularTarget=<1 0 0>}
      localJointOffset=<2 0 0>
     }]
    relativeSpaceBone {boneName=root}
   }
   n30 = Anim Dynamics - Bone: dyn_hair_05_mid(ComponentPose=@n29, Alpha=@n22, ExternalForce="0,0,0") {
    angularDampingOverride=0.9999
    angularSpringConstant=100
    bAngularSpring=true
    bGravityOverrideInSimSpace=true
    bOverrideAngularDamping=true
    simulationSpace=RootRelative
    alphaBoolBlend {blendOption=Linear}
    alphaScaleBiasClamp {bInterpResult=true}
    boundBone {boneName=dyn_hair_05_mid}
    componentAppliedLinearAccClamp=<3500 3500 3500>
    componentLinearAccScale=<0.8 0.8 1.2>
    componentLinearVelScale=<0.8 0.8 1.2>
    physicsBodyDefinitions [{
      sphereCollisionRadius=0
      boundBone {boneName=dyn_hair_05_mid}
      boxExtents=<20 20 10>
      constraintSetup {angularLimitsMax=<15 5 0>, angularLimitsMin=<-15 -5 -35>, angularTarget=<1 0 0>}
      localJointOffset=<2 0 0>
     }]
    relativeSpaceBone {boneName=root}
   }
   n31 = Anim Dynamics - Bone: dyn_hair_back_05_r(ComponentPose=@n30, Alpha=@n22, ExternalForce="0,0,0") {
    angularDampingOverride=0.9999
    angularSpringConstant=100
    bAngularSpring=true
    bGravityOverrideInSimSpace=true
    bOverrideAngularDamping=true
    simulationSpace=RootRelative
    alphaBoolBlend {blendOption=Linear}
    alphaScaleBiasClamp {bInterpResult=true}
    boundBone {boneName=dyn_hair_back_05_r}
    componentAppliedLinearAccClamp=<3500 3500 3500>
    componentLinearAccScale=<0.8 0.8 1.2>
    componentLinearVelScale=<0.8 0.8 1.2>
    physicsBodyDefinitions [{
      sphereCollisionRadius=0
      boundBone {boneName=dyn_hair_back_05_r}
      boxExtents=<20 20 10>
      constraintSetup {angularLimitsMax=<15 3 0>, angularLimitsMin=<-15 -15 -35>, angularTarget=<1 0 0>}
      localJointOffset=<2 0 0>
     }]
    relativeSpaceBone {boneName=root}
   }
   n32 = Get gravity_override() -> (gravity_override:Vector)
   n33 = RigidBody!(
    ComponentPose=Local To Component(LocalPose=Use cached pose 'presim'())
    Alpha=Get is_wearing_backling_REV_alpha() -> (is_wearing_backling_REV_alpha:double)
    OverrideWorldGravity:Vector=@n32
    ExternalForce
   ) -> (Pose) {
    bClampLinearTranslationLimitToRefPose=true
    bOverrideWorldGravity=true
    overridePhysicsAsset=/Game/Characters/Player/Female/Medium/Bodies/F_MED_Astral/Meshes/Parts/F_MED_Astral_FaceAcc_Physics
    simulationSpace=BaseBoneSpace
    alphaBoolBlend {blendOption=Linear}
    baseBoneRef {boneName=spine_05}
    componentAppliedLinearAccClamp=<1850 1850 1850>
    componentLinearAccScale=<0.4 0.15 0.3>
    componentLinearVelScale=<0.4 0.15 0.3>
    overrideWorldGravity {z=-980}
    simSpaceSettings {maxAngularAcceleration=3000, maxAngularVelocity=3, maxLinearAcceleration=250, maxLinearVelocity=500, velocityScaleZ=0, worldAlpha=1}
   }
   n34 = Blend Poses by bool(BlendPose_0=Component To Local(ComponentPose=@n33), BlendPose_1=Component To Local(ComponentPose=@n31), bActiveValue=Get bRigidBodyEnabled()) {blendType=Linear}
   n35 = "Constraint - dyn_hair_strand_01_r (PP)"(ComponentPose=Local To Component(LocalPose=Use cached pose 'presim'()), Alpha=0.7, ConstraintWeights_0=0.7, ConstraintWeights_1=0.3) {
    alphaBoolBlend {blendOption=Linear}
    boneToModify {boneName=dyn_hair_strand_01_r}
    constraintSetup [{targetBone {boneName=head}}, {targetBone {boneName=clavicle_r}}]
    constraintWeights [1,1]
   }
   n36 = "Constraint - dyn_hair_strand_02_r (PP)"(ComponentPose=@n35, Alpha=0.7, ConstraintWeights_0=1, ConstraintWeights_1=0) {
    alphaBoolBlend {blendOption=Linear}
    boneToModify {boneName=dyn_hair_strand_02_r}
    constraintSetup [{targetBone {boneName=clavicle_r}}, {targetBone {boneName=head}}]
    constraintWeights [1,1]
   }
   n37 = "Constraint - dyn_hair_strand_03_r (P)"(ComponentPose=@n36, Alpha=0.7, ConstraintWeights_0=1) {
    alphaBoolBlend {blendOption=Linear}
    boneToModify {boneName=dyn_hair_strand_03_r}
    constraintSetup [{targetBone {boneName=clavicle_r}}]
    constraintWeights [1]
   }
   n38 = "Constraint - dyn_hair_strand_a_01_l (PP)"(ComponentPose=@n37, Alpha=0.5, ConstraintWeights_0=0, ConstraintWeights_1=1) {
    alphaBoolBlend {blendOption=Linear}
    boneToModify {boneName=dyn_hair_strand_a_01_l}
    constraintSetup [{targetBone {boneName=spine_05}}, {targetBone {boneName=head}}]
    constraintWeights [1,1]
   }
   n39 = "Constraint - dyn_hair_strand_a_02_l (PP)"(ComponentPose=@n38, Alpha=0.5, ConstraintWeights_0=1, ConstraintWeights_1=0) {
    alphaBoolBlend {blendOption=Linear}
    boneToModify {boneName=dyn_hair_strand_a_02_l}
    constraintSetup [{targetBone {boneName=spine_05}}, {targetBone {boneName=head}}]
    constraintWeights [1,1]
   }
   n40 = "Transform (Modify) Bone - Bone: dyn_hair_strand_a_02_l"(ComponentPose=@n39, bAlphaBoolEnabled=True, Translation="0,0,0", Rotation="-25,-35.000,-0.000", Scale="1,1,1") {
    alphaInputType=Bool
    rotationMode=BMM_Additive
    rotationSpace=BCS_BoneSpace
    alphaBoolBlend {blendInTime=0.1, blendOption=Linear, blendOutTime=0.1}
    alphaScaleBiasClamp {bInterpResult=true}
    boneToModify {boneName=dyn_hair_strand_a_02_l}
    rotation {pitch=-25, yaw=-35}
   }
   n41 = RigidBody(ComponentPose=@n40, OverrideWorldGravity=Get gravity_override()) {
    bClampLinearTranslationLimitToRefPose=true
    bOverrideWorldGravity=true
    overridePhysicsAsset=/Game/Characters/Player/Female/Medium/Bodies/F_MED_Astral/Meshes/Parts/F_MED_Astral_FaceAcc_Backbling_Physics_New
    simulationSpace=BaseBoneSpace
    alphaBoolBlend {blendOption=Linear}
    baseBoneRef {boneName=neck_02}
    componentAppliedLinearAccClamp=<6000 6000 6000>
    componentLinearAccScale=<0.1 0.1 0.1>
    componentLinearVelScale=<0.1 0.1 0.1>
    overrideWorldGravity {z=-980}
    simSpaceSettings {
     maxAngularAcceleration=100
     maxAngularVelocity=3
     maxLinearAcceleration=400
     maxLinearVelocity=250
     velocityScaleZ=0
     worldAlpha=1
     externalLinearDragV=<0.1 0.1>
    }
   }
   n42 = "Constraint - dyn_hair_02_mid (P)"(ComponentPose=@n41, Alpha=0.3, ConstraintWeights_0=1) {
    alphaBoolBlend {blendOption=Linear}
    boneToModify {boneName=dyn_hair_02_mid}
    constraintSetup [{targetBone {boneName=spine_05}}]
    constraintWeights [1]
   }
   n43 = "Constraint - dyn_hair_back_02_l (P)"(ComponentPose=@n42, Alpha=0.3, ConstraintWeights_0=1) {
    alphaBoolBlend {blendOption=Linear}
    boneToModify {boneName=dyn_hair_back_02_l}
    constraintSetup [{targetBone {boneName=spine_05}}]
    constraintWeights [1]
   }
   n44 = "Constraint - dyn_hair_back_02_r (P)"(ComponentPose=@n43, Alpha=0.3, ConstraintWeights_0=1) {
    alphaBoolBlend {blendOption=Linear}
    boneToModify {boneName=dyn_hair_back_02_r}
    constraintSetup [{targetBone {boneName=spine_05}}]
    constraintWeights [1]
   }
   n45 = "Constraint - dyn_hair_03_mid (P)"(ComponentPose=@n44, Alpha=0.7, ConstraintWeights_0=1) {
    alphaBoolBlend {blendOption=Linear}
    boneToModify {boneName=dyn_hair_03_mid}
    constraintSetup [{targetBone {boneName=spine_05}}]
    constraintWeights [1]
   }
   n46 = "Constraint - dyn_hair_back_03_l (P)"(ComponentPose=@n45, Alpha=0.7, ConstraintWeights_0=1) {
    alphaBoolBlend {blendOption=Linear}
    boneToModify {boneName=dyn_hair_back_03_l}
    constraintSetup [{targetBone {boneName=spine_05}}]
    constraintWeights [1]
   }
   n47 = "Constraint - dyn_hair_back_03_r (P)"(ComponentPose=@n46, Alpha=0.7, ConstraintWeights_0=1) {
    alphaBoolBlend {blendOption=Linear}
    boneToModify {boneName=dyn_hair_back_03_r}
    constraintSetup [{targetBone {boneName=spine_05}}]
    constraintWeights [1]
   }
   n48 = "Constraint - dyn_hair_04_mid (P)"(ComponentPose=@n47, Alpha=0.7, ConstraintWeights_0=1) {
    alphaBoolBlend {blendOption=Linear}
    boneToModify {boneName=dyn_hair_04_mid}
    constraintSetup [{targetBone {boneName=spine_05}}]
    constraintWeights [1]
   }
   n49 = "Constraint - dyn_hair_back_04_l (P)"(ComponentPose=@n48, Alpha=0.7, ConstraintWeights_0=1) {
    alphaBoolBlend {blendOption=Linear}
    boneToModify {boneName=dyn_hair_back_04_l}
    constraintSetup [{targetBone {boneName=spine_05}}]
    constraintWeights [1]
   }
   n50 = "Constraint - dyn_hair_back_04_r (P)"(ComponentPose=@n49, Alpha=0.7, ConstraintWeights_0=1) {
    alphaBoolBlend {blendOption=Linear}
    boneToModify {boneName=dyn_hair_back_04_r}
    constraintSetup [{targetBone {boneName=spine_05}}]
    constraintWeights [1]
   }
   n51 = "Transform (Modify) Bone - Bone: dyn_hair_back_04_r"(ComponentPose=@n50, bAlphaBoolEnabled=True, Translation="0,0,0", Rotation="0,-15,-0.000", Scale="1,1,1") {
    alphaInputType=Bool
    rotationMode=BMM_Additive
    rotationSpace=BCS_BoneSpace
    alphaBoolBlend {blendInTime=0.1, blendOption=Linear, blendOutTime=0.1}
    alphaScaleBiasClamp {bInterpResult=true}
    boneToModify {boneName=dyn_hair_back_04_r}
    rotation {yaw=-15}
   }
   n52 = "Transform (Modify) Bone - Bone: dyn_hair_04_mid"(ComponentPose=@n51, bAlphaBoolEnabled=True, Translation="0,0,0", Rotation="0,-15,-0.000", Scale="1,1,1") {
    alphaInputType=Bool
    rotationMode=BMM_Additive
    rotationSpace=BCS_BoneSpace
    alphaBoolBlend {blendInTime=0.1, blendOption=Linear, blendOutTime=0.1}
    alphaScaleBiasClamp {bInterpResult=true}
    boneToModify {boneName=dyn_hair_04_mid}
    rotation {yaw=-15}
   }
   n53 = "Transform (Modify) Bone - Bone: dyn_hair_back_04_l"(ComponentPose=@n52, bAlphaBoolEnabled=True, Translation="0,0,0", Rotation="0,-15,-0.000", Scale="1,1,1") {
    alphaInputType=Bool
    rotationMode=BMM_Additive
    rotationSpace=BCS_BoneSpace
    alphaBoolBlend {blendInTime=0.1, blendOption=Linear, blendOutTime=0.1}
    alphaScaleBiasClamp {bInterpResult=true}
    boneToModify {boneName=dyn_hair_back_04_l}
    rotation {yaw=-15}
   }
   n54 = Component To Local(ComponentPose=@n53)
   n55 = Get backpack() -> (backpack:bool)
   n56 = RigidBody(
    ComponentPose=Local To Component(LocalPose=Blend Poses by bool(BlendPose_0=@n54, BlendPose_1=@n34, bActiveValue=@n55) {blendType=Linear})
    OverrideWorldGravity=@n32
   ) {
    bClampLinearTranslationLimitToRefPose=true
    bOverrideWorldGravity=true
    overridePhysicsAsset=&/F_MED_Astral_FaceAcc_Bangs_Physics
    simulationSpace=BaseBoneSpace
    alphaBoolBlend {blendOption=Linear}
    baseBoneRef {boneName=head}
    componentAppliedLinearAccClamp=<1850 1850 1850>
    componentLinearAccScale=<0.4 0.15 0.3>
    componentLinearVelScale=<0.4 0.15 0.3>
    overrideWorldGravity {z=-980}
    simSpaceSettings {maxAngularAcceleration=3000, maxAngularVelocity=3, maxLinearAcceleration=250, maxLinearVelocity=500, velocityScaleZ=0, worldAlpha=1}
   }
   Component To Local(ComponentPose=@n56)
   Component To Local()
   n57 = "Transform (Modify) Bone - Bone: dfrm_hair_mstr"(
    ComponentPose=Local To Component(LocalPose=Use cached pose 'presim'())
    Alpha=Get is_playing_custom_idle_alpha() -> (is_playing_custom_idle_alpha:double)
    Translation="0,0,0"
    Rotation="-0.000,-15,15.000"
    Scale="1,1,1"
   ) {rotationMode=BMM_Additive, alphaBoolBlend {blendOption=Linear}, boneToModify {boneName=dfrm_hair_mstr}, rotation {roll=15, yaw=-15}}
   Local To Component(LocalPose=Blend Poses by bool(BlendPose_0=Component To Local(ComponentPose=@n57), bActiveValue=Get bIsInFrontEnd()) {blendType=Linear})
   n58 = Blend Poses by bool(BlendPose_0=Use cached pose 'HairCachePose'(), BlendPose_1="Use cached pose 'HairBeard  on/off'"(), bActiveValue=Get bIsDBNOCarried()) {blendType=Linear}
   n59 = Use cached pose 'HairCachePose'()
   n60 = Get bAnimDynamicsEnabled() -> (bAnimDynamicsEnabled:bool)
   n61 = RigidBody(
    ComponentPose=Local To Component(LocalPose=Blend Poses by bool(BlendPose_0=@n58, BlendPose_1=@n59, bActiveValue=@n60) {blendType=Linear})
    OverrideWorldGravity="0,0,-980"
   ) {
    bClampLinearTranslationLimitToRefPose=true
    bOverrideWorldGravity=true
    overridePhysicsAsset=&F_MED_Astral_FaceAcc_Bangs_Physics
    simulationSpace=BaseBoneSpace
    alphaBoolBlend {blendOption=Linear}
    baseBoneRef {boneName=head}
    componentAppliedLinearAccClamp=<1850 1850 1850>
    componentLinearAccScale=<0.4 0.15 0.3>
    componentLinearVelScale=<0.4 0.15 0.3>
    overrideWorldGravity {z=-980}
    simSpaceSettings {maxAngularAcceleration=3000, maxAngularVelocity=3, maxLinearAcceleration=250, maxLinearVelocity=500, velocityScaleZ=0, worldAlpha=1}
   }
   n62 = "Transform (Modify) Bone - Bone: dyn_hair_strand_a_02_l"!(ComponentPose=@n61, Alpha:float=1, Translation="0.291,-0.458,-4.178", Rotation="-14.133,-15.161,5.551") -> (Pose) {
    rotationMode=BMM_Additive
    rotationSpace=BCS_BoneSpace
    scaleMode=BMM_Additive
    scaleSpace=BCS_BoneSpace
    translationMode=BMM_Additive
    translationSpace=BCS_BoneSpace
    alphaBoolBlend {blendOption=Linear}
    boneToModify {boneName=dyn_hair_strand_a_02_l}
    rotation=<rot -14.1329 -15.1609 5.5509>
    translation=<0.291 -0.458 -4.178>
   }
   n63 = "Transform (Modify) Bone - Bone: dyn_hair_back_02_l"!(ComponentPose=@n62, Alpha:float=1, Translation="1.331,-0.000,-2.964", Rotation="0.052,-14.831,2.167") -> (Pose) {
    rotationMode=BMM_Additive
    rotationSpace=BCS_BoneSpace
    scaleMode=BMM_Additive
    scaleSpace=BCS_BoneSpace
    translationMode=BMM_Additive
    translationSpace=BCS_BoneSpace
    alphaBoolBlend {blendOption=Linear}
    boneToModify {boneName=dyn_hair_back_02_l}
    rotation=<rot 0.052 -14.8309 2.167>
    translation {x=1.331, z=-2.964}
   }
   n64 = "Transform (Modify) Bone - Bone: dyn_hair_strand_01_r"!(ComponentPose=@n63, Alpha=1, Translation="-0.000,-0.000,3.395", Rotation="10.000,0.000,0.000") -> (Pose) {
    rotationMode=BMM_Additive
    rotationSpace=BCS_BoneSpace
    scaleMode=BMM_Additive
    scaleSpace=BCS_BoneSpace
    translationMode=BMM_Additive
    translationSpace=BCS_BoneSpace
    alphaBoolBlend {blendOption=Linear}
    boneToModify {boneName=dyn_hair_strand_01_r}
    rotation=<rot 9.9999 0 0>
    translation {z=3.395}
   }
   n65 = "Transform (Modify) Bone - Bone: dyn_hair_back_02_r"!(ComponentPose=@n64, Alpha:float=1, Translation="-0.000,-2.645,2.157", Rotation="-0.000,-10.000,0.000") -> (Pose) {
    rotationMode=BMM_Additive
    rotationSpace=BCS_BoneSpace
    scaleMode=BMM_Additive
    scaleSpace=BCS_BoneSpace
    translationMode=BMM_Additive
    translationSpace=BCS_BoneSpace
    alphaBoolBlend {blendOption=Linear}
    boneToModify {boneName=dyn_hair_back_02_r}
    rotation=<rot 0 -9.9998 0.0001>
    translation {y=-2.645, z=2.157}
   }
   n66 = "Transform (Modify) Bone - Bone: dyn_hair_mid_mstr"!(ComponentPose=@n65, Alpha=1, Translation="-0.000,-2.645,2.157", Rotation="-0.000,-15.000,-0.000") -> (Pose) {
    rotationMode=BMM_Additive
    rotationSpace=BCS_BoneSpace
    scaleMode=BMM_Additive
    scaleSpace=BCS_BoneSpace
    translationMode=BMM_Additive
    translationSpace=BCS_BoneSpace
    alphaBoolBlend {blendOption=Linear}
    boneToModify {boneName=dyn_hair_mid_mstr}
    rotation=<rot 0 -15 0>
    translation {y=-2.645, z=2.157}
   }
   n67 = "Transform (Modify) Bone - Bone: dyn_hair_strand_a_02_l"(ComponentPose=@n66, Alpha=1, Translation="-0.403,2.384,-16.209", Rotation="-9.847,10.151,-1.753") {
    rotationMode=BMM_Additive
    rotationSpace=BCS_BoneSpace
    scaleMode=BMM_Additive
    scaleSpace=BCS_BoneSpace
    translationMode=BMM_Additive
    translationSpace=BCS_BoneSpace
    alphaBoolBlend {blendOption=Linear}
    alphaScaleBiasClamp {bInterpResult=true, interpSpeedDecreasing=5, interpSpeedIncreasing=7}
    boneToModify {boneName=dyn_hair_strand_a_02_l}
    rotation=<rot -9.847 10.151 -1.753>
    translation=<-0.403 2.384 -16.209>
   }
   n68 = "Transform (Modify) Bone - Bone: dyn_hair_back_02_l"(ComponentPose=@n67, Alpha=1, Translation="3.275,0.609,-13.724", Rotation="0,0,0") {
    rotationMode=BMM_Additive
    rotationSpace=BCS_BoneSpace
    scaleMode=BMM_Additive
    scaleSpace=BCS_BoneSpace
    translationMode=BMM_Additive
    translationSpace=BCS_BoneSpace
    alphaBoolBlend {blendOption=Linear}
    alphaScaleBiasClamp {bInterpResult=true, interpSpeedDecreasing=5, interpSpeedIncreasing=7}
    boneToModify {boneName=dyn_hair_back_02_l}
    rotation=<rot 0 0 0>
    translation=<3.275 0.6095 -13.724>
   }
   n69 = "Transform (Modify) Bone - Bone: dyn_hair_strand_01_r"(ComponentPose=@n68, Alpha=1, Translation="-0.000,-0.000,2.004", Rotation="0,0,0") {
    rotationMode=BMM_Additive
    rotationSpace=BCS_BoneSpace
    scaleMode=BMM_Additive
    scaleSpace=BCS_BoneSpace
    translationMode=BMM_Additive
    translationSpace=BCS_BoneSpace
    alphaBoolBlend {blendOption=Linear}
    alphaScaleBiasClamp {bInterpResult=true, interpSpeedDecreasing=5, interpSpeedIncreasing=7}
    boneToModify {boneName=dyn_hair_strand_01_r}
    translation {z=2.004}
   }
   n70 = "Transform (Modify) Bone - Bone: dyn_hair_back_02_r"(ComponentPose=@n69, Alpha=1, Translation="-0.000,-2.645,2.157", Rotation="-0.000,-5.000,0.000") {
    rotationMode=BMM_Additive
    rotationSpace=BCS_BoneSpace
    scaleMode=BMM_Additive
    scaleSpace=BCS_BoneSpace
    translationMode=BMM_Additive
    translationSpace=BCS_BoneSpace
    alphaBoolBlend {blendOption=Linear}
    alphaScaleBiasClamp {bInterpResult=true, interpSpeedDecreasing=5, interpSpeedIncreasing=7}
    boneToModify {boneName=dyn_hair_back_02_r}
    rotation=<rot 0 -5 0>
    translation {y=-2.645, z=2.157}
   }
   Output Pose(Result=Component To Local(ComponentPose=@n70)) {layerGroup=None}
   n71 = Control Rig(Source=Copy Pose From Mesh() {bUseAttachedParent=true}) {
    alphaBoolBlend {blendOption=Linear}
    controlRigAssetReference {blueprintRigClass=&/F_MED_Astral_FaceAcc_CtrlRig1.F_MED_Astral_FaceAcc_CtrlRig1_C}
   }
   Save cached pose 'HairCachePose'(Pose=@n71) {cacheName=HairCachePose}
   n72 = Control Rig!(Source, Alpha, is_front_end:bool=Get bIsInFrontEnd(), is_playing_custom_idle_alpha:double=Get is_playing_custom_idle_alpha()) -> (Pose) {
    alphaBoolBlend {blendOption=Linear}
    controlRigAssetReference {blueprintRigClass=&/F_MED_Astral_FaceAcc_CtrlRig.F_MED_Astral_FaceAcc_CtrlRig_C}
   }
   Save cached pose 'presim'(Pose=@n72) {cacheName=presim}
   n73 = "Transform (Modify) Bone - Bone: dyn_hair_back_02_l"!(
    ComponentPose=Local To Component(LocalPose=Use cached pose 'HairCachePose'())
    Alpha=1
    Translation="0,0,0"
    Rotation="0,-5,0"
    Scale:Vector="1,1,1"
   ) -> (Pose) {
    rotationMode=BMM_Additive
    rotationSpace=BCS_BoneSpace
    scaleMode=BMM_Additive
    scaleSpace=BCS_BoneSpace
    translationMode=BMM_Additive
    translationSpace=BCS_BoneSpace
    alphaBoolBlend {blendOption=Linear}
    boneToModify {boneName=dyn_hair_back_02_l}
    rotation {yaw=-5}
   }
   n74 = "Transform (Modify) Bone - Bone: dyn_hair_back_03_l"(ComponentPose=@n73, Alpha=1, Translation="0,0,0", Rotation="0,-5,0", Scale="1,1,1") {
    rotationMode=BMM_Additive
    rotationSpace=BCS_BoneSpace
    scaleMode=BMM_Additive
    scaleSpace=BCS_BoneSpace
    translationMode=BMM_Additive
    translationSpace=BCS_BoneSpace
    alphaBoolBlend {blendOption=Linear}
    boneToModify {boneName=dyn_hair_back_03_l}
    rotation {yaw=-5}
   }
   n75 = "Transform (Modify) Bone - Bone: dyn_hair_back_04_l"!(ComponentPose=@n74, Alpha:float=1, Translation="0,0,0", Rotation="0,-5,0", Scale="1,1,1") -> (Pose) {
    rotationMode=BMM_Additive
    rotationSpace=BCS_BoneSpace
    scaleMode=BMM_Additive
    scaleSpace=BCS_BoneSpace
    translationMode=BMM_Additive
    translationSpace=BCS_BoneSpace
    alphaBoolBlend {blendOption=Linear}
    boneToModify {boneName=dyn_hair_back_04_l}
    rotation {yaw=-5}
   }
   n76 = "Transform (Modify) Bone - Bone: dyn_hair_back_05_l"(ComponentPose=@n75, Alpha=1, Translation="0,0,0", Rotation="0,-10,0", Scale="1,1,1") {
    rotationMode=BMM_Additive
    rotationSpace=BCS_BoneSpace
    scaleMode=BMM_Additive
    scaleSpace=BCS_BoneSpace
    translationMode=BMM_Additive
    translationSpace=BCS_BoneSpace
    alphaBoolBlend {blendOption=Linear}
    boneToModify {boneName=dyn_hair_back_05_l}
    rotation {yaw=-10}
   }
   n77 = "Transform (Modify) Bone - Bone: dyn_hair_02_mid"!(ComponentPose=@n76, Alpha:float=1, Translation="0,0,0", Rotation="0,-5,0", Scale="1,1,1") -> (Pose) {
    rotationMode=BMM_Additive
    rotationSpace=BCS_BoneSpace
    scaleMode=BMM_Additive
    scaleSpace=BCS_BoneSpace
    translationMode=BMM_Additive
    translationSpace=BCS_BoneSpace
    alphaBoolBlend {blendOption=Linear}
    boneToModify {boneName=dyn_hair_02_mid}
    rotation {yaw=-5}
   }
   n78 = "Transform (Modify) Bone - Bone: dyn_hair_03_mid"(ComponentPose=@n77, Alpha=1, Translation="0,0,0", Rotation="0,-5,0", Scale="1,1,1") {
    rotationMode=BMM_Additive
    rotationSpace=BCS_BoneSpace
    scaleMode=BMM_Additive
    scaleSpace=BCS_BoneSpace
    translationMode=BMM_Additive
    translationSpace=BCS_BoneSpace
    alphaBoolBlend {blendOption=Linear}
    boneToModify {boneName=dyn_hair_03_mid}
    rotation {yaw=-5}
   }
   n79 = "Transform (Modify) Bone - Bone: dyn_hair_04_mid"!(ComponentPose=@n78, Alpha:float=1, Translation="0,0,0", Rotation="0,-5,0", Scale="1,1,1") -> (Pose) {
    rotationMode=BMM_Additive
    rotationSpace=BCS_BoneSpace
    scaleMode=BMM_Additive
    scaleSpace=BCS_BoneSpace
    translationMode=BMM_Additive
    translationSpace=BCS_BoneSpace
    alphaBoolBlend {blendOption=Linear}
    boneToModify {boneName=dyn_hair_04_mid}
    rotation {yaw=-5}
   }
   n80 = "Transform (Modify) Bone - Bone: dyn_hair_05_mid"(ComponentPose=@n79, Alpha=1, Translation="0,0,0", Rotation="0,-10,0", Scale="1,1,1") {
    rotationMode=BMM_Additive
    rotationSpace=BCS_BoneSpace
    scaleMode=BMM_Additive
    scaleSpace=BCS_BoneSpace
    translationMode=BMM_Additive
    translationSpace=BCS_BoneSpace
    alphaBoolBlend {blendOption=Linear}
    boneToModify {boneName=dyn_hair_05_mid}
    rotation {yaw=-10}
   }
   n81 = "Transform (Modify) Bone - Bone: dyn_hair_06_mid"(ComponentPose=@n80, Alpha=1, Translation="0,0,0", Rotation="0,-10,0", Scale="1,1,1") {
    rotationMode=BMM_Additive
    rotationSpace=BCS_BoneSpace
    scaleMode=BMM_Additive
    scaleSpace=BCS_BoneSpace
    translationMode=BMM_Additive
    translationSpace=BCS_BoneSpace
    alphaBoolBlend {blendOption=Linear}
    boneToModify {boneName=dyn_hair_06_mid}
    rotation {yaw=-10}
   }
   n82 = "Transform (Modify) Bone - Bone: dyn_hair_back_02_r"!(ComponentPose=@n81, Alpha=1, Translation="0,0,0", Rotation="0,-5,0", Scale:Vector="1,1,1") -> (Pose) {
    rotationMode=BMM_Additive
    rotationSpace=BCS_BoneSpace
    scaleMode=BMM_Additive
    scaleSpace=BCS_BoneSpace
    translationMode=BMM_Additive
    translationSpace=BCS_BoneSpace
    alphaBoolBlend {blendOption=Linear}
    boneToModify {boneName=dyn_hair_back_02_r}
    rotation {yaw=-5}
   }
   n83 = "Transform (Modify) Bone - Bone: dyn_hair_back_03_r"(ComponentPose=@n82, Alpha=1, Translation="0,0,0", Rotation="0,-5,0", Scale="1,1,1") {
    rotationMode=BMM_Additive
    rotationSpace=BCS_BoneSpace
    scaleMode=BMM_Additive
    scaleSpace=BCS_BoneSpace
    translationMode=BMM_Additive
    translationSpace=BCS_BoneSpace
    alphaBoolBlend {blendOption=Linear}
    boneToModify {boneName=dyn_hair_back_03_r}
    rotation {yaw=-5}
   }
   n84 = "Transform (Modify) Bone - Bone: dyn_hair_back_04_r"!(ComponentPose=@n83, Alpha:float=1, Translation="0,0,0", Rotation="0,-5,0", Scale="1,1,1") -> (Pose) {
    rotationMode=BMM_Additive
    rotationSpace=BCS_BoneSpace
    scaleMode=BMM_Additive
    scaleSpace=BCS_BoneSpace
    translationMode=BMM_Additive
    translationSpace=BCS_BoneSpace
    alphaBoolBlend {blendOption=Linear}
    boneToModify {boneName=dyn_hair_back_04_r}
    rotation {yaw=-5}
   }
   n85 = "Transform (Modify) Bone - Bone: dyn_hair_back_05_r"(ComponentPose=@n84, Alpha=1, Translation="0,0,0", Rotation="0,-10,0", Scale="1,1,1") {
    rotationMode=BMM_Additive
    rotationSpace=BCS_BoneSpace
    scaleMode=BMM_Additive
    scaleSpace=BCS_BoneSpace
    translationMode=BMM_Additive
    translationSpace=BCS_BoneSpace
    alphaBoolBlend {blendOption=Linear}
    boneToModify {boneName=dyn_hair_back_05_r}
    rotation {yaw=-10}
   }
   n86 = Anim Dynamics - Bone: dyn_hair_back_02_l(ComponentPose=@n85, Alpha=0.2, ExternalForce="0,0,0") {
    angularDampingOverride=0.99
    angularSpringConstant=75
    bAngularSpring=true
    bGravityOverrideInSimSpace=true
    bOverrideAngularDamping=true
    bUseGravityOverride=true
    simulationSpace=RootRelative
    alphaBoolBlend {blendOption=Linear}
    boundBone {boneName=dyn_hair_back_02_l}
    componentAppliedLinearAccClamp=<3500 3500 3000>
    componentLinearAccScale=<0.5 0.5 0.5>
    componentLinearVelScale=<0.65 0.65 0.55>
    gravityOverride {z=-980}
    physicsBodyDefinitions [{
      collisionType=CustomSphere
      sphereCollisionRadius=2
      boundBone {boneName=dyn_hair_back_02_l}
      boxExtents=<40 40 40>
      constraintSetup {angularLimitsMax=<0 0 10>, angularLimitsMin=<0 0 -35>, angularTarget=<0 0 0>}
      localJointOffset=<5 -0.1 0.1>
     }]
    relativeSpaceBone {boneName=root}
   }
   n87 = Anim Dynamics - Bone: dyn_hair_back_03_l(ComponentPose=@n86, Alpha=1, ExternalForce="0,0,0") {
    angularDampingOverride=0.99
    angularSpringConstant=75
    bAngularSpring=true
    bGravityOverrideInSimSpace=true
    bOverrideAngularDamping=true
    bUseGravityOverride=true
    bUseSphericalLimits=true
    simulationSpace=RootRelative
    alphaBoolBlend {blendOption=Linear}
    boundBone {boneName=dyn_hair_back_03_l}
    componentAppliedLinearAccClamp=<3500 3500 3000>
    componentLinearAccScale=<0.5 0.5 0.5>
    componentLinearVelScale=<0.65 0.65 0.55>
    gravityOverride {z=-980}
    physicsBodyDefinitions [{
      collisionType=CustomSphere
      sphereCollisionRadius=2
      boundBone {boneName=dyn_hair_back_03_l}
      boxExtents=<25 25 25>
      constraintSetup {angularLimitsMax=<0 0 10>, angularLimitsMin=<0 0 -35>, angularTarget=<0 0 0>}
      localJointOffset=<5 -0.1 0.1>
     }]
    relativeSpaceBone {boneName=root}
    sphericalLimits [{limitRadius=18, drivingBone {boneName=upperarm_l}, sphereLocalOffset=<12 12 0>}]
   }
   n88 = Anim Dynamics - Bone: dyn_hair_back_04_l(ComponentPose=@n87, Alpha=1, ExternalForce="0,0,0") {
    angularDampingOverride=0.99
    angularSpringConstant=75
    bAngularSpring=true
    bGravityOverrideInSimSpace=true
    bOverrideAngularDamping=true
    bUseGravityOverride=true
    bUseSphericalLimits=true
    simulationSpace=RootRelative
    alphaBoolBlend {blendOption=Linear}
    boundBone {boneName=dyn_hair_back_04_l}
    componentAppliedLinearAccClamp=<3500 3500 3000>
    componentLinearAccScale=<0.5 0.5 0.5>
    componentLinearVelScale=<0.65 0.65 0.55>
    gravityOverride {z=-980}
    physicsBodyDefinitions [{
      collisionType=CustomSphere
      sphereCollisionRadius=2
      boundBone {boneName=dyn_hair_back_04_l}
      boxExtents=<23 23 23>
      constraintSetup {angularLimitsMax=<0 0 10>, angularLimitsMin=<0 0 -35>, angularTarget=<0 0 0>}
      localJointOffset=<5 -0.1 0.1>
     }]
    relativeSpaceBone {boneName=root}
    sphericalLimits [{limitRadius=18, drivingBone {boneName=upperarm_l}, sphereLocalOffset=<12 12 0>}]
   }
   n89 = Anim Dynamics - Bone: dyn_hair_back_05_l(ComponentPose=@n88, Alpha=1, ExternalForce="0,0,0") {
    angularDampingOverride=0.99
    angularSpringConstant=75
    bAngularSpring=true
    bGravityOverrideInSimSpace=true
    bOverrideAngularDamping=true
    bUseGravityOverride=true
    bUseSphericalLimits=true
    simulationSpace=RootRelative
    alphaBoolBlend {blendOption=Linear}
    boundBone {boneName=dyn_hair_back_05_l}
    componentAppliedLinearAccClamp=<3500 3500 3000>
    componentLinearAccScale=<0.5 0.5 0.5>
    componentLinearVelScale=<0.65 0.65 0.55>
    gravityOverride {z=-980}
    physicsBodyDefinitions [{
      collisionType=CustomSphere
      sphereCollisionRadius=2
      boundBone {boneName=dyn_hair_back_05_l}
      boxExtents=<23 23 23>
      constraintSetup {angularLimitsMax=<0 0 10>, angularLimitsMin=<0 0 -35>, angularTarget=<0 0 0>}
      localJointOffset=<5 -0.1 0.1>
     }]
    relativeSpaceBone {boneName=root}
    sphericalLimits [{limitRadius=18, drivingBone {boneName=upperarm_l}, sphereLocalOffset=<12 12 0>}]
   }
   n90 = Anim Dynamics - Bone: dyn_hair_02_mid(ComponentPose=@n89, Alpha=0.2, ExternalForce="0,0,0") {
    angularDampingOverride=0.99
    angularSpringConstant=75
    bAngularSpring=true
    bGravityOverrideInSimSpace=true
    bOverrideAngularDamping=true
    bUseGravityOverride=true
    simulationSpace=RootRelative
    alphaBoolBlend {blendOption=Linear}
    boundBone {boneName=dyn_hair_02_mid}
    componentAppliedLinearAccClamp=<3500 3500 3000>
    componentLinearAccScale=<0.5 0.5 0.5>
    componentLinearVelScale=<0.65 0.65 0.55>
    gravityOverride {z=-980}
    physicsBodyDefinitions [{
      collisionType=CustomSphere
      sphereCollisionRadius=2
      boundBone {boneName=dyn_hair_02_mid}
      boxExtents=<40 40 40>
      constraintSetup {angularLimitsMax=<0 0 10>, angularLimitsMin=<0 0 -35>, angularTarget=<0 0 0>}
      localJointOffset=<5 -0.1 0.1>
     }]
    relativeSpaceBone {boneName=root}
   }
   n91 = Anim Dynamics - Bone: dyn_hair_03_mid(ComponentPose=@n90, Alpha=1, ExternalForce="0,0,0") {
    angularDampingOverride=0.99
    angularSpringConstant=75
    bAngularSpring=true
    bGravityOverrideInSimSpace=true
    bOverrideAngularDamping=true
    bUseGravityOverride=true
    simulationSpace=RootRelative
    alphaBoolBlend {blendOption=Linear}
    boundBone {boneName=dyn_hair_03_mid}
    componentAppliedLinearAccClamp=<3500 3500 3000>
    componentLinearAccScale=<0.5 0.5 0.5>
    componentLinearVelScale=<0.65 0.65 0.55>
    gravityOverride {z=-980}
    physicsBodyDefinitions [{
      collisionType=CustomSphere
      sphereCollisionRadius=2
      boundBone {boneName=dyn_hair_03_mid}
      boxExtents=<25 25 25>
      constraintSetup {angularLimitsMax=<0 0 10>, angularLimitsMin=<0 0 -35>, angularTarget=<0 0 0>}
      localJointOffset=<5 -0.1 0.1>
     }]
    relativeSpaceBone {boneName=root}
   }
   n92 = Anim Dynamics - Bone: dyn_hair_04_mid(ComponentPose=@n91, Alpha=1, ExternalForce="0,0,0") {
    angularDampingOverride=0.99
    angularSpringConstant=75
    bAngularSpring=true
    bGravityOverrideInSimSpace=true
    bOverrideAngularDamping=true
    bUseGravityOverride=true
    bUseSphericalLimits=true
    simulationSpace=RootRelative
    alphaBoolBlend {blendOption=Linear}
    boundBone {boneName=dyn_hair_04_mid}
    componentAppliedLinearAccClamp=<3500 3500 3000>
    componentLinearAccScale=<0.5 0.5 0.5>
    componentLinearVelScale=<0.65 0.65 0.55>
    gravityOverride {z=-980}
    physicsBodyDefinitions [{
      collisionType=CustomSphere
      sphereCollisionRadius=2
      boundBone {boneName=dyn_hair_04_mid}
      boxExtents=<23 23 23>
      constraintSetup {angularLimitsMax=<0 0 10>, angularLimitsMin=<0 0 -35>, angularTarget=<0 0 0>}
      localJointOffset=<5 -0.1 0.1>
     }]
    relativeSpaceBone {boneName=root}
    sphericalLimits [{limitRadius=18, drivingBone {boneName=pelvis}, sphereLocalOffset=<0 2 0>}]
   }
   n93 = Anim Dynamics - Bone: dyn_hair_05_mid(ComponentPose=@n92, Alpha=1, ExternalForce="0,0,0") {
    angularDampingOverride=0.99
    angularSpringConstant=75
    bAngularSpring=true
    bGravityOverrideInSimSpace=true
    bOverrideAngularDamping=true
    bUseGravityOverride=true
    bUseSphericalLimits=true
    simulationSpace=RootRelative
    alphaBoolBlend {blendOption=Linear}
    boundBone {boneName=dyn_hair_05_mid}
    componentAppliedLinearAccClamp=<3500 3500 3000>
    componentLinearAccScale=<0.5 0.5 0.5>
    componentLinearVelScale=<0.65 0.65 0.55>
    gravityOverride {z=-980}
    physicsBodyDefinitions [{
      collisionType=CustomSphere
      sphereCollisionRadius=2
      boundBone {boneName=dyn_hair_05_mid}
      boxExtents=<23 23 23>
      constraintSetup {angularLimitsMax=<0 0 10>, angularLimitsMin=<0 0 -35>, angularTarget=<0 0 0>}
      localJointOffset=<5 -0.1 0.1>
     }]
    relativeSpaceBone {boneName=root}
    sphericalLimits [{limitRadius=18, drivingBone {boneName=pelvis}, sphereLocalOffset=<0 4 0>}]
   }
   n94 = Anim Dynamics - Bone: dyn_hair_back_02_r(ComponentPose=@n93, Alpha=0.2, ExternalForce="0,0,0") {
    angularDampingOverride=0.99
    angularSpringConstant=75
    bAngularSpring=true
    bGravityOverrideInSimSpace=true
    bOverrideAngularDamping=true
    bUseGravityOverride=true
    bUseSphericalLimits=true
    simulationSpace=RootRelative
    alphaBoolBlend {blendOption=Linear}
    boundBone {boneName=dyn_hair_back_02_r}
    componentAppliedLinearAccClamp=<3500 3500 3000>
    componentLinearAccScale=<0.5 0.5 0.5>
    componentLinearVelScale=<0.65 0.65 0.55>
    gravityOverride {z=-980}
    physicsBodyDefinitions [{
      collisionType=CustomSphere
      sphereCollisionRadius=2
      boundBone {boneName=dyn_hair_back_02_r}
      boxExtents=<40 40 40>
      constraintSetup {angularLimitsMax=<0 0 10>, angularLimitsMin=<0 0 -35>, angularTarget=<0 0 0>}
      localJointOffset=<5 -0.1 0.1>
     }]
    relativeSpaceBone {boneName=root}
   }
   n95 = Anim Dynamics - Bone: dyn_hair_back_03_r(ComponentPose=@n94, Alpha=1, ExternalForce="0,0,0") {
    angularDampingOverride=0.99
    angularSpringConstant=75
    bAngularSpring=true
    bGravityOverrideInSimSpace=true
    bOverrideAngularDamping=true
    bUseGravityOverride=true
    bUseSphericalLimits=true
    simulationSpace=RootRelative
    alphaBoolBlend {blendOption=Linear}
    boundBone {boneName=dyn_hair_back_03_r}
    componentAppliedLinearAccClamp=<3500 3500 3000>
    componentLinearAccScale=<0.5 0.5 0.5>
    componentLinearVelScale=<0.65 0.65 0.55>
    gravityOverride {z=-980}
    physicsBodyDefinitions [{
      collisionType=CustomSphere
      sphereCollisionRadius=2
      boundBone {boneName=dyn_hair_back_03_r}
      boxExtents=<25 25 25>
      constraintSetup {angularLimitsMax=<0 0 10>, angularLimitsMin=<0 0 -35>, angularTarget=<0 0 0>}
      localJointOffset=<5 -0.1 0.1>
     }]
    relativeSpaceBone {boneName=root}
    sphericalLimits [{limitRadius=18, drivingBone {boneName=upperarm_r}, sphereLocalOffset=<-12 -12 0>}]
   }
   n96 = Anim Dynamics - Bone: dyn_hair_back_04_r(ComponentPose=@n95, Alpha=1, ExternalForce="0,0,0") {
    angularDampingOverride=0.99
    angularSpringConstant=75
    bAngularSpring=true
    bGravityOverrideInSimSpace=true
    bOverrideAngularDamping=true
    bUseGravityOverride=true
    bUseSphericalLimits=true
    simulationSpace=RootRelative
    alphaBoolBlend {blendOption=Linear}
    boundBone {boneName=dyn_hair_back_04_r}
    componentAppliedLinearAccClamp=<3500 3500 3000>
    componentLinearAccScale=<0.5 0.5 0.5>
    componentLinearVelScale=<0.65 0.65 0.55>
    gravityOverride {z=-980}
    physicsBodyDefinitions [{
      collisionType=CustomSphere
      sphereCollisionRadius=2
      boundBone {boneName=dyn_hair_back_04_r}
      boxExtents=<15 15 15>
      constraintSetup {angularLimitsMax=<0 0 10>, angularLimitsMin=<0 0 -35>, angularTarget=<0 0 0>}
      localJointOffset=<5 -0.1 0.1>
     }]
    relativeSpaceBone {boneName=root}
    sphericalLimits [{limitRadius=18, drivingBone {boneName=upperarm_r}, sphereLocalOffset=<-12 -12 0>}]
   }
   n97 = Anim Dynamics - Bone: dyn_hair_back_05_r(ComponentPose=@n96, Alpha=1, ExternalForce="0,0,0") {
    angularDampingOverride=0.99
    angularSpringConstant=75
    bAngularSpring=true
    bGravityOverrideInSimSpace=true
    bOverrideAngularDamping=true
    bUseGravityOverride=true
    bUseSphericalLimits=true
    simulationSpace=RootRelative
    alphaBoolBlend {blendOption=Linear}
    boundBone {boneName=dyn_hair_back_05_r}
    componentAppliedLinearAccClamp=<3500 3500 3000>
    componentLinearAccScale=<0.5 0.5 0.5>
    componentLinearVelScale=<0.65 0.65 0.55>
    gravityOverride {z=-980}
    physicsBodyDefinitions [{
      collisionType=CustomSphere
      sphereCollisionRadius=2
      boundBone {boneName=dyn_hair_back_05_r}
      boxExtents=<15 15 15>
      constraintSetup {angularLimitsMax=<0 0 10>, angularLimitsMin=<0 0 -35>, angularTarget=<0 0 0>}
      localJointOffset=<5 -0.1 0.1>
     }]
    relativeSpaceBone {boneName=root}
    sphericalLimits [{limitRadius=18, drivingBone {boneName=upperarm_r}, sphereLocalOffset=<-12 -12 0>}]
   }
   Save cached pose 'HairDynPose'(Pose=Component To Local(ComponentPose=@n97)) {cacheName=HairDynPose}
   n98 = Use cached pose 'HairDynPose'()
   n99 = Use cached pose 'HairCachePose'()
   n100 = SelectFloat(A=0.8, B=0.2, bPickA=OR Boolean!(A=Get isWearingPackBool() -> (isWearingPackBool:bool), B=false, C:bool) -> (ReturnValue))
   n101 = Get isWearingPackBool()
   n102 = Get bIsSkydiveDiveMode()
   n103 = Get bIsSkydiveFloating()
   "Save cached pose 'HairDynBackpack&&CapePose'"(Pose=Blend Poses by bool(
     BlendPose_0=Two Way Blend(A=@n98, B=@n99, Alpha=@n100) {blendNode {alphaBoolBlend {blendOption=Linear}}}
     BlendPose_1=Use cached pose 'HairDynPose'()
     bActiveValue=OR Boolean!(A=@n102, B=@n103, C=@n101, D:bool) -> (ReturnValue)
    ) {blendType=Linear}) {cacheName="HairDynBackpack&&CapePose"}
   n104 = Anim Dynamics - Bone: dyn_hair_back_strand_01_r(ComponentPose=Local To Component(LocalPose="Use cached pose 'HairDynBackpack&&CapePose'"()), Alpha=1, ExternalForce="0,0,0") {
    angularDampingOverride=0.99
    angularSpringConstant=75
    bAngularSpring=true
    bGravityOverrideInSimSpace=true
    bOverrideAngularDamping=true
    bUseGravityOverride=true
    simulationSpace=RootRelative
    alphaBoolBlend {blendOption=Linear}
    boundBone {boneName=dyn_hair_back_strand_01_r}
    componentAppliedLinearAccClamp=<3500 3500 3000>
    componentLinearAccScale=<0.5 0.5 0.5>
    componentLinearVelScale=<0.65 0.65 0.55>
    gravityOverride {z=-980}
    physicsBodyDefinitions [{
      collisionType=CustomSphere
      sphereCollisionRadius=2
      boundBone {boneName=dyn_hair_back_strand_01_r}
      boxExtents=<23 23 23>
      constraintSetup {angularLimitsMax=<1 5 1>, angularLimitsMin=<-1 -5 -35>, angularTarget=<0 0 0>}
      localJointOffset=<5 -0.1 0.1>
     }]
    relativeSpaceBone {boneName=root}
   }
   n105 = "Constraint - dyn_hair_back_04_l (PP)"!(ComponentPose=@n104, Alpha=1, ConstraintWeights_0=0.5, ConstraintWeights_1:float=0.5) -> (Pose) {
    alphaBoolBlend {blendOption=Linear}
    boneToModify {boneName=dyn_hair_back_04_l}
    constraintSetup [{targetBone {boneName=dyn_hair_04_mid}}, {targetBone {boneName=dyn_hair_back_04_l}}]
    constraintWeights [1,1]
   }
   n106 = "Constraint - dyn_hair_back_04_r (PP)"!(ComponentPose=@n105, Alpha=1, ConstraintWeights_0=0.5, ConstraintWeights_1:float=0.5) -> (Pose) {
    alphaBoolBlend {blendOption=Linear}
    boneToModify {boneName=dyn_hair_back_04_r}
    constraintSetup [{targetBone {boneName=dyn_hair_04_mid}}, {targetBone {boneName=dyn_hair_back_04_r}}]
    constraintWeights [1,1]
   }
   n107 = "Constraint - dyn_hair_back_05_l (PP)"!(ComponentPose=@n106, Alpha=1, ConstraintWeights_0=0.5, ConstraintWeights_1:float=0.5) -> (Pose) {
    alphaBoolBlend {blendOption=Linear}
    boneToModify {boneName=dyn_hair_back_05_l}
    constraintSetup [{targetBone {boneName=dyn_hair_05_mid}}, {targetBone {boneName=dyn_hair_back_05_l}}]
    constraintWeights [1,1]
   }
   n108 = "Constraint - dyn_hair_back_05_r (PP)"!(ComponentPose=@n107, Alpha=1, ConstraintWeights_0=0.5, ConstraintWeights_1:float=0.5) -> (Pose) {
    alphaBoolBlend {blendOption=Linear}
    boneToModify {boneName=dyn_hair_back_05_r}
    constraintSetup [{targetBone {boneName=dyn_hair_05_mid}}, {targetBone {boneName=dyn_hair_back_05_r}}]
    constraintWeights [1,1]
   }
   n109 = "Constraint - dyn_hair_strand_01_r (PP)"(ComponentPose=@n108, Alpha=1, ConstraintWeights_0=1, ConstraintWeights_1=0) {
    alphaBoolBlend {blendOption=Linear}
    boneToModify {boneName=dyn_hair_strand_01_r}
    constraintSetup [{targetBone {boneName=dyn_hair_back_01_r}}, {targetBone {boneName=head}}]
    constraintWeights [1,1]
   }
   n110 = "Constraint - dyn_hair_strand_02_r (PP)"(ComponentPose=@n109, Alpha=1, ConstraintWeights_0=1, ConstraintWeights_1=0) {
    alphaBoolBlend {blendOption=Linear}
    boneToModify {boneName=dyn_hair_strand_02_r}
    constraintSetup [{targetBone {boneName=dyn_hair_back_02_r}}, {targetBone {boneName=upperarm_r}}]
    constraintWeights [1,1]
   }
   n111 = "Constraint - dyn_hair_strand_03_r (PP)"!(ComponentPose=@n110, Alpha=1, ConstraintWeights_0=1, ConstraintWeights_1:float=0) -> (Pose) {
    alphaBoolBlend {blendOption=Linear}
    boneToModify {boneName=dyn_hair_strand_03_r}
    constraintSetup [{targetBone {boneName=dyn_hair_back_02_r}}, {targetBone {boneName=upperarm_r}}]
    constraintWeights [1,1]
   }
   n112 = "Constraint - dyn_hair_strand_04_r (PP)"!(ComponentPose=@n111, Alpha=1, ConstraintWeights_0=0.3, ConstraintWeights_1:float=0.7) -> (Pose) {
    alphaBoolBlend {blendOption=Linear}
    boneToModify {boneName=dyn_hair_strand_04_r}
    constraintSetup [{targetBone {boneName=dyn_hair_back_02_r}}, {targetBone {boneName=dyn_hair_back_03_r}}]
    constraintWeights [1,1]
   }
   n113 = "Constraint - dyn_hair_strand_05_r (PP)"!(ComponentPose=@n112, Alpha=1, ConstraintWeights_0=0, ConstraintWeights_1:float=1) -> (Pose) {
    alphaBoolBlend {blendOption=Linear}
    boneToModify {boneName=dyn_hair_strand_05_r}
    constraintSetup [{targetBone {boneName=dyn_hair_back_02_r}}, {targetBone {boneName=dyn_hair_back_03_r}}]
    constraintWeights [1,1]
   }
   n114 = "Constraint - dyn_hair_strand_06_r (PP)"!(ComponentPose=@n113, Alpha=1, ConstraintWeights_0=1, ConstraintWeights_1:float=0) -> (Pose) {
    alphaBoolBlend {blendOption=Linear}
    boneToModify {boneName=dyn_hair_strand_06_r}
    constraintSetup [{targetBone {boneName=dyn_hair_back_03_r}}, {targetBone {boneName=head}}]
    constraintWeights [1,1]
   }
   n115 = "Constraint - dyn_hair_strand_07_r (PP)"!(ComponentPose=@n114, Alpha=1, ConstraintWeights_0=0.5, ConstraintWeights_1:float=0.5) -> (Pose) {
    alphaBoolBlend {blendOption=Linear}
    boneToModify {boneName=dyn_hair_strand_07_r}
    constraintSetup [{targetBone {boneName=dyn_hair_back_03_r}}, {targetBone {boneName=dyn_hair_back_04_r}}]
    constraintWeights [1,1]
   }
   n116 = "Constraint - dyn_hair_strand_08_r (PP)"!(ComponentPose=@n115, Alpha=1, ConstraintWeights_0=0.5, ConstraintWeights_1:float=0.5) -> (Pose) {
    alphaBoolBlend {blendOption=Linear}
    boneToModify {boneName=dyn_hair_strand_08_r}
    constraintSetup [{targetBone {boneName=dyn_hair_back_04_r}}, {targetBone {boneName=dyn_hair_back_05_r}}]
    constraintWeights [1,1]
   }
   n117 = Anim Dynamics - Bone: dyn_hair_strand_05_r(ComponentPose=@n116, Alpha=1, ExternalForce="0,0,0") {
    angularDampingOverride=0.99
    angularSpringConstant=75
    bAngularSpring=true
    bGravityOverrideInSimSpace=true
    bOverrideAngularDamping=true
    bUseGravityOverride=true
    bUseSphericalLimits=true
    simulationSpace=RootRelative
    alphaBoolBlend {blendOption=Linear}
    boundBone {boneName=dyn_hair_strand_05_r}
    componentAppliedLinearAccClamp=<3500 3500 3000>
    componentLinearAccScale=<0.5 0.5 0.5>
    componentLinearVelScale=<0.65 0.65 0.55>
    gravityOverride {z=-980}
    physicsBodyDefinitions [{
      collisionType=CustomSphere
      sphereCollisionRadius=2
      boundBone {boneName=dyn_hair_strand_05_r}
      boxExtents=<23 23 23>
      constraintSetup {angularLimitsMax=<1 5 1>, angularLimitsMin=<-1 -5 -35>, angularTarget=<0 0 0>}
      localJointOffset=<5 -0.1 0.1>
     }]
    relativeSpaceBone {boneName=root}
    sphericalLimits [{limitRadius=18, drivingBone {boneName=upperarm_r}, sphereLocalOffset=<-12 -14 0>}]
   }
   n118 = Anim Dynamics - Bone: dyn_hair_strand_06_r(ComponentPose=@n117, Alpha=1, ExternalForce="0,0,0") {
    angularDampingOverride=0.99
    angularSpringConstant=75
    bAngularSpring=true
    bGravityOverrideInSimSpace=true
    bOverrideAngularDamping=true
    bUseGravityOverride=true
    bUseSphericalLimits=true
    simulationSpace=RootRelative
    alphaBoolBlend {blendOption=Linear}
    boundBone {boneName=dyn_hair_strand_06_r}
    componentAppliedLinearAccClamp=<3500 3500 3000>
    componentLinearAccScale=<0.5 0.5 0.5>
    componentLinearVelScale=<0.65 0.65 0.55>
    gravityOverride {z=-980}
    physicsBodyDefinitions [{
      collisionType=CustomSphere
      sphereCollisionRadius=2
      boundBone {boneName=dyn_hair_strand_06_r}
      boxExtents=<23 23 23>
      constraintSetup {angularLimitsMax=<1 5 1>, angularLimitsMin=<-1 -5 -35>, angularTarget=<0 0 0>}
      localJointOffset=<5 -0.1 0.1>
     }]
    relativeSpaceBone {boneName=root}
    sphericalLimits [{limitRadius=18, drivingBone {boneName=upperarm_r}, sphereLocalOffset=<-12 -14 0>}]
   }
   n119 = Anim Dynamics - Bone: dyn_hair_strand_07_r(ComponentPose=@n118, Alpha=1, ExternalForce="0,0,0") {
    angularDampingOverride=0.99
    angularSpringConstant=75
    bAngularSpring=true
    bGravityOverrideInSimSpace=true
    bOverrideAngularDamping=true
    bUseGravityOverride=true
    bUseSphericalLimits=true
    simulationSpace=RootRelative
    alphaBoolBlend {blendOption=Linear}
    boundBone {boneName=dyn_hair_strand_07_r}
    componentAppliedLinearAccClamp=<3500 3500 3000>
    componentLinearAccScale=<0.5 0.5 0.5>
    componentLinearVelScale=<0.65 0.65 0.55>
    gravityOverride {z=-980}
    physicsBodyDefinitions [{
      collisionType=CustomSphere
      sphereCollisionRadius=2
      boundBone {boneName=dyn_hair_strand_07_r}
      boxExtents=<23 23 23>
      constraintSetup {angularLimitsMax=<1 5 1>, angularLimitsMin=<-1 -5 -35>, angularTarget=<0 0 0>}
      localJointOffset=<5 -0.1 0.1>
     }]
    relativeSpaceBone {boneName=root}
    sphericalLimits [{limitRadius=18, drivingBone {boneName=upperarm_r}, sphereLocalOffset=<-12 -14 0>}]
   }
   n120 = Anim Dynamics - Bone: dyn_hair_strand_08_r(ComponentPose=@n119, Alpha=1, ExternalForce="0,0,0") {
    angularDampingOverride=0.99
    angularSpringConstant=75
    bAngularSpring=true
    bGravityOverrideInSimSpace=true
    bOverrideAngularDamping=true
    bUseGravityOverride=true
    bUseSphericalLimits=true
    simulationSpace=RootRelative
    alphaBoolBlend {blendOption=Linear}
    boundBone {boneName=dyn_hair_strand_08_r}
    componentAppliedLinearAccClamp=<3500 3500 3000>
    componentLinearAccScale=<0.5 0.5 0.5>
    componentLinearVelScale=<0.65 0.65 0.55>
    gravityOverride {z=-980}
    physicsBodyDefinitions [{
      collisionType=CustomSphere
      sphereCollisionRadius=2
      boundBone {boneName=dyn_hair_strand_08_r}
      boxExtents=<23 23 23>
      constraintSetup {angularLimitsMax=<1 5 1>, angularLimitsMin=<-1 -5 -35>, angularTarget=<0 0 0>}
      localJointOffset=<5 -0.1 0.1>
     }]
    relativeSpaceBone {boneName=root}
    sphericalLimits [{limitRadius=18, drivingBone {boneName=upperarm_r}, sphereLocalOffset=<-12 -14 0>}]
   }
   n121 = "Constraint - dyn_hair_strand_a_01_l (PP)"(ComponentPose=@n120, Alpha=1, ConstraintWeights_0=1, ConstraintWeights_1=0) {
    alphaBoolBlend {blendOption=Linear}
    boneToModify {boneName=dyn_hair_strand_a_01_l}
    constraintSetup [{targetBone {boneName=dyn_hair_back_01_l}}, {targetBone {boneName=head}}]
    constraintWeights [1,1]
   }
   n122 = "Constraint - dyn_hair_strand_a_02_l (PP)"(ComponentPose=@n121, Alpha=1, ConstraintWeights_0=0.8, ConstraintWeights_1=0.2) {
    alphaBoolBlend {blendOption=Linear}
    boneToModify {boneName=dyn_hair_strand_a_02_l}
    constraintSetup [{targetBone {boneName=dyn_hair_back_02_l}}, {targetBone {boneName=dyn_hair_back_03_l}}]
    constraintWeights [1,1]
   }
   n123 = "Constraint - dyn_hair_strand_a_03_l (PP)"!(ComponentPose=@n122, Alpha=1, ConstraintWeights_0=0.5, ConstraintWeights_1:float=0.5) -> (Pose) {
    alphaBoolBlend {blendOption=Linear}
    boneToModify {boneName=dyn_hair_strand_a_03_l}
    constraintSetup [{targetBone {boneName=dyn_hair_back_02_l}}, {targetBone {boneName=dyn_hair_back_03_l}}]
    constraintWeights [1,1]
   }
   n124 = "Constraint - dyn_hair_strand_a_04_l (PP)"!(ComponentPose=@n123, Alpha=1, ConstraintWeights_0=0.3, ConstraintWeights_1:float=0.7) -> (Pose) {
    alphaBoolBlend {blendOption=Linear}
    boneToModify {boneName=dyn_hair_strand_a_04_l}
    constraintSetup [{targetBone {boneName=dyn_hair_back_02_l}}, {targetBone {boneName=dyn_hair_back_03_l}}]
    constraintWeights [1,1]
   }
   n125 = "Constraint - dyn_hair_strand_a_05_l (PP)"!(ComponentPose=@n124, Alpha=1, ConstraintWeights_0=0.5, ConstraintWeights_1:float=0.5) -> (Pose) {
    alphaBoolBlend {blendOption=Linear}
    boneToModify {boneName=dyn_hair_strand_a_05_l}
    constraintSetup [{targetBone {boneName=dyn_hair_back_03_l}}, {targetBone {boneName=dyn_hair_back_04_l}}]
    constraintWeights [1,1]
   }
   n126 = "Constraint - dyn_hair_strand_a_06_l (PP)"!(ComponentPose=@n125, Alpha=1, ConstraintWeights_0=1, ConstraintWeights_1:float=0) -> (Pose) {
    alphaBoolBlend {blendOption=Linear}
    boneToModify {boneName=dyn_hair_strand_a_06_l}
    constraintSetup [{targetBone {boneName=dyn_hair_back_04_l}}, {targetBone {boneName=head}}]
    constraintWeights [1,1]
   }
   n127 = Anim Dynamics - Bone: dyn_hair_strand_a_03_l(ComponentPose=@n126, Alpha=1, ExternalForce="0,0,0") {
    angularDampingOverride=0.99
    angularSpringConstant=75
    bAngularSpring=true
    bGravityOverrideInSimSpace=true
    bOverrideAngularDamping=true
    bUseGravityOverride=true
    bUseSphericalLimits=true
    simulationSpace=RootRelative
    alphaBoolBlend {blendOption=Linear}
    boundBone {boneName=dyn_hair_strand_a_03_l}
    componentAppliedLinearAccClamp=<3500 3500 3000>
    componentLinearAccScale=<0.5 0.5 0.5>
    componentLinearVelScale=<0.65 0.65 0.55>
    gravityOverride {z=-980}
    physicsBodyDefinitions [{
      collisionType=CustomSphere
      sphereCollisionRadius=2
      boundBone {boneName=dyn_hair_strand_a_03_l}
      boxExtents=<23 23 23>
      constraintSetup {angularLimitsMax=<1 5 1>, angularLimitsMin=<-1 -5 -35>, angularTarget=<0 0 0>}
      localJointOffset=<5 -0.1 0.1>
     }]
    relativeSpaceBone {boneName=root}
    sphericalLimits [{limitRadius=18, drivingBone {boneName=upperarm_l}, sphereLocalOffset=<12 12 0>}]
   }
   n128 = Anim Dynamics - Bone: dyn_hair_strand_a_04_l(ComponentPose=@n127, Alpha=1, ExternalForce="0,0,0") {
    angularDampingOverride=0.99
    angularSpringConstant=75
    bAngularSpring=true
    bGravityOverrideInSimSpace=true
    bOverrideAngularDamping=true
    bUseGravityOverride=true
    bUseSphericalLimits=true
    simulationSpace=RootRelative
    alphaBoolBlend {blendOption=Linear}
    boundBone {boneName=dyn_hair_strand_a_04_l}
    componentAppliedLinearAccClamp=<3500 3500 3000>
    componentLinearAccScale=<0.5 0.5 0.5>
    componentLinearVelScale=<0.65 0.65 0.55>
    gravityOverride {z=-980}
    physicsBodyDefinitions [{
      collisionType=CustomSphere
      sphereCollisionRadius=2
      boundBone {boneName=dyn_hair_strand_a_04_l}
      boxExtents=<23 23 23>
      constraintSetup {angularLimitsMax=<1 5 1>, angularLimitsMin=<-1 -5 -35>, angularTarget=<0 0 0>}
      localJointOffset=<5 -0.1 0.1>
     }]
    relativeSpaceBone {boneName=root}
    sphericalLimits [{limitRadius=18, drivingBone {boneName=upperarm_l}, sphereLocalOffset=<12 12 0>}]
   }
   n129 = Anim Dynamics - Bone: dyn_hair_strand_a_05_l(ComponentPose=@n128, Alpha=1, ExternalForce="0,0,0") {
    angularDampingOverride=0.99
    angularSpringConstant=75
    bAngularSpring=true
    bGravityOverrideInSimSpace=true
    bOverrideAngularDamping=true
    bUseGravityOverride=true
    bUseSphericalLimits=true
    simulationSpace=RootRelative
    alphaBoolBlend {blendOption=Linear}
    boundBone {boneName=dyn_hair_strand_a_05_l}
    componentAppliedLinearAccClamp=<3500 3500 3000>
    componentLinearAccScale=<0.5 0.5 0.5>
    componentLinearVelScale=<0.65 0.65 0.55>
    gravityOverride {z=-980}
    physicsBodyDefinitions [{
      collisionType=CustomSphere
      sphereCollisionRadius=2
      boundBone {boneName=dyn_hair_strand_a_05_l}
      boxExtents=<23 23 23>
      constraintSetup {angularLimitsMax=<1 5 1>, angularLimitsMin=<-1 -5 -35>, angularTarget=<0 0 0>}
      localJointOffset=<5 -0.1 0.1>
     }]
    relativeSpaceBone {boneName=root}
    sphericalLimits [{limitRadius=18, drivingBone {boneName=upperarm_l}, sphereLocalOffset=<12 12 0>}]
   }
   n130 = Anim Dynamics - Bone: dyn_hair_strand_a_06_l(ComponentPose=@n129, Alpha=1, ExternalForce="0,0,0") {
    angularDampingOverride=0.99
    angularSpringConstant=75
    bAngularSpring=true
    bGravityOverrideInSimSpace=true
    bOverrideAngularDamping=true
    bUseGravityOverride=true
    bUseSphericalLimits=true
    simulationSpace=RootRelative
    alphaBoolBlend {blendOption=Linear}
    boundBone {boneName=dyn_hair_strand_a_06_l}
    componentAppliedLinearAccClamp=<3500 3500 3000>
    componentLinearAccScale=<0.5 0.5 0.5>
    componentLinearVelScale=<0.65 0.65 0.55>
    gravityOverride {z=-980}
    physicsBodyDefinitions [{
      collisionType=CustomSphere
      sphereCollisionRadius=2
      boundBone {boneName=dyn_hair_strand_a_06_l}
      boxExtents=<23 23 23>
      constraintSetup {angularLimitsMax=<1 5 1>, angularLimitsMin=<-1 -5 -35>, angularTarget=<0 0 0>}
      localJointOffset=<5 -0.1 0.1>
     }]
    relativeSpaceBone {boneName=root}
    sphericalLimits [{limitRadius=18, drivingBone {boneName=upperarm_l}, sphereLocalOffset=<12 12 0>}]
   }
   "Save cached pose 'HairBeard  on/off'"(Pose=Component To Local(ComponentPose=@n130)) {cacheName="HairBeard  on/off"}
   Get bIsDBNO()
   Get bIsDBNO()
   Get is_wearing_backbling_alpha() -> (is_wearing_backbling_alpha:double)
  ]}
 EventGraph {nodes [
   Event BlueprintUpdateAnimation()
   > n1 = Sequence!(execute) -> (then_0, then_1, then_2:Exec, then_3:Exec, then_4:Exec)
   > n2 = Gravity Override State Function DEPRECATED(CharacterPartAnimInstance=Self-Reference(), GameTimeInSeconds=GetGameTimeInSeconds())
   > Set gravity_override(gravity_override=@n2)
   Is Valid(exec=@n1.then_1, InputObject=Get pawn_owner() -> (pawn_owner:Player Pawn Athena Object Reference))
   > Do Once()
   n3 = CheckForBackbling()
   > n4 = Branch(Condition=OR Boolean!(A=@n3, B=@n3.bIsBackblingACape, C=@n3.bIsBackblingLong) -> (ReturnValue))
   > Set is_wearing_backbling_alpha(is_wearing_backbling_alpha:double=1) -> (Output_Get:double)
   > Set is_wearing_backling_REV_alpha(is_wearing_backling_REV_alpha:double=0) -> (Output_Get:double)
   > Set backpack(backpack:bool=true) -> (Output_Get:bool)
   Set is_wearing_backbling_alpha(execute=@n4.else, is_wearing_backbling_alpha=0)
   > Set is_wearing_backling_REV_alpha(is_wearing_backling_REV_alpha=1)
   > Set backpack(backpack=false)
   n5 = PlayerMovement(execute=@n1.then_2, AnimInstance=Self-Reference(), XDotProductCoefficient=1, YDotProductCoefficient=1, ZDotProductCoefficient=1)
   n6 = MapRangeClamped(Value=@n5.YMovement, InRangeA=0, InRangeB=-1)
   n7 = float * float(A="Absolute (Float)"(A=@n5), B=0.5)
   > Set BwdMovement(BwdMovement:double=MapRangeClamped(
     Value=float + float(A=@n7, B=@n6) {is=title K2Node_CommutativeAssociativeBinaryOperator}
     InRangeA=0
     InRangeB=1.1
     OutRangeA=0.75
     OutRangeB=0
    )) -> (Output_Get:double)
   n8 = Branch(execute=@n1.then_3, Condition=Get bIsSkyDiving())
   > n9 = Branch(Condition="float > float"(A=Get PawnVelocityZ(), B=50) {is=title K2Node_CallFunction})
   > Set BwdMovement(BwdMovement=float * float(A=Get BwdMovement(), B=0))
   Set BwdMovement(execute=[@n9.else, @n8.else], BwdMovement=float * float(A=Get BwdMovement(), B=1))
   n10 = GetAnimInstance(self=GetOwnerSkeletalMeshComponent())
   Is Valid(exec=@n1.then_4, InputObject=@n10)
   > Set is_playing_custom_idle_alpha(is_playing_custom_idle_alpha:double=GetCurveValue(self=@n10, CurveName=emote_builtIn)) -> (Output_Get:double)
   Event OnCharacterCustomizationLoaded()
   > Delay(Duration=0.1)
   > ResetDynamics(InTeleportType=ResetPhysics)
   > ResetDynamics(InTeleportType=TeleportPhysics)
   n11 = CheckForBackbling()
   > Set isWearingPackBool(isWearingPackBool:bool=OR Boolean(A=@n11, B=@n11.bIsBackblingACape, C=@n11.bIsBackblingLong)) -> (Output_Get:bool)
   Event BlueprintInitializeAnimation()
   > Set body_mesh(body_mesh:Skeletal Mesh Component Object Reference=GetOwnerSkeletalMeshComponent()) -> (Output_Get:Skeletal Mesh Component Object Reference)
   > n12 = Cast To PlayerPawn_Athena(Object=TryGetPawnOwner())
   > Set pawn_owner(pawn_owner:Player Pawn Athena Object Reference=@n12) -> (Output_Get:Player Pawn Athena Object Reference)
   GetAnimInstance!(Context:"Anim Execution Context Structure (by ref)") -> (ReturnValue)
   AND Boolean(A=@n11, B=NOT Boolean(A=@n11.bIsBackblingLong))
   Reroute Node(InputPin:Wildcard) -> (OutputPin:Wildcard)
  ]}
]
variables [
 backpack {type=boolean}
 body_mesh {type=SkeletalMeshComponent}
 BwdMovement {type=number}
 gravity_override {type=Vector}
 is_playing_custom_idle_alpha {type=number}
 is_wearing_backbling_alpha {type=number}
 is_wearing_backling_REV_alpha {type=number}
 isWearingPackBool {type=boolean}
 pawn_owner {type=PlayerPawn_Athena_C}
]
